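# Control Rig for the F_MED_SilverCanine female body that procedurally poses its skirt bones to follow leg motion and gameplay state.
# A swing-twist DeltaRotation function remaps thigh, calf and foot angles into weights, many cached in SPR variables.
# These weights and flags like skydive, parachute, sprint and DBNO drive layered ModifyTransforms offsets, some smoothed by AlphaInterp.
# Parent constraints average back-middle skirt bones from left and right chains and place a rose bone.
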ControlRigBlueprint /BRCosmetics/Characters/Player/Female/Medium/Bodies/F_MED_SilverCanine/Meshes/F_MED_SilverCanine_CtrlRig
graphs [
 Multiply_12_ContainedGraph {nodes [
   n1 = RigVMFunctionEntryNode()
   RigVMFunctionReturnNode(Result=&RigVMFunction_MathDoubleMul(B=@n1.C, A=&RigVMFunction_MathDoubleMul(A=@n1.A, B=@n1.B).Result).Result)
  ]}
 Multiply_12_ContainedGraph {nodes [
   n1 = RigVMFunctionEntryNode()
   RigVMFunctionReturnNode(Result=&RigVMFunction_MathDoubleMul(B=@n1.C, A=&RigVMFunction_MathDoubleMul(A=@n1.A, B=@n1.B).Result).Result)
  ]}
 Multiply_19_ContainedGraph {nodes [
   n1 = RigVMFunctionEntryNode()
   Multiply = &RigVMFunction_MathDoubleMul(A=@n1.A, B=@n1.B)
   RigVMFunctionReturnNode(Result=&RigVMFunction_MathDoubleMul(B=@n1.C, A=@Multiply.Result).Result)
  ]}
 Multiply_19_ContainedGraph {nodes [
   n1 = RigVMFunctionEntryNode()
   RigVMFunctionReturnNode(Result=&RigVMFunction_MathDoubleMul(B=@n1.C, A=&RigVMFunction_MathDoubleMul(A=@n1.A, B=@n1.B).Result).Result)
  ]}
 Multiply_19_ContainedGraph {nodes [
   n1 = RigVMFunctionEntryNode()
   RigVMFunctionReturnNode(Result=&RigVMFunction_MathDoubleMul(B=@n1.C, A=&RigVMFunction_MathDoubleMul(A=@n1.A, B=@n1.B).Result).Result)
  ]}
 Multiply_4_ContainedGraph {nodes [
   n1 = RigVMFunctionEntryNode()
   n2 = &RigVMFunction_MathDoubleMul(A=@n1.A, B=@n1.B)
   RigVMFunctionReturnNode(Result=&RigVMFunction_MathDoubleMul(B=@n1.D, A=&RigVMFunction_MathDoubleMul(B=@n1.C, A=@n2.Result).Result).Result)
  ]}
 New Function_ContainedGraph {nodes [
   n1 = RigVMFunctionEntryNode()
   RigVMFunction_Sequence = &RigVMFunction_Sequence(ExecuteContext=@n1.ExecuteContext)
   &RigUnit_ParentConstraint(
    AdvancedSettings=(InterpolationType=Shortest,RotationOrderForFilter=XZY)
    Child=(Type=Bone,Name="dyn_skirt_back_m_01")
    Filter=(TranslationFilter=(bX=false,bY=false,bZ=false),RotationFilter=(bX=True,bY=true,bZ=True),ScaleFilter=(bX=True,bY=True,bZ=True))
    Parents=((Item=(Type=Bone,Name="dyn_skirt_back_l_01"),Weight=0.5),(Item=(Type=Bone,Name="dyn_skirt_back_r_01"),Weight=0.5))
    Weight=1
    ExecutePin=@RigVMFunction_Sequence.A
   )
   > &RigUnit_ParentConstraint(
    AdvancedSettings=(InterpolationType=Shortest,RotationOrderForFilter=XZY)
    Child=(Type=Bone,Name="dyn_skirt_back_m_02")
    Filter=(TranslationFilter=(bX=false,bY=false,bZ=false),RotationFilter=(bX=true,bY=true,bZ=true),ScaleFilter=(bX=True,bY=True,bZ=True))
    Parents=((Item=(Type=Bone,Name="dyn_skirt_back_l_02"),Weight=0.5),(Item=(Type=Bone,Name="dyn_skirt_back_r_02"),Weight=0.5))
    Weight=1
   )
   > &RigUnit_ParentConstraint(
    AdvancedSettings=(InterpolationType=Shortest,RotationOrderForFilter=XZY)
    Child=(Type=Bone,Name="dyn_skirt_back_m_03")
    Filter=(TranslationFilter=(bX=false,bY=false,bZ=false),RotationFilter=(bX=True,bY=true,bZ=True),ScaleFilter=(bX=True,bY=True,bZ=True))
    Parents=((Item=(Type=Bone,Name="dyn_skirt_back_l_03"),Weight=0.5),(Item=(Type=Bone,Name="dyn_skirt_back_r_03"),Weight=0.5))
    Weight=1
   )
   > &RigUnit_ParentConstraint(
    AdvancedSettings=(InterpolationType=Shortest,RotationOrderForFilter=XZY)
    Child=(Type=Bone,Name="dyn_skirt_back_m_04")
    Filter=(TranslationFilter=(bX=false,bY=false,bZ=false),RotationFilter=(bX=True,bY=true,bZ=True),ScaleFilter=(bX=True,bY=True,bZ=True))
    Parents=((Item=(Type=Bone,Name="dyn_skirt_back_l_04"),Weight=0.5),(Item=(Type=Bone,Name="dyn_skirt_back_r_04"),Weight=0.5))
    Weight=1
   )
   > &RigUnit_ParentConstraint(
    AdvancedSettings=(RotationOrderForFilter=XZY)
    Child=(Type=Bone,Name="base_rose_r")
    Parents=((Item=(Type=Bone,Name="spine_04"),Weight=0.0205),(Item=(Type=Bone,Name="upperarm_twist_01_r"),Weight=0.0829),(Item=(Type=Bone,Name="spine_05"),Weight=0.2173),(Item=(Type=Bone,Name="clavicle_r"),Weight=0.6793))
    Weight=1
   )
   > &RigUnit_ModifyTransforms(
    ItemToModify=((Item=(Type=Bone,Name="dyn_offset_hip"),Transform=(Rotation=(X=0,Y=0,Z=0,W=1),Translation=(X=0,Y=0,Z=0),Scale3D=(X=0.7,Y=0.7,Z=0.7))))
    Weight=1
    WeightMinimum=0
   )
   RigVMFunctionReturnNode(ExecuteContext=@RigVMFunction_Sequence.B)
  ]}
 New Function_ContainedGraph {nodes [
   n1 = RigVMFunctionEntryNode()
   RigVMFunction_Sequence = &RigVMFunction_Sequence(ExecuteContext=@n1.ExecuteContext)
   DeltaRotation v4 = RigVMFunctionReferenceNode(
    Rotation Order=ZYX
    child=(Type=Bone,Name="thigh_l")
    debugRotation=false
    onlyTwist=false
    outputMinMax=(X=0,Y=1)
    parent=(Type=Null,Name="thigh_l_null")
    premult=1
    rotationMinMax=(X=0,Y=-120)
    triggerAxis=Z
    twistAxis=(X=1,Y=0,Z=0)
    ExecuteContext=@RigVMFunction_Sequence.A
   )
   > RigVMVariableNode(Value=@"DeltaRotation v4".outputWeight) {variable=SPR L Leg Fwd}
   > DeltaRotation v4_1 = RigVMFunctionReferenceNode(
    Rotation Order=ZYX
    child=(Type=Bone,Name="thigh_r")
    debugRotation=false
    onlyTwist=false
    outputMinMax=(X=0,Y=1)
    parent=(Type=Null,Name="thigh_r_null")
    premult=1
    rotationMinMax=(X=0,Y=-120)
    triggerAxis=Z
    twistAxis=(X=1,Y=0,Z=0)
   )
   > RigVMVariableNode(Value=@"DeltaRotation v4_1".outputWeight) {variable=SPR R Leg Fwd}
   > DeltaRotation v4_2 = RigVMFunctionReferenceNode(
    Rotation Order=ZYX
    absoluteValue=false
    child=(Type=Bone,Name="thigh_l")
    debugRotation=false
    onlyTwist=false
    outputMinMax=(X=0,Y=1)
    parent=(Type=Null,Name="thigh_l_null")
    premult=1
    rotationMinMax=(X=0,Y=120)
    triggerAxis=Z
    twistAxis=(X=1,Y=0,Z=0)
   )
   > RigVMVariableNode(Value=@"DeltaRotation v4_2".outputWeight) {variable=SPR L Leg Bkd}
   > DeltaRotation v4_3 = RigVMFunctionReferenceNode(
    Rotation Order=ZYX
    child=(Type=Bone,Name="thigh_r")
    outputMinMax=(X=0,Y=1)
    parent=(Type=Null,Name="thigh_r_null")
    premult=1
    rotationMinMax=(X=0,Y=120)
    triggerAxis=Z
    twistAxis=(X=1,Y=0,Z=0)
   )
   > RigVMVariableNode(Value=@"DeltaRotation v4_3".outputWeight) {variable=SPR R Leg Bkd}
   > DeltaRotation v4_4 = RigVMFunctionReferenceNode(
    Rotation Order=ZYX
    child=(Type=Bone,Name="thigh_l")
    outputMinMax=(X=0,Y=1)
    parent=(Type=Null,Name="thigh_l_null")
    premult=1
    rotationMinMax=(X=0,Y=-90)
    triggerAxis=Y
    twistAxis=(X=1,Y=0,Z=0)
   )
   > RigVMVariableNode(Value=@"DeltaRotation v4_4".outputWeight) {variable=SPR L Leg Otw}
   > DeltaRotation v4_5 = RigVMFunctionReferenceNode(
    Rotation Order=ZYX
    child=(Type=Bone,Name="thigh_r")
    debugRotation=false
    outputMinMax=(X=0,Y=1)
    parent=(Type=Null,Name="thigh_r_null")
    premult=1
    rotationMinMax=(X=0,Y=-90)
    triggerAxis=Y
    twistAxis=(X=1,Y=0,Z=0)
   )
   > RigVMVariableNode(Value=@"DeltaRotation v4_5".outputWeight) {variable=SPR R Leg Otw}
   > DeltaRotation v4_6 = RigVMFunctionReferenceNode(
    Rotation Order=ZYX
    child=(Type=Bone,Name="thigh_l")
    debugRotation=false
    onlyTwist=false
    outputMinMax=(X=0,Y=1)
    parent=(Type=Null,Name="thigh_l_null")
    premult=1
    rotationMinMax=(X=0,Y=60)
    triggerAxis=Y
    twistAxis=(X=1,Y=0,Z=0)
   )
   > RigVMVariableNode(Value=@"DeltaRotation v4_6".outputWeight) {variable=SPR L Leg Itw}
   > DeltaRotation v4_7 = RigVMFunctionReferenceNode(
    Rotation Order=ZYX
    child=(Type=Bone,Name="thigh_r")
    debugRotation=false
    outputMinMax=(X=0,Y=1)
    parent=(Type=Null,Name="thigh_r_null")
    premult=1
    rotationMinMax=(X=0,Y=60)
    triggerAxis=Y
    twistAxis=(X=1,Y=0,Z=0)
   )
   > RigVMVariableNode(Value=@"DeltaRotation v4_7".outputWeight) {variable=SPR R Leg Itw}
   > DeltaRotation v4_8 = RigVMFunctionReferenceNode(
    Rotation Order=ZYX
    child=(Type=Bone,Name="calf_l")
    debugRotation=false
    onlyTwist=false
    outputMinMax=(X=0,Y=1)
    parent=(Type=Null,Name="calf_l_null")
    premult=1
    rotationMinMax=(X=30,Y=120)
    triggerAxis=Z
    twistAxis=(X=1,Y=0,Z=0)
   )
   > RigVMVariableNode(Value=@"DeltaRotation v4_8".outputWeight) {variable=SPR L calf Bkd}
   > DeltaRotation v4_9 = RigVMFunctionReferenceNode(
    Rotation Order=ZYX
    child=(Type=Bone,Name="calf_r")
    debugRotation=false
    onlyTwist=false
    outputMinMax=(X=0,Y=1)
    parent=(Type=Null,Name="calf_r_null")
    premult=1
    rotationMinMax=(X=30,Y=120)
    triggerAxis=Z
    twistAxis=(X=1,Y=0,Z=0)
   )
   > RigVMVariableNode(Value=@"DeltaRotation v4_9".outputWeight) {variable=SPR R calf Bkd}
   > DeltaRotation v4_14 = RigVMFunctionReferenceNode(
    Rotation Order=ZYX
    child=(Type=Bone,Name="calf_l")
    debugRotation=false
    onlyTwist=false
    outputMinMax=(X=0,Y=1)
    parent=(Type=Null,Name="calf_l_null")
    premult=1
    rotationMinMax=(X=120,Y=140)
    triggerAxis=Z
    twistAxis=(X=1,Y=0,Z=0)
   )
   > RigVMVariableNode(Value=@"DeltaRotation v4_14".outputWeight) {variable=SPR L calf ExtremeBkd}
   > DeltaRotation v4_15 = RigVMFunctionReferenceNode(
    Rotation Order=ZYX
    child=(Type=Bone,Name="calf_r")
    debugRotation=false
    onlyTwist=false
    outputMinMax=(X=0,Y=1)
    parent=(Type=Null,Name="calf_r_null")
    premult=1
    rotationMinMax=(X=120,Y=140)
    triggerAxis=Z
    twistAxis=(X=1,Y=0,Z=0)
   )
   > RigVMVariableNode(Value=@"DeltaRotation v4_15".outputWeight) {variable=SPR R calf ExtremeBkd}
   > DeltaRotation v4_10 = RigVMFunctionReferenceNode(
    Rotation Order=ZYX
    child=(Type=Bone,Name="thigh_l")
    debugRotation=false
    onlyTwist=true
    outputMinMax=(X=0,Y=1)
    parent=(Type=Null,Name="thigh_l_null")
    premult=1
    rotationMinMax=(X=0,Y=130)
    triggerAxis=X
    twistAxis=(X=1,Y=0,Z=0)
   )
   > RigVMVariableNode(Value=@"DeltaRotation v4_10".outputWeight) {variable=SPR L Leg In Twist}
   > DeltaRotation v4_11 = RigVMFunctionReferenceNode(
    Rotation Order=ZYX
    child=(Type=Bone,Name="thigh_r")
    debugRotation=false
    onlyTwist=true
    outputMinMax=(X=0,Y=1)
    parent=(Type=Null,Name="thigh_r_null")
    premult=1
    rotationMinMax=(X=0,Y=130)
    triggerAxis=X
    twistAxis=(X=1,Y=0,Z=0)
   )
   > RigVMVariableNode(Value=@"DeltaRotation v4_11".outputWeight) {variable=SPR R Leg In Twist}
   > DeltaRotation v4_12 = RigVMFunctionReferenceNode(
    Rotation Order=ZYX
    child=(Type=Bone,Name="thigh_l")
    debugRotation=false
    onlyTwist=true
    outputMinMax=(X=0,Y=1)
    parent=(Type=Null,Name="thigh_l_null")
    premult=1
    rotationMinMax=(X=0,Y=-130)
    triggerAxis=X
    twistAxis=(X=1,Y=0,Z=0)
   )
   > RigVMVariableNode(Value=@"DeltaRotation v4_12".outputWeight) {variable=SPR L Leg Out Twist}
   > DeltaRotation v4_13 = RigVMFunctionReferenceNode(
    Rotation Order=ZYX
    child=(Type=Bone,Name="thigh_r")
    debugRotation=false
    onlyTwist=true
    outputMinMax=(X=0,Y=1)
    parent=(Type=Null,Name="thigh_r_null")
    premult=1
    rotationMinMax=(X=0,Y=-130)
    triggerAxis=X
    twistAxis=(X=1,Y=0,Z=0)
   )
   > RigVMVariableNode(Value=@"DeltaRotation v4_13".outputWeight) {variable=SPR R Leg Out Twist}
   n2 = &RigVMFunction_AnimEvalRichCurve(
    Curve="(EditorCurveData=(Keys=((InterpMode=RCIM_Cubic,TangentMode=RCTM_Break,TangentWeightMode=RCTWM_WeightedNone,Time=0,Value=0,ArriveTangent=0,ArriveTangentWeight=0,LeaveTangent=0,LeaveTangentWeight=0),(InterpMode=RCIM_Cubic,TangentMode=RCTM_Break,TangentWeight ...(truncated)"
    SourceMinimum=0.5
    TargetMinimum=0
    Value=&RigVMFunction_MathDoubleRemap(
     SourceMaximum=30
     SourceMinimum=60
     TargetMaximum=0
     TargetMinimum=1
     bClamp=False
     Value=&RigUnit_GetRelativeTransformForItem(Child=(Type=Bone,Name="pelvis"), Parent=(Type=Bone,Name="root"), bChildInitial=False, bParentInitial=False).RelativeTransform.Translation.Z
    ).Result
   )
   > RigVMVariableNode(Value=@n2.Result) {variable=PelvisProx}
   RigVMFunctionReturnNode(ExecuteContext=@RigVMFunction_Sequence.B)
  ]}
 New Function_ContainedGraph {nodes [
   n1 = RigVMFunctionEntryNode()
   RigVMFunction_MathBoolToInteger = &RigVMFunction_MathBoolToInteger(Value=@n1.onlyTwist)
   RigVMFunction_MathQuaternionSwingTwist = &RigVMFunction_MathQuaternionSwingTwist(
    TwistAxis=@n1.twistAxis
    Input=&RigUnit_GetRelativeTransformForItem(bChildInitial=False, bParentInitial=False, Child=@n1.child, Parent=@n1.parent).RelativeTransform.Rotation
   )
   n2 = RigVMDispatchNode(
    Values=((),())
    Index=@RigVMFunction_MathBoolToInteger.Result
    Values.0=@RigVMFunction_MathQuaternionSwingTwist.Swing
    Values.1=@RigVMFunction_MathQuaternionSwingTwist.Twist
   ) {template="DISPATCH_RigVMDispatch_SelectInt32(in Index,in Values,out Result)"}
   RigVMFunction_MathQuaternionToEuler = &RigVMFunction_MathQuaternionToEuler(Value=@n2.Result, RotationOrder=@n1.Rotation Order)
   > DISPATCH_RigVMDispatch_Print = RigVMDispatchNode(ScreenColor=(R=1,G=1,B=1,A=1), ScreenDuration=1.5, Enabled=@n1.debugRotation, Value=@RigVMFunction_MathQuaternionToEuler.Result) {template="DISPATCH_RigVMDispatch_Print(in Prefix,in Value,in Enabled,in ScreenDuration,in ScreenColor)"}
   Cast = RigVMDispatchNode(Value=@n1.triggerAxis) {template="DISPATCH_RigVMDispatch_CastEnumToInt(in Value,out Result)"}
   DISPATCH_RigVMDispatch_SelectInt32 = RigVMDispatchNode(
    Values=(0,0,0)
    Index=@Cast.Result
    Values.0=@RigVMFunction_MathQuaternionToEuler.Result.X
    Values.1=@RigVMFunction_MathQuaternionToEuler.Result.Y
    Values.2=@RigVMFunction_MathQuaternionToEuler.Result.Z
   ) {template="DISPATCH_RigVMDispatch_SelectInt32(in Index,in Values,out Result)"}
   Remap = &RigVMFunction_MathDoubleRemap(
    bClamp=False
    Value=@DISPATCH_RigVMDispatch_SelectInt32.Result
    TargetMinimum=@n1.outputMinMax.X
    TargetMaximum=@n1.outputMinMax.Y
    SourceMinimum=@n1.rotationMinMax.X
    SourceMaximum=@n1.rotationMinMax.Y
   )
   Absolute = &RigVMFunction_MathDoubleAbs(Value=@Remap.Result)
   Multiply = &RigVMFunction_MathDoubleMul(
    A=&RigVMFunction_MathDoubleClamp(Value=RigVMDispatchNode(True=@Absolute.Result, Condition=@n1.absoluteValue, False=@Remap.Result) {template="DISPATCH_RigVMDispatch_If(in Condition,in True,in False,out Result)"}.Result).Result
    B=@n1.premult
   )
   > RigVMFunctionReturnNode(outputQuat=@n2.Result, outputWeight=@Multiply.Result)
  ]}
 New Function_ContainedGraph {nodes [
   n1 = RigVMFunctionEntryNode()
   RigVMFunction_Sequence = &RigVMFunction_Sequence(ExecuteContext=@n1.ExecuteContext)
   &RigUnit_ModifyTransforms(
    ItemToModify="((Item=(Type=Bone,Name=\"dyn_skirt_side_l_01\"),Transform=(Rotation=(X=0,Y=0,Z=-0.866,W=0.5),Translation=(X=0,Y=0,Z=0),Scale3D=(X=1,Y=1,Z=1))),(Item=(Type=Bone,Name=\"dyn_skirt_back_l_01\"),Transform=(Rotation=(X=0,Y=0,Z=-0.866,W=0.5),Translation=(X=0,Y=10,Z=0 ...(truncated)"
    WeightMinimum=0
    ExecutePin=@RigVMFunction_Sequence.A
    Weight=RigVMVariableNode() {variable=SPR L Leg Fwd}.Value
   )
   > &RigUnit_ModifyTransforms(
    ItemToModify="((Item=(Type=Bone,Name=\"dyn_skirt_side_r_01\"),Transform=(Rotation=(X=0,Y=0,Z=-0.866,W=0.5),Translation=(X=0,Y=0,Z=0),Scale3D=(X=1,Y=1,Z=1))),(Item=(Type=Bone,Name=\"dyn_skirt_back_r_01\"),Transform=(Rotation=(X=0,Y=0,Z=-0.866,W=0.5),Translation=(X=0,Y=10,Z=0 ...(truncated)"
    WeightMinimum=0
    Weight=RigVMVariableNode() {variable=SPR R Leg Fwd}.Value
   )
   > DeltaRotation v4_7 = RigVMFunctionReferenceNode(
    Rotation Order=ZYX
    child=(Type=Bone,Name="thigh_l")
    debugRotation=false
    outputMinMax=(X=0,Y=1)
    parent=(Type=Null,Name="thigh_l_null")
    premult=1
    rotationMinMax=(X=0,Y=-30)
    triggerAxis=Z
    twistAxis=(X=1,Y=0,Z=0)
   )
   > DeltaRotation v4_10 = RigVMFunctionReferenceNode(
    Rotation Order=ZYX
    child=(Type=Bone,Name="thigh_r")
    outputMinMax=(X=0,Y=1)
    parent=(Type=Null,Name="thigh_r_null")
    premult=1
    rotationMinMax=(X=0,Y=-30)
    triggerAxis=Z
    twistAxis=(X=1,Y=0,Z=0)
   )
   > &RigUnit_ModifyTransforms(
    ItemToModify=((Item=(Type=Bone,Name="dyn_skirt_front_m_01"),Transform=(Rotation=(X=0,Y=0.1774,Z=0,W=0.9841),Translation=(X=3.8469,Y=0.0001,Z=-1.4331),Scale3D=(X=1,Y=1,Z=1))))
    WeightMinimum=0
    Weight=&RigVMFunction_MathDoubleMul(A=@"DeltaRotation v4_10".outputWeight, B=@"DeltaRotation v4_7".outputWeight).Result
   )
   > DeltaRotation v4_11 = RigVMFunctionReferenceNode(
    Rotation Order=ZYX
    child=(Type=Bone,Name="thigh_l")
    debugRotation=false
    outputMinMax=(X=0,Y=1)
    parent=(Type=Null,Name="thigh_l_null")
    premult=1
    rotationMinMax=(X=-35,Y=-50)
    triggerAxis=Z
    twistAxis=(X=1,Y=0,Z=0)
   )
   > DeltaRotation v4_12 = RigVMFunctionReferenceNode(
    Rotation Order=ZYX
    child=(Type=Bone,Name="thigh_r")
    outputMinMax=(X=0,Y=1)
    parent=(Type=Null,Name="thigh_r_null")
    premult=1
    rotationMinMax=(X=-35,Y=-50)
    triggerAxis=Z
    twistAxis=(X=1,Y=0,Z=0)
   )
   > &RigUnit_ModifyTransforms(
    ItemToModify=((Item=(Type=Bone,Name="dyn_skirt_front_m_01"),Transform=(Rotation=(X=0.0693,Y=0.131,Z=-0.0766,W=0.986),Translation=(X=0.7705,Y=-0.1853,Z=-1.6562),Scale3D=(X=1,Y=1,Z=1))))
    WeightMinimum=0
    Weight=&RigVMFunction_MathDoubleMul(A=@"DeltaRotation v4_11".outputWeight, B=@"DeltaRotation v4_12".outputWeight).Result
   )
   > DeltaRotation v4_15 = RigVMFunctionReferenceNode(
    Rotation Order=ZYX
    child=(Type=Bone,Name="thigh_l")
    debugRotation=false
    outputMinMax=(X=0,Y=1)
    parent=(Type=Null,Name="thigh_l_null")
    premult=1
    rotationMinMax=(X=-55,Y=-80)
    triggerAxis=Z
    twistAxis=(X=1,Y=0,Z=0)
   )
   > DeltaRotation v4_16 = RigVMFunctionReferenceNode(
    Rotation Order=ZYX
    child=(Type=Bone,Name="thigh_r")
    outputMinMax=(X=0,Y=1)
    parent=(Type=Null,Name="thigh_r_null")
    premult=1
    rotationMinMax=(X=-55,Y=-80)
    triggerAxis=Z
    twistAxis=(X=1,Y=0,Z=0)
   )
   > &RigUnit_ModifyTransforms(
    ItemToModify=((Item=(Type=Bone,Name="dyn_skirt_front_m_01"),Transform=(Rotation=(X=0.021,Y=0.2268,Z=-0.0897,W=0.9696),Translation=(X=0,Y=0,Z=0),Scale3D=(X=1,Y=1,Z=1))))
    WeightMinimum=0
    Weight=&RigVMFunction_MathDoubleMul(A=@"DeltaRotation v4_15".outputWeight, B=@"DeltaRotation v4_16".outputWeight).Result
   )
   > DeltaRotation v4_13 = RigVMFunctionReferenceNode(
    Rotation Order=ZYX
    child=(Type=Bone,Name="thigh_l")
    debugRotation=false
    outputMinMax=(X=0,Y=1)
    parent=(Type=Null,Name="thigh_l_null")
    premult=1
    rotationMinMax=(X=-85,Y=-120)
    triggerAxis=Z
    twistAxis=(X=1,Y=0,Z=0)
   )
   > DeltaRotation v4_14 = RigVMFunctionReferenceNode(
    Rotation Order=ZYX
    child=(Type=Bone,Name="thigh_r")
    outputMinMax=(X=0,Y=1)
    parent=(Type=Null,Name="thigh_r_null")
    premult=1
    rotationMinMax=(X=-85,Y=-120)
    triggerAxis=Z
    twistAxis=(X=1,Y=0,Z=0)
   )
   > &RigUnit_ModifyTransforms(
    ItemToModify=((Item=(Type=Bone,Name="dyn_skirt_front_m_01"),Transform=(Rotation=(X=0.0479,Y=0.4028,Z=-0.3622,W=0.8392),Translation=(X=0,Y=0,Z=0),Scale3D=(X=1,Y=1,Z=1))))
    WeightMinimum=0
    Weight=&RigVMFunction_MathDoubleMul(A=@"DeltaRotation v4_13".outputWeight, B=@"DeltaRotation v4_14".outputWeight).Result
   )
   > &RigUnit_ModifyTransforms(
    ItemToModify="((Item=(Type=Bone,Name=\"dyn_skirt_side_l_01\"),Transform=(Rotation=(X=0,Y=0,Z=0,W=1),Translation=(X=0,Y=0,Z=0),Scale3D=(X=1,Y=1,Z=1))),(Item=(Type=Bone,Name=\"dyn_skirt_back_l_01\"),Transform=(Rotation=(X=0.0002,Y=-0.0444,Z=0.0015,W=-0.999),Translation=(X=16. ...(truncated)"
    WeightMinimum=0
    Weight=RigVMVariableNode() {variable=SPR L Leg Bkd}.Value
   )
   > &RigUnit_ModifyTransforms(
    ItemToModify="((Item=(Type=Bone,Name=\"dyn_skirt_side_r_01\"),Transform=(Rotation=(X=0,Y=0,Z=0.024,W=-0.9997),Translation=(X=0,Y=0,Z=0),Scale3D=(X=1,Y=1,Z=1))),(Item=(Type=Bone,Name=\"dyn_skirt_back_r_01\"),Transform=(Rotation=(X=0.005,Y=0.0182,Z=-0.0304,W=-0.9994),Translat ...(truncated)"
    WeightMinimum=0
    Weight=RigVMVariableNode() {variable=SPR R Leg Bkd}.Value
   )
   > DeltaRotation v4_1 = RigVMFunctionReferenceNode(
    Rotation Order=ZYX
    child=(Type=Bone,Name="thigh_l")
    debugRotation=false
    outputMinMax=(X=0,Y=1)
    parent=(Type=Null,Name="thigh_l_null")
    premult=1
    rotationMinMax=(X=0,Y=30)
    triggerAxis=Z
    twistAxis=(X=1,Y=0,Z=0)
   )
   > DeltaRotation v4_2 = RigVMFunctionReferenceNode(
    Rotation Order=ZYX
    child=(Type=Bone,Name="thigh_r")
    outputMinMax=(X=0,Y=1)
    parent=(Type=Null,Name="thigh_r_null")
    premult=1
    rotationMinMax=(X=0,Y=30)
    triggerAxis=Z
    twistAxis=(X=1,Y=0,Z=0)
   )
   > &RigUnit_ModifyTransforms(
    ItemToModify=((Item=(Type=Bone,Name="dyn_skirt_back_m_01"),Transform=(Rotation=(X=-0.003,Y=-0.0035,Z=0.2617,W=0.9651),Translation=(X=0,Y=0,Z=0),Scale3D=(X=1,Y=1,Z=1))))
    WeightMinimum=0
    Weight=&RigVMFunction_MathDoubleMul(B=@"DeltaRotation v4_1".outputWeight, A=@"DeltaRotation v4_2".outputWeight).Result
   )
   > DeltaRotation v4_3 = RigVMFunctionReferenceNode(
    Rotation Order=ZYX
    child=(Type=Bone,Name="thigh_l")
    debugRotation=false
    outputMinMax=(X=0,Y=1)
    parent=(Type=Null,Name="thigh_l_null")
    premult=1
    rotationMinMax=(X=35,Y=80)
    triggerAxis=Z
    twistAxis=(X=1,Y=0,Z=0)
   )
   > DeltaRotation v4_4 = RigVMFunctionReferenceNode(
    Rotation Order=ZYX
    child=(Type=Bone,Name="thigh_r")
    outputMinMax=(X=0,Y=1)
    parent=(Type=Null,Name="thigh_r_null")
    premult=1
    rotationMinMax=(X=35,Y=80)
    triggerAxis=Z
    twistAxis=(X=1,Y=0,Z=0)
   )
   > &RigUnit_ModifyTransforms(
    ItemToModify=((Item=(Type=Bone,Name="dyn_skirt_back_m_01"),Transform=(Rotation=(X=-0.0079,Y=0.0119,Z=0.3924,W=0.9197),Translation=(X=0,Y=0,Z=0),Scale3D=(X=1,Y=1,Z=1))))
    WeightMinimum=0
    Weight=&RigVMFunction_MathDoubleMul(A=@"DeltaRotation v4_3".outputWeight, B=@"DeltaRotation v4_4".outputWeight).Result
   )
   > DeltaRotation v4_5 = RigVMFunctionReferenceNode(
    Rotation Order=ZYX
    child=(Type=Bone,Name="thigh_l")
    debugRotation=false
    outputMinMax=(X=0,Y=1)
    parent=(Type=Null,Name="thigh_l_null")
    premult=1
    rotationMinMax=(X=85,Y=120)
    triggerAxis=Z
    twistAxis=(X=1,Y=0,Z=0)
   )
   > DeltaRotation v4_6 = RigVMFunctionReferenceNode(
    Rotation Order=ZYX
    child=(Type=Bone,Name="thigh_r")
    outputMinMax=(X=0,Y=1)
    parent=(Type=Null,Name="thigh_r_null")
    premult=1
    rotationMinMax=(X=85,Y=120)
    triggerAxis=Z
    twistAxis=(X=1,Y=0,Z=0)
   )
   > &RigUnit_ModifyTransforms(
    ItemToModify=((Item=(Type=Bone,Name="dyn_skirt_back_m_01"),Transform=(Rotation=(X=-0.0035,Y=-0.0216,Z=0.3084,W=0.951),Translation=(X=0,Y=0,Z=0),Scale3D=(X=1,Y=1,Z=1))))
    WeightMinimum=0
    Weight=&RigVMFunction_MathDoubleMul(A=@"DeltaRotation v4_5".outputWeight, B=@"DeltaRotation v4_6".outputWeight).Result
   )
   > &RigUnit_ModifyTransforms(
    ItemToModify="((Item=(Type=Bone,Name=\"dyn_skirt_side_l_01\"),Transform=(Rotation=(X=0,Y=-0.7071,Z=0,W=-0.7071),Translation=(X=9,Y=-10,Z=13),Scale3D=(X=1,Y=1,Z=1))),(Item=(Type=Bone,Name=\"dyn_skirt_back_l_01\"),Transform=(Rotation=(X=0,Y=0.7071,Z=0,W=0.7071),Translation=(X ...(truncated)"
    WeightMinimum=0
    Weight=RigVMVariableNode() {variable=SPR L Leg Otw}.Value
   )
   > &RigUnit_ModifyTransforms(
    ItemToModify="((Item=(Type=Bone,Name=\"dyn_skirt_side_r_01\"),Transform=(Rotation=(X=0,Y=-0.7071,Z=0,W=0.7071),Translation=(X=9,Y=-10,Z=-13),Scale3D=(X=1,Y=1,Z=1))),(Item=(Type=Bone,Name=\"dyn_skirt_back_r_01\"),Transform=(Rotation=(X=0,Y=-0.7071,Z=0,W=0.7071),Translation=( ...(truncated)"
    WeightMinimum=0
    Weight=RigVMVariableNode() {variable=SPR R Leg Otw}.Value
   )
   > &RigUnit_ModifyTransforms(
    ItemToModify="((Item=(Type=Bone,Name=\"dyn_skirt_side_l_01\"),Transform=(Rotation=(X=0,Y=-0.5,Z=0,W=0.866),Translation=(X=0,Y=0,Z=0),Scale3D=(X=1,Y=1,Z=1))),(Item=(Type=Bone,Name=\"dyn_skirt_back_l_01\"),Transform=(Rotation=(X=0,Y=-0.5,Z=0,W=0.866),Translation=(X=0,Y=0,Z=0) ...(truncated)"
    WeightMinimum=0
    Weight=RigVMVariableNode() {variable=SPR L Leg Itw}.Value
   )
   > &RigUnit_ModifyTransforms(
    ItemToModify="((Item=(Type=Bone,Name=\"dyn_skirt_side_r_01\"),Transform=(Rotation=(X=0,Y=0.5,Z=0,W=0.866),Translation=(X=0,Y=0,Z=0),Scale3D=(X=1,Y=1,Z=1))),(Item=(Type=Bone,Name=\"dyn_skirt_back_r_01\"),Transform=(Rotation=(X=0,Y=0.5,Z=0,W=0.866),Translation=(X=0,Y=0,Z=0),S ...(truncated)"
    WeightMinimum=0
    Weight=RigVMVariableNode() {variable=SPR R Leg Itw}.Value
   )
   > DeltaRotation v4_8 = RigVMFunctionReferenceNode(
    Rotation Order=ZYX
    child=(Type=Bone,Name="calf_l")
    debugRotation=false
    onlyTwist=false
    outputMinMax=(X=0,Y=1)
    parent=(Type=Null,Name="calf_l_null")
    premult=1
    rotationMinMax=(X=0,Y=30)
    triggerAxis=Z
    twistAxis=(X=1,Y=0,Z=0)
   )
   > &RigUnit_ModifyTransforms(
    ItemToModify="((Item=(Type=Bone,Name=\"dyn_skirt_side_l_01\"),Transform=(Rotation=(X=-0.0012,Y=-0.0305,Z=-0.0392,W=0.9988),Translation=(X=0,Y=0,Z=0),Scale3D=(X=1,Y=1,Z=1))),(Item=(Type=Bone,Name=\"dyn_skirt_back_l_01\"),Transform=(Rotation=(X=0,Y=0,Z=-0.1219,W=0.9925),Trans ...(truncated)"
    WeightMinimum=0
    Weight=@"DeltaRotation v4_8".outputWeight
   )
   > DeltaRotation v4_9 = RigVMFunctionReferenceNode(
    Rotation Order=ZYX
    child=(Type=Bone,Name="calf_r")
    debugRotation=false
    onlyTwist=false
    outputMinMax=(X=0,Y=1)
    parent=(Type=Null,Name="calf_r_null")
    premult=1
    rotationMinMax=(X=0,Y=30)
    triggerAxis=Z
    twistAxis=(X=1,Y=0,Z=0)
   )
   > &RigUnit_ModifyTransforms(
    ItemToModify="((Item=(Type=Bone,Name=\"dyn_skirt_side_r_01\"),Transform=(Rotation=(X=0.0012,Y=0.0305,Z=-0.0392,W=0.9988),Translation=(X=0,Y=0,Z=0),Scale3D=(X=1,Y=1,Z=1))),(Item=(Type=Bone,Name=\"dyn_skirt_back_r_01\"),Transform=(Rotation=(X=0,Y=0,Z=-0.1219,W=0.9925),Transla ...(truncated)"
    WeightMinimum=0
    Weight=@"DeltaRotation v4_9".outputWeight
   )
   n2 = RigVMDispatchNode(False=1, True=0, Condition=RigVMVariableNode() {variable=Pub_IsCelebration Emote}.Value) {template="DISPATCH_RigVMDispatch_If(in Condition,in True,in False,out Result)"}
   n3 = RigVMVariableNode() {variable=SPR L calf Bkd}
   n4 = &RigVMFunction_AlphaInterp(
    InterpSpeedDecreasing=5
    InterpSpeedIncreasing=5
    Scale=1
    bClampResult=False
    Value=&RigVMFunction_MathDoubleMul(B=@n2.Result, A=@n3.Value).Result
   )
   > &RigUnit_ModifyTransforms(
    ItemToModify="((Item=(Type=Bone,Name=\"dyn_skirt_side_l_01\"),Transform=(Rotation=(X=0,Y=0,Z=-0.5736,W=0.8192),Translation=(X=0,Y=0,Z=0),Scale3D=(X=1,Y=1,Z=1))),(Item=(Type=Bone,Name=\"dyn_skirt_back_l_01\"),Transform=(Rotation=(X=0,Y=0,Z=0.4942,W=-0.8693),Translation=(X=0, ...(truncated)"
    WeightMinimum=0
    Weight=@n4.Result
   )
   n5 = RigVMVariableNode() {variable=SPR R calf Bkd}
   n6 = &RigVMFunction_AlphaInterp(
    InterpSpeedDecreasing=5
    InterpSpeedIncreasing=5
    Scale=1
    bClampResult=False
    Value=&RigVMFunction_MathDoubleMul(B=@n2.Result, A=@n5.Value).Result
   )
   > &RigUnit_ModifyTransforms(
    ItemToModify="((Item=(Type=Bone,Name=\"dyn_skirt_side_r_01\"),Transform=(Rotation=(X=0,Y=0,Z=-0.5736,W=0.8192),Translation=(X=0,Y=0,Z=0),Scale3D=(X=1,Y=1,Z=1))),(Item=(Type=Bone,Name=\"dyn_skirt_back_r_01\"),Transform=(Rotation=(X=-0.0039,Y=-0.0078,Z=0.4976,W=-0.8674),Trans ...(truncated)"
    WeightMinimum=0
    Weight=@n6.Result
   )
   AlphaInterp = &RigVMFunction_AlphaInterp(
    InterpSpeedDecreasing=5
    InterpSpeedIncreasing=5
    Scale=1
    bClampResult=False
    Value=&RigVMFunction_MathDoubleMul(A=@n3.Value, B=@n5.Value).Result
   )
   > &RigUnit_ModifyTransforms(
    ItemToModify=((Item=(Type=Bone,Name="dyn_skirt_back_m_01"),Transform=(Rotation=(X=0,Y=0,Z=0.2588,W=0.9659),Translation=(X=0,Y=0,Z=0),Scale3D=(X=1,Y=1,Z=1))))
    WeightMinimum=0
    Weight=@AlphaInterp.Result
   )
   > &RigUnit_ModifyTransforms(
    ItemToModify="((Item=(Type=Bone,Name=\"skirt_l_root\"),Transform=(Rotation=(X=0.9063,Y=0,Z=0,W=0.4226),Translation=(X=0,Y=0,Z=0),Scale3D=(X=1,Y=1,Z=1))),(Item=(Type=Bone,Name=\"skirt_l_front_root\"),Transform=(Rotation=(X=0.9063,Y=0,Z=0,W=0.4226),Translation=(X=0,Y=0,Z=0),S ...(truncated)"
    WeightMinimum=0
    Weight=RigVMVariableNode() {variable=SPR L Leg In Twist}.Value
   )
   > &RigUnit_ModifyTransforms(
    ItemToModify="((Item=(Type=Bone,Name=\"skirt_r_root\"),Transform=(Rotation=(X=-0.9063,Y=0,Z=0,W=0.4226),Translation=(X=0,Y=0,Z=0),Scale3D=(X=1,Y=1,Z=1))),(Item=(Type=Bone,Name=\"skirt_r_front_root\"),Transform=(Rotation=(X=-0.9063,Y=0,Z=0,W=0.4226),Translation=(X=0,Y=0,Z=0) ...(truncated)"
    WeightMinimum=0
    Weight=RigVMVariableNode() {variable=SPR R Leg In Twist}.Value
   )
   > &RigUnit_ModifyTransforms(
    ItemToModify="((Item=(Type=Bone,Name=\"skirt_l_root\"),Transform=(Rotation=(X=-0.9063,Y=0,Z=0,W=0.4226),Translation=(X=0,Y=0,Z=0),Scale3D=(X=1,Y=1,Z=1))),(Item=(Type=Bone,Name=\"skirt_l_front_root\"),Transform=(Rotation=(X=-0.9063,Y=0,Z=0,W=0.4226),Translation=(X=0,Y=0,Z=0) ...(truncated)"
    WeightMinimum=0
    Weight=RigVMVariableNode() {variable=SPR L Leg Out Twist}.Value
   )
   > &RigUnit_ModifyTransforms(
    ItemToModify="((Item=(Type=Bone,Name=\"skirt_r_root\"),Transform=(Rotation=(X=0.9063,Y=0,Z=0,W=0.4226),Translation=(X=0,Y=0,Z=0),Scale3D=(X=1,Y=1,Z=1))),(Item=(Type=Bone,Name=\"skirt_r_front_root\"),Transform=(Rotation=(X=0.9063,Y=0,Z=0,W=0.4226),Translation=(X=0,Y=0,Z=0),S ...(truncated)"
    WeightMinimum=0
    Weight=RigVMVariableNode() {variable=SPR R Leg Out Twist}.Value
   )
   > DeltaRotation v4 = RigVMFunctionReferenceNode(
    Rotation Order=ZYX
    child=(Type=Bone,Name="thigh_r")
    debugRotation=false
    outputMinMax=(X=0,Y=1)
    parent=(Type=Null,Name="thigh_r_null")
    premult=1
    rotationMinMax=(X=0,Y=-33)
    triggerAxis=Y
    twistAxis=(X=1,Y=0,Z=0)
   )
   > &RigUnit_ModifyTransforms(
    ItemToModify="((Item=(Type=Bone,Name=\"dyn_skirt_side_r_01\"),Transform=(Rotation=(X=-0.1013,Y=0.1086,Z=-0.1865,W=0.9711),Translation=(X=0,Y=0,Z=0),Scale3D=(X=1,Y=1,Z=1))),(Item=(Type=Bone,Name=\"dyn_skirt_back_r_01\"),Transform=(Rotation=(X=-0.044,Y=0.2769,Z=-0.1507,W=0.94 ...(truncated)"
    WeightMinimum=0
    Weight=@"DeltaRotation v4".outputWeight
   )
   n7 = &RigVMFunction_MathBoolToFloat(Value=RigVMVariableNode() {variable=Pub_IsSkydive}.Value)
   n8 = &RigVMFunction_AlphaInterp(
    InterpSpeedDecreasing=5
    InterpSpeedIncreasing=5
    Scale=1
    bClampResult=False
    Value=&RigVMFunction_MathDoubleMul(A=@n7.Result, B=RigVMVariableNode() {variable=Pub_strafeLeft float}.Value).Result
   )
   > &RigUnit_ModifyTransforms(
    ItemToModify=((Item=(Type=Bone,Name="dyn_skirt_side_l_01"),Transform=(Rotation=(X=0,Y=-0.1305,Z=0,W=0.9914),Translation=(X=0,Y=0,Z=0),Scale3D=(X=1,Y=1,Z=1))))
    WeightMinimum=0
    Weight=@n8.Result
   )
   n9 = &RigVMFunction_AlphaInterp(
    InterpSpeedDecreasing=5
    InterpSpeedIncreasing=5
    Scale=1
    bClampResult=False
    Value=&RigVMFunction_MathDoubleMul(A=@n7.Result, B=RigVMVariableNode() {variable=Pub_strafeRight float}.Value).Result
   )
   > &RigUnit_ModifyTransforms(
    ItemToModify="((Item=(Type=Bone,Name=\"dyn_skirt_side_l_01\"),Transform=(Rotation=(X=0,Y=-0.0872,Z=0,W=0.9962),Translation=(X=0,Y=0,Z=0),Scale3D=(X=1,Y=1,Z=1))),(Item=(Type=Bone,Name=\"dyn_skirt_side_r_01\"),Transform=(Rotation=(X=0,Y=0.0436,Z=0,W=0.999),Translation=(X=0,Y= ...(truncated)"
    WeightMinimum=0
    Weight=@n9.Result
   )
   n10 = &RigVMFunction_AlphaInterp(
    InterpSpeedDecreasing=5
    InterpSpeedIncreasing=5
    Scale=1
    bClampResult=False
    Value=&RigVMFunction_MathDoubleMul(B=@n7.Result, A=RigVMVariableNode() {variable=Pub_moveForward}.Value).Result
   )
   > &RigUnit_ModifyTransforms(
    ItemToModify="((Item=(Type=Bone,Name=\"dyn_skirt_side_l_01\"),Transform=(Rotation=(X=-0.0047,Y=-0.0817,Z=-0.0568,W=0.995),Translation=(X=0,Y=0,Z=0),Scale3D=(X=1,Y=1,Z=1))),(Item=(Type=Bone,Name=\"dyn_skirt_side_r_01\"),Transform=(Rotation=(X=0.0012,Y=-0.069,Z=0.0391,W=-0.99 ...(truncated)"
    WeightMinimum=0
    Weight=@n10.Result
   )
   n11 = RigVMVariableNode() {variable=Pub_IsOpenParachute}
   n12 = &RigVMFunction_AlphaInterp(
    InterpSpeedDecreasing=5
    InterpSpeedIncreasing=5
    Scale=1
    bClampResult=False
    Value=&RigVMFunction_MathDoubleMul(B=&RigVMFunction_MathBoolToFloat(Value=@n11.Value).Result, A=RigVMVariableNode() {variable=Pub_strafeLeft float}.Value).Result
   )
   > &RigUnit_ModifyTransforms(
    ItemToModify="((Item=(Type=Bone,Name=\"dyn_skirt_side_l_01\"),Transform=(Rotation=(X=0.0084,Y=-0.09,Z=-0.0947,W=0.9914),Translation=(X=0,Y=0,Z=0),Scale3D=(X=1,Y=1,Z=1))),(Item=(Type=Bone,Name=\"dyn_skirt_back_l_01\"),Transform=(Rotation=(X=0.0047,Y=-0.0733,Z=-0.0643,W=0.995 ...(truncated)"
    WeightMinimum=0
    Weight=@n12.Result
   )
   n13 = RigVMVariableNode() {variable=Pub_IsOpenParachute}
   n14 = &RigVMFunction_AlphaInterp(
    InterpSpeedDecreasing=5
    InterpSpeedIncreasing=5
    Scale=1
    bClampResult=False
    Value=&RigVMFunction_MathDoubleMul(B=&RigVMFunction_MathBoolToFloat(Value=@n13.Value).Result, A=RigVMVariableNode() {variable=Pub_strafeRight float}.Value).Result
   )
   > &RigUnit_ModifyTransforms(
    ItemToModify="((Item=(Type=Bone,Name=\"dyn_skirt_side_r_01\"),Transform=(Rotation=(X=0.0084,Y=0.09,Z=0.0947,W=0.9914),Translation=(X=0,Y=0,Z=0),Scale3D=(X=1,Y=1,Z=1))),(Item=(Type=Bone,Name=\"dyn_skirt_back_r_01\"),Transform=(Rotation=(X=0.0142,Y=0.0721,Z=-0.0657,W=0.9951), ...(truncated)"
    WeightMinimum=0
    Weight=@n14.Result
   )
   RigVMFunction_MathBoolToFloat = &RigVMFunction_MathBoolToFloat(Value=RigVMVariableNode() {variable=Pub_IsTacticalSprint}.Value)
   n15 = &RigVMFunction_AlphaInterp(InterpSpeedDecreasing=5, InterpSpeedIncreasing=5, Scale=1, bClampResult=False, Value=@RigVMFunction_MathBoolToFloat.Result)
   > &RigUnit_ModifyTransforms(
    ItemToModify="((Item=(Type=Bone,Name=\"dyn_skirt_side_l_01\"),Transform=(Rotation=(X=0,Y=0,Z=-0.0872,W=0.9962),Translation=(X=0,Y=0,Z=0),Scale3D=(X=1,Y=1,Z=1))),(Item=(Type=Bone,Name=\"dyn_skirt_back_l_01\"),Transform=(Rotation=(X=0,Y=0,Z=-0.0872,W=0.9962),Translation=(X=0, ...(truncated)"
    WeightMinimum=0
    Weight=@n15.Result
   )
   > &RigUnit_ModifyTransforms(
    ItemToModify="((Item=(Type=Bone,Name=\"dyn_skirt_side_r_01\"),Transform=(Rotation=(X=0,Y=0,Z=-0.0872,W=0.9962),Translation=(X=0,Y=0,Z=0),Scale3D=(X=1,Y=1,Z=1))),(Item=(Type=Bone,Name=\"dyn_skirt_back_r_01\"),Transform=(Rotation=(X=0,Y=0,Z=-0.0872,W=0.9962),Translation=(X=0, ...(truncated)"
    WeightMinimum=0
    Weight=@n15.Result
   )
   n16 = &RigVMFunction_MathBoolToFloat(Value=RigVMVariableNode() {variable=Pub_IsSkydive}.Value)
   n17 = &RigVMFunction_MathBoolToFloat(Value=RigVMVariableNode() {variable=Pub_IsEnableDynamics}.Value)
   Multiply_12 = RigVMAggregateNode(A=@n16.Result, B=@n17.Result, C=RigVMVariableNode() {variable=Pub_moveForward}.Value)
   n18 = &RigVMFunction_AlphaInterp(InterpSpeedDecreasing=15, InterpSpeedIncreasing=15, Scale=1, bClampResult=False, Value=@Multiply_12.Result)
   > &RigUnit_ModifyTransforms(
    ItemToModify="((Item=(Type=Bone,Name=\"dyn_skirt_side_l_01\"),Transform=(Rotation=(X=0,Y=0,Z=0,W=1),Translation=(X=0,Y=0,Z=0),Scale3D=(X=1,Y=1,Z=1))),(Item=(Type=Bone,Name=\"dyn_skirt_back_l_01\"),Transform=(Rotation=(X=0,Y=0,Z=-0.1653,W=0.9862),Translation=(X=0,Y=0,Z=0),Sc ...(truncated)"
    WeightMinimum=0
    Weight=@n18.Result
   )
   Multiply_11 = RigVMAggregateNode(
    C=RigVMDispatchNode(False=1, True=0, Condition=RigVMVariableNode() {variable=Pub_IsOpenParachute}.Value) {template="DISPATCH_RigVMDispatch_If(in Condition,in True,in False,out Result)"}.Result
    A=@n16.Result
    B=@n17.Result
   )
   n19 = &RigVMFunction_AlphaInterp(InterpSpeedDecreasing=15, InterpSpeedIncreasing=15, Scale=1, bClampResult=False, Value=@Multiply_11.Result)
   > &RigUnit_ModifyTransforms(
    ItemToModify="((Item=(Type=Bone,Name=\"dyn_skirt_side_l_01\"),Transform=(Rotation=(X=0,Y=0,Z=0.0872,W=0.9962),Translation=(X=0,Y=0,Z=0),Scale3D=(X=1,Y=1,Z=1))),(Item=(Type=Bone,Name=\"dyn_skirt_back_l_01\"),Transform=(Rotation=(X=0,Y=0,Z=0.1736,W=0.9848),Translation=(X=0,Y= ...(truncated)"
    WeightMinimum=0
    Weight=@n19.Result
   )
   TimeOffset = /Script/RigVM.RigVMFunction_TimeOffsetFloat(
    BufferSize=16
    SecondsAgo=0.5
    TimeRange=1
    Value=&RigVMFunction_MathBoolToFloat(Value=RigVMVariableNode() {variable=Pub_IsDBNO}.Value).Result
   ) {is=struct RigVMUnitNode, template="TimeOffset::Execute(in Value,in SecondsAgo,in BufferSize,in TimeRange,out Result)"}
   > &RigUnit_ModifyTransforms(
    ItemToModify="((Item=(Type=Bone,Name=\"dyn_skirt_side_l_01\"),Transform=(Rotation=(X=0,Y=0,Z=0.4226,W=0.9063),Translation=(X=0,Y=0,Z=0),Scale3D=(X=1,Y=1,Z=1))),(Item=(Type=Bone,Name=\"dyn_skirt_back_l_01\"),Transform=(Rotation=(X=0,Y=0,Z=0.4226,W=0.9063),Translation=(X=0,Y= ...(truncated)"
    WeightMinimum=0
    Weight=@TimeOffset.Result
   )
   > &RigUnit_ModifyTransforms(
    ItemToModify="((Item=(Type=Bone,Name=\"dyn_skirt_side_l_01\"),Transform=(Rotation=(X=0,Y=0,Z=-0.1736,W=0.9848),Translation=(X=0,Y=0,Z=0),Scale3D=(X=1,Y=1,Z=1))),(Item=(Type=Bone,Name=\"dyn_skirt_back_l_01\"),Transform=(Rotation=(X=0,Y=0,Z=-0.1736,W=0.9848),Translation=(X=0, ...(truncated)"
    WeightMinimum=0
    Weight=&RigVMFunction_MathBoolToFloat(Value=RigVMVariableNode() {variable=Pub_IsDBNO_Carried}.Value).Result
   )
   > &RigUnit_ModifyTransforms(
    ItemToModify=((Item=(Type=Bone,Name="dyn_skirt_side_l_01"),Transform=(Rotation=(X=0,Y=0,Z=-0.0872,W=0.9962),Translation=(X=0,Y=0,Z=0),Scale3D=(X=1,Y=1,Z=1))))
    Weight=1
    WeightMinimum=0
   )
   > DeltaRotation v4_17 = RigVMFunctionReferenceNode(
    Rotation Order=ZYX
    child=(Type=Bone,Name="foot_l")
    debugRotation=false
    onlyTwist=false
    outputMinMax=(X=0,Y=1)
    parent=(Type=Null,Name="foot_l_null")
    premult=1
    rotationMinMax=(X=0,Y=4)
    triggerAxis=Y
    twistAxis=(X=1,Y=0,Z=0)
   )
   n20 = &RigVMFunction_MathBoolToFloat(Value=RigVMVariableNode() {variable=Pub_IsFrontEnd}.Value)
   n21 = RigVMVariableNode() {variable=Pub_IsIdleMontage}
   Multiply = RigVMAggregateNode(A=@"DeltaRotation v4_17".outputWeight, B=@n20.Result, C=@n21.Value)
   > &RigUnit_ModifyTransforms(
    ItemToModify="((Item=(Type=Bone,Name=\"dyn_skirt_front_l_01\"),Transform=(Rotation=(X=0,Y=0.0872,Z=0,W=0.9962),Translation=(X=0,Y=0,Z=0),Scale3D=(X=1,Y=1,Z=1))),(Item=(Type=Bone,Name=\"dyn_skirt_front_m_01\"),Transform=(Rotation=(X=0,Y=0.0872,Z=0,W=0.9962),Translation=(X=0, ...(truncated)"
    WeightMinimum=0
    Weight=@Multiply.Result
   )
   > DeltaRotation v4_18 = RigVMFunctionReferenceNode(
    Rotation Order=ZYX
    child=(Type=Bone,Name="foot_r")
    debugRotation=false
    onlyTwist=false
    outputMinMax=(X=0,Y=1)
    parent=(Type=Null,Name="foot_r_null")
    premult=1
    rotationMinMax=(X=0,Y=-5)
    triggerAxis=Y
    twistAxis=(X=1,Y=0,Z=0)
   )
   Multiply_10 = RigVMAggregateNode(A=@"DeltaRotation v4_18".outputWeight, B=@n20.Result, C=@n21.Value)
   > &RigUnit_ModifyTransforms(
    ItemToModify="((Item=(Type=Bone,Name=\"dyn_skirt_front_r_01\"),Transform=(Rotation=(X=0,Y=0.0567,Z=0,W=0.9984),Translation=(X=0,Y=0,Z=0),Scale3D=(X=1,Y=1,Z=1))),(Item=(Type=Bone,Name=\"dyn_skirt_front_r_02\"),Transform=(Rotation=(X=0,Y=0.0436,Z=0,W=0.999),Translation=(X=0,Y ...(truncated)"
    WeightMinimum=0
    Weight=@Multiply_10.Result
   )
   > DeltaRotation v4_19 = RigVMFunctionReferenceNode(
    Rotation Order=ZYX
    child=(Type=Bone,Name="foot_r")
    debugRotation=false
    onlyTwist=false
    outputMinMax=(X=0,Y=1)
    parent=(Type=Null,Name="foot_r_null")
    premult=1
    rotationMinMax=(X=0,Y=2)
    triggerAxis=Y
    twistAxis=(X=1,Y=0,Z=0)
   )
   Multiply_19 = RigVMAggregateNode(A=@"DeltaRotation v4_19".outputWeight, B=@"DeltaRotation v4_19".outputWeight, C=@n21.Value)
   > &RigUnit_ModifyTransforms(
    ItemToModify="((Item=(Type=Bone,Name=\"dyn_skirt_front_r_01\"),Transform=(Rotation=(X=0,Y=0.0567,Z=0,W=0.9984),Translation=(X=0,Y=0,Z=0),Scale3D=(X=1,Y=1,Z=1))),(Item=(Type=Bone,Name=\"dyn_skirt_front_m_01\"),Transform=(Rotation=(X=0,Y=0.0872,Z=0,W=0.9962),Translation=(X=0, ...(truncated)"
    WeightMinimum=0
    Weight=@Multiply_19.Result
   )
   n22 = RigVMVariableNode() {variable=Pub_IsFrontEnd}
   n23 = RigVMVariableNode() {variable=Pub_IsSkydive}
   > &RigUnit_ModifyTransforms(
    ItemToModify=((Item=(Type=Bone,Name="dyn_skirt_back_m_01"),Transform=(Rotation=(X=0,Y=0,Z=-0.1736,W=0.9848),Translation=(X=0,Y=0,Z=0),Scale3D=(X=1,Y=1,Z=1))))
    WeightMinimum=0
    Weight=&RigVMFunction_MathDoubleMul(
     B=&RigVMFunction_MathBoolToFloat(Value=@n22.Value).Result
     A=&RigVMFunction_MathBoolToFloat(Value=@n23.Value).Result
    ).Result
   )
   RigVMFunctionReturnNode(ExecuteContext=@RigVMFunction_Sequence.B)
   &RigUnit_ModifyTransforms(
    ExecutePin=()
    ItemToModify="((Item=(Type=Bone,Name=\"dyn_skirt_side_l_01\"),Transform=(Rotation=(X=0,Y=0,Z=0,W=1),Translation=(X=0,Y=0,Z=0),Scale3D=(X=1,Y=1,Z=1))),(Item=(Type=Bone,Name=\"dyn_skirt_back_l_01\"),Transform=(Rotation=(X=0,Y=0,Z=-0.0872,W=0.9962),Translation=(X=0,Y=0,Z=0),Sc ...(truncated)"
    WeightMinimum=0
    Weight=RigVMVariableNode() {variable=SPR L calf ExtremeBkd}.Value
   )
   > &RigUnit_ModifyTransforms(
    ItemToModify="((Item=(Type=Bone,Name=\"dyn_skirt_side_r_01\"),Transform=(Rotation=(X=0,Y=0,Z=0,W=1),Translation=(X=0,Y=0,Z=0),Scale3D=(X=1,Y=1,Z=1))),(Item=(Type=Bone,Name=\"dyn_skirt_back_r_01\"),Transform=(Rotation=(X=0,Y=0,Z=-0.0872,W=0.9962),Translation=(X=0,Y=0,Z=0),Sc ...(truncated)"
    WeightMinimum=0
    Weight=RigVMVariableNode() {variable=SPR R calf ExtremeBkd}.Value
   )
  ]}
 New Function_ContainedGraph {nodes [
   n1 = RigVMFunctionEntryNode()
   RigVMFunction_Sequence = &RigVMFunction_Sequence(ExecuteContext=@n1.ExecuteContext)
   RigVMFunction_MathBoolToFloat = &RigVMFunction_MathBoolToFloat(Value=RigVMVariableNode() {variable=Pub_IsCrouching}.Value)
   n2 = &RigVMFunction_AlphaInterp(InterpSpeedDecreasing=5, InterpSpeedIncreasing=5, Scale=1, bClampResult=False, Value=@RigVMFunction_MathBoolToFloat.Result)
   &RigUnit_ModifyTransforms(
    ItemToModify="((Item=(Type=Bone,Name=\"dyn_skirt_side_l_01\"),Transform=(Rotation=(X=-0.2729,Y=-0.1103,Z=0.5318,W=0.7941),Translation=(X=-4.501,Y=1.6383,Z=-1.3828),Scale3D=(X=1,Y=1,Z=1))),(Item=(Type=Bone,Name=\"dyn_skirt_back_l_01\"),Transform=(Rotation=(X=-0.2028,Y=-0.129 ...(truncated)"
    WeightMinimum=0
    Weight=@n2.Result
    ExecutePin=@RigVMFunction_Sequence.A
   )
   > &RigUnit_ModifyTransforms(
    ItemToModify="((Item=(Type=Bone,Name=\"dyn_skirt_side_l_02\"),Transform=(Rotation=(X=0,Y=0,Z=-0.2164,W=0.9763),Translation=(X=0,Y=0,Z=0),Scale3D=(X=1,Y=1,Z=1))),(Item=(Type=Bone,Name=\"dyn_skirt_back_l_02\"),Transform=(Rotation=(X=0,Y=0,Z=-0.2164,W=0.9763),Translation=(X=0, ...(truncated)"
    WeightMinimum=0
    Weight=@n2.Result
   )
   > &RigUnit_ModifyTransforms(
    ItemToModify="((Item=(Type=Bone,Name=\"dyn_skirt_side_l_03\"),Transform=(Rotation=(X=0,Y=0,Z=-0.2164,W=0.9763),Translation=(X=0,Y=0,Z=0),Scale3D=(X=1,Y=1,Z=1))),(Item=(Type=Bone,Name=\"dyn_skirt_back_l_03\"),Transform=(Rotation=(X=0,Y=0,Z=-0.2164,W=0.9763),Translation=(X=0, ...(truncated)"
    WeightMinimum=0
    Weight=@n2.Result
   )
   > &RigUnit_ModifyTransforms(
    ItemToModify="((Item=(Type=Bone,Name=\"dyn_skirt_side_l_04\"),Transform=(Rotation=(X=0,Y=0,Z=-0.2164,W=0.9763),Translation=(X=0,Y=0,Z=0),Scale3D=(X=1,Y=1,Z=1))),(Item=(Type=Bone,Name=\"dyn_skirt_back_l_04\"),Transform=(Rotation=(X=0,Y=0,Z=-0.2164,W=0.9763),Translation=(X=0, ...(truncated)"
    WeightMinimum=0
    Weight=@n2.Result
   )
   > &RigUnit_ModifyTransforms(
    ItemToModify="((Item=(Type=Bone,Name=\"dyn_skirt_side_l_05\"),Transform=(Rotation=(X=0.5521,Y=-0.105,Z=0.1187,W=0.8186),Translation=(X=0,Y=0,Z=0),Scale3D=(X=1,Y=1,Z=1))),(Item=(Type=Bone,Name=\"dyn_skirt_side_r_05\"),Transform=(Rotation=(X=-0.4361,Y=0.108,Z=0.0236,W=0.8931) ...(truncated)"
    WeightMinimum=0
    Weight=@n2.Result
   )
   > &RigUnit_ModifyTransforms(
    ItemToModify="((Item=(Type=Bone,Name=\"dyn_skirt_side_l_06\"),Transform=(Rotation=(X=0.0132,Y=0.0594,Z=0.216,W=-0.9745),Translation=(X=0,Y=0,Z=0),Scale3D=(X=1,Y=1,Z=1))),(Item=(Type=Bone,Name=\"dyn_skirt_side_r_06\"),Transform=(Rotation=(X=-0.0189,Y=-0.0853,Z=0.2156,W=-0.97 ...(truncated)"
    WeightMinimum=0
    Weight=@n2.Result
   )
   n3 = RigVMVariableNode() {variable=Pub_IsTargeting}
   n4 = &RigVMFunction_AlphaInterp(
    InterpSpeedDecreasing=5
    InterpSpeedIncreasing=5
    Scale=1
    bClampResult=False
    Value=&RigVMFunction_MathDoubleMul(A=@RigVMFunction_MathBoolToFloat.Result, B=&RigVMFunction_MathBoolToFloat(Value=@n3.Value).Result).Result
   )
   > &RigUnit_ModifyTransforms(
    ItemToModify="((Item=(Type=Bone,Name=\"dyn_skirt_front_m_01\"),Transform=(Rotation=(X=0,Y=-0.2588,Z=0,W=0.9659),Translation=(X=0,Y=0,Z=0),Scale3D=(X=1,Y=1,Z=1))),(Item=(Type=Bone,Name=\"dyn_skirt_front_r_02\"),Transform=(Rotation=(X=0,Y=-0.1736,Z=0,W=0.9848),Translation=(X= ...(truncated)"
    WeightMinimum=0
    Weight=@n4.Result
   )
   AlphaInterp = &RigVMFunction_AlphaInterp(
    InterpSpeedDecreasing=5
    InterpSpeedIncreasing=5
    Scale=1
    bClampResult=False
    Value=&RigVMFunction_MathDoubleMul(A=@RigVMFunction_MathBoolToFloat.Result, B=RigVMVariableNode() {variable=PelvisProx}.Value).Result
   )
   > &RigUnit_ModifyTransforms(
    ItemToModify="((Item=(Type=Bone,Name=\"dyn_skirt_side_l_03\"),Transform=(Rotation=(X=0.0044,Y=0.1283,Z=-0.0493,W=0.9905),Translation=(X=0,Y=0,Z=0),Scale3D=(X=1,Y=1,Z=1))),(Item=(Type=Bone,Name=\"dyn_skirt_back_l_03\"),Transform=(Rotation=(X=-0.0011,Y=0.0364,Z=-0.0291,W=0.99 ...(truncated)"
    WeightMinimum=0
    Weight=@AlphaInterp.Result
   )
   n5 = &RigVMFunction_AlphaInterp(
    InterpSpeedDecreasing=5
    InterpSpeedIncreasing=5
    Scale=1
    bClampResult=False
    Value=&RigVMFunction_MathBoolToFloat(Value=RigVMVariableNode() {variable=Pub_IsKneeSliding}.Value).Result
   )
   > &RigUnit_ModifyTransforms(
    ItemToModify="((Item=(Type=Bone,Name=\"dyn_skirt_side_l_01\"),Transform=(Rotation=(X=0.0454,Y=-0.0114,Z=0.5012,W=-0.8641),Translation=(X=0,Y=0,Z=0),Scale3D=(X=1,Y=1,Z=1))),(Item=(Type=Bone,Name=\"dyn_skirt_back_l_01\"),Transform=(Rotation=(X=0.0333,Y=0.0168,Z=-0.5823,W=0.81 ...(truncated)"
    WeightMinimum=0
    Weight=@n5.Result
   )
   n6 = &RigVMFunction_AlphaInterp(
    InterpSpeedDecreasing=5
    InterpSpeedIncreasing=5
    Scale=1
    bClampResult=False
    Value=&RigVMFunction_MathBoolToFloat(Value=RigVMVariableNode() {variable=Pub_IsbaseballSliding}.Value).Result
   )
   > &RigUnit_ModifyTransforms(
    ItemToModify="((Item=(Type=Bone,Name=\"dyn_skirt_side_l_01\"),Transform=(Rotation=(X=-0.0397,Y=0.4148,Z=0.7227,W=-0.5515),Translation=(X=0,Y=0,Z=0),Scale3D=(X=1,Y=1,Z=1))),(Item=(Type=Bone,Name=\"dyn_skirt_back_l_01\"),Transform=(Rotation=(X=-0.0046,Y=-0.4431,Z=-0.7842,W=0. ...(truncated)"
    WeightMinimum=0
    Weight=@n6.Result
   )
   > &RigUnit_ModifyTransforms(
    ItemToModify="((Item=(Type=Bone,Name=\"dyn_skirt_side_l_01\"),Transform=(Rotation=(X=-0.0371,Y=0.1733,Z=-0.2733,W=-0.9455),Translation=(X=-1.9896,Y=5.258,Z=3.4761),Scale3D=(X=1,Y=1,Z=1))),(Item=(Type=Bone,Name=\"dyn_skirt_back_l_01\"),Transform=(Rotation=(X=0.0303,Y=-0.1037 ...(truncated)"
    WeightMinimum=0
    Weight=&RigVMFunction_MathBoolToFloat(Value=RigVMVariableNode() {variable=Pub_IsGrindRail}.Value).Result
   )
   > DeltaRotation v4 = RigVMFunctionReferenceNode(
    Rotation Order=ZYX
    child=(Type=Bone,Name="thigh_r")
    debugRotation=false
    outputMinMax=(X=0,Y=1)
    parent=(Type=Null,Name="thigh_r_null")
    premult=1
    rotationMinMax=(X=0,Y=-40)
    triggerAxis=Y
    twistAxis=(X=1,Y=0,Z=0)
   )
   > DeltaRotation v4_7 = RigVMFunctionReferenceNode(
    Rotation Order=ZYX
    child=(Type=Bone,Name="foot_r")
    debugRotation=false
    outputMinMax=(X=0,Y=1)
    parent=(Type=Null,Name="foot_r_null")
    premult=1
    rotationMinMax=(X=0,Y=-11)
    triggerAxis=Y
    twistAxis=(X=1,Y=0,Z=0)
   )
   n7 = &RigVMFunction_MathBoolToFloat(Value=RigVMVariableNode() {variable=Pub_IsBalletSpin Emote}.Value)
   n8 = &RigVMFunction_AlphaInterp(
    InterpSpeedDecreasing=5
    InterpSpeedIncreasing=5
    Scale=1
    bClampResult=False
    Value=&RigVMFunction_MathDoubleMul(B=@"DeltaRotation v4_7".outputWeight, A=@n7.Result).Result
   )
   > &RigUnit_ModifyTransforms(
    ItemToModify="((Item=(Type=Bone,Name=\"dyn_skirt_front_r_01\"),Transform=(Rotation=(X=-0.4404,Y=0.1215,Z=0.2437,W=0.8555),Translation=(X=5,Y=0,Z=0),Scale3D=(X=1,Y=1,Z=1))),(Item=(Type=Bone,Name=\"dyn_skirt_front_r_02\"),Transform=(Rotation=(X=0,Y=0.0872,Z=0,W=0.9962),Transl ...(truncated)"
    WeightMinimum=0
    Weight=@n8.Result
   )
   n9 = &RigVMFunction_AlphaInterp(
    InterpSpeedDecreasing=5
    InterpSpeedIncreasing=5
    Scale=1
    bClampResult=False
    Value=&RigVMFunction_MathDoubleMul(A=@"DeltaRotation v4".outputWeight, B=@n7.Result).Result
   )
   > &RigUnit_ModifyTransforms(
    ItemToModify="((Item=(Type=Bone,Name=\"dyn_skirt_side_r_01\"),Transform=(Rotation=(X=0.0481,Y=0.074,Z=0.5429,W=0.8352),Translation=(X=0,Y=0,Z=0),Scale3D=(X=1,Y=1,Z=1))),(Item=(Type=Bone,Name=\"dyn_skirt_back_r_01\"),Transform=(Rotation=(X=0.0023,Y=0.231,Z=-0.399,W=-0.8874), ...(truncated)"
    WeightMinimum=0
    Weight=@n9.Result
   )
   > DeltaRotation v4_2 = RigVMFunctionReferenceNode(
    Rotation Order=ZYX
    child=(Type=Bone,Name="thigh_l")
    debugRotation=false
    onlyTwist=true
    outputMinMax=(X=0,Y=1)
    parent=(Type=Null,Name="thigh_l_null")
    premult=1
    rotationMinMax=(X=0,Y=-30)
    triggerAxis=X
    twistAxis=(X=1,Y=0,Z=0)
   )
   > DeltaRotation v4_4 = RigVMFunctionReferenceNode(
    Rotation Order=ZYX
    child=(Type=Bone,Name="thigh_l")
    debugRotation=false
    onlyTwist=false
    outputMinMax=(X=0,Y=1)
    parent=(Type=Null,Name="thigh_l_null")
    premult=1
    rotationMinMax=(X=0,Y=-70)
    triggerAxis=Y
    twistAxis=(X=1,Y=0,Z=0)
   )
   > DeltaRotation v4_5 = RigVMFunctionReferenceNode(
    Rotation Order=ZYX
    child=(Type=Bone,Name="thigh_l")
    debugRotation=false
    onlyTwist=false
    outputMinMax=(X=0,Y=1)
    parent=(Type=Null,Name="thigh_l_null")
    premult=1
    rotationMinMax=(X=0,Y=-127)
    triggerAxis=Z
    twistAxis=(X=1,Y=0,Z=0)
   )
   Multiply_1 = RigVMAggregateNode(
    A=@"DeltaRotation v4_2".outputWeight
    B=@"DeltaRotation v4_4".outputWeight
    C=@"DeltaRotation v4_5".outputWeight
    D=&RigVMFunction_MathBoolToFloat(Value=RigVMVariableNode() {variable=Pub_EasternBlock Emote}.Value).Result
   )
   > &RigUnit_ModifyTransforms(
    ItemToModify="((Item=(Type=Bone,Name=\"dyn_skirt_side_l_01\"),Transform=(Rotation=(X=0.0484,Y=-0.1547,Z=0.2948,W=0.9417),Translation=(X=-1.2303,Y=19.1676,Z=-3.9668),Scale3D=(X=1,Y=1,Z=1))),(Item=(Type=Bone,Name=\"dyn_skirt_back_l_01\"),Transform=(Rotation=(X=-0.1218,Y=-0.32 ...(truncated)"
    WeightMinimum=0
    Weight=@Multiply_1.Result
   )
   n10 = &RigVMFunction_AlphaInterp(
    InterpSpeedDecreasing=5
    InterpSpeedIncreasing=5
    Scale=1
    bClampResult=False
    Value=&RigVMFunction_MathBoolToFloat(Value=RigVMVariableNode() {variable=Pub_IsJumping}.Value).Result
   )
   > &RigUnit_ModifyTransforms(
    ItemToModify="((Item=(Type=Bone,Name=\"dyn_skirt_side_l_01\"),Transform=(Rotation=(X=-0.0073,Y=-0.1671,Z=-0.043,W=0.985),Translation=(X=0,Y=0,Z=0),Scale3D=(X=1,Y=1,Z=1))),(Item=(Type=Bone,Name=\"dyn_skirt_back_l_01\"),Transform=(Rotation=(X=-0.0073,Y=-0.1673,Z=-0.043,W=0.98 ...(truncated)"
    WeightMinimum=0
    Weight=@n10.Result
   )
   > DeltaRotation v4_1 = RigVMFunctionReferenceNode(
    Rotation Order=ZYX
    child=(Type=Bone,Name="upperarm_r")
    debugRotation=false
    outputMinMax=(X=0,Y=1)
    parent=(Type=Null,Name="upperarm_r_null")
    premult=1
    rotationMinMax=(X=0,Y=-20)
    triggerAxis=Y
    twistAxis=(X=1,Y=0,Z=0)
   )
   n11 = &RigVMFunction_AlphaInterp(InterpSpeedDecreasing=5, InterpSpeedIncreasing=5, Scale=1, bClampResult=False, Value=@"DeltaRotation v4_1".outputWeight)
   > &RigUnit_ModifyTransforms(
    ItemToModify=((Item=(Type=Bone,Name="dyn_rose_r"),Transform=(Rotation=(X=0,Y=0,Z=0,W=1),Translation=(X=0,Y=0,Z=0),Scale3D=(X=1,Y=1,Z=1))))
    WeightMinimum=0
    Weight=@n11.Result
   )
   > DeltaRotation v4_3 = RigVMFunctionReferenceNode(
    Rotation Order=ZYX
    child=(Type=Bone,Name="upperarm_r")
    debugRotation=false
    outputMinMax=(X=0,Y=1)
    parent=(Type=Null,Name="upperarm_r_null")
    premult=1
    rotationMinMax=(X=-20,Y=-40)
    triggerAxis=Y
    twistAxis=(X=1,Y=0,Z=0)
   )
   n12 = &RigVMFunction_AlphaInterp(InterpSpeedDecreasing=5, InterpSpeedIncreasing=5, Scale=1, bClampResult=False, Value=@"DeltaRotation v4_3".outputWeight)
   > &RigUnit_ModifyTransforms(
    ItemToModify=((Item=(Type=Bone,Name="dyn_rose_r"),Transform=(Rotation=(X=-0.2938,Y=-0.1682,Z=0.0186,W=0.9408),Translation=(X=-0.0869,Y=0.4292,Z=0.0438),Scale3D=(X=1,Y=1,Z=1))))
    WeightMinimum=0
    Weight=@n12.Result
   )
   > DeltaRotation v4_6 = RigVMFunctionReferenceNode(
    Rotation Order=ZYX
    child=(Type=Bone,Name="head")
    debugRotation=false
    outputMinMax=(X=0,Y=1)
    parent=(Type=Null,Name="head_null")
    premult=1
    rotationMinMax=(X=0,Y=-28)
    triggerAxis=Z
    twistAxis=(X=1,Y=0,Z=0)
   )
   n13 = RigVMVariableNode() {variable=Pub_IsPistolEquipped}
   > &RigUnit_ModifyTransforms(
    ItemToModify=((Item=(Type=Bone,Name="dyn_rose_r"),Transform=(Rotation=(X=0.12,Y=-0.0057,Z=-0.6862,W=0.7175),Translation=(X=0,Y=0,Z=0),Scale3D=(X=1,Y=1,Z=1))))
    WeightMinimum=0
    Weight=&RigVMFunction_MathDoubleMul(A=@"DeltaRotation v4_6".outputWeight, B=&RigVMFunction_MathBoolToFloat(Value=@n13.Value).Result).Result
   )
   n14 = &RigVMFunction_AlphaInterp(
    InterpSpeedDecreasing=5
    InterpSpeedIncreasing=5
    Scale=1
    bClampResult=False
    Value=&RigVMFunction_MathBoolToFloat(Value=RigVMVariableNode() {variable=Pub_IsTargeting}.Value).Result
   )
   > &RigUnit_ModifyTransforms(
    ItemToModify=((Item=(Type=Bone,Name="dyn_rose_r"),Transform=(Rotation=(X=-0.1916,Y=-0.0261,Z=-0.0236,W=0.9808),Translation=(X=0,Y=0,Z=0),Scale3D=(X=1,Y=1,Z=1))))
    WeightMinimum=0
    Weight=@n14.Result
   )
   n15 = &RigVMFunction_AlphaInterp(
    InterpSpeedDecreasing=5
    InterpSpeedIncreasing=5
    Scale=1
    bClampResult=False
    Value=&RigVMFunction_MathBoolToFloat(Value=RigVMVariableNode() {variable=Pub_IsCelebration Emote}.Value).Result
   )
   > &RigUnit_ModifyTransforms(
    ItemToModify=((Item=(Type=Bone,Name="dyn_rose_r"),Transform=(Rotation=(X=0,Y=0.089,Z=0,W=0.996),Translation=(X=0,Y=0,Z=0),Scale3D=(X=1,Y=1,Z=1))))
    WeightMinimum=0
    Weight=@n15.Result
   )
   RigVMFunctionReturnNode(ExecuteContext=@RigVMFunction_Sequence.B)
   Multiply = &RigVMFunction_MathDoubleMul(
    A=&RigVMFunction_MathBoolToFloat(Value=RigVMVariableNode() {variable=Pub_IswearingCape}.Value).Result
    B=&RigVMFunction_MathBoolToFloat(Value=RigVMVariableNode() {variable=Pub_IsCrouching}.Value).Result
   )
   n16 = &RigVMFunction_AlphaInterp(InterpSpeedDecreasing=5, InterpSpeedIncreasing=5, Scale=1, bClampResult=False, Value=@Multiply.Result)
   &RigUnit_ModifyTransforms(
    ExecutePin=()
    ItemToModify="((Item=(Type=Bone,Name=\"dyn_skirt_side_l_06\"),Transform=(Rotation=(X=0,Y=0,Z=0.2588,W=0.9659),Translation=(X=0,Y=0,Z=0),Scale3D=(X=1,Y=1,Z=1))),(Item=(Type=Bone,Name=\"dyn_skirt_back_l_03\"),Transform=(Rotation=(X=0,Y=0,Z=0,W=1),Translation=(X=0,Y=0,Z=0),Sca ...(truncated)"
    WeightMinimum=0
    Weight=@n16.Result
   )
  ]}
 RigVMFunctionLibrary {nodes [
   Constraints = RigVMCollapseNode(ExecuteContext=())
   DeltaRotation v4 = RigVMCollapseNode(
    ExecuteContext=()
    Rotation Order=ZYX
    child=(Type=Bone,Name="None")
    outputMinMax=(X=0,Y=1)
    parent=(Type=Bone,Name="None")
    premult=1
    rotationMinMax=(X=0,Y=0)
    twistAxis=(X=1,Y=0,Z=0)
   )
   IsCrouching_Fix = RigVMCollapseNode(ExecuteContext=())
   Skirt_Fix = RigVMCollapseNode(ExecuteContext=())
   Variables = RigVMCollapseNode(ExecuteContext=())
  ]}
 RigVMModel {nodes [
   RigUnit_BeginExecution = &RigUnit_BeginExecution()
   FN_DeformRig_FNC = RigVMFunctionReferenceNode(
    DeformRig_ Curve="(deformRig_l_upperarm_up_alpha_72_#1=true,deformRig_r_upperarm_up_alpha_73_#2=true,deformRig_l_upperarm_fr_alpha_74_#3=true,deformRig_r_upperarm_fr_alpha_75_CB51B80C4 ...(truncated)"
    DeformRig_Bool="(upperarm_1_#4=True,elbow_3_#5=True,wrist_7_#6=True,groin_8_#7=True,glute_17_#8=True,knee_9_40C408324CA261 ...(truncated)"
    DeformRig_Debug=(upperarm_debug_18_#4=False,wrist_debug_20_#6=False,lat_debug_26_#9=False,groinGlute_debug_27_#7=False)
    DeformRig_Value="(upperarm_up_18_#10=(Rotation=(X=0.0,Y=-0.0,Z=0.0,W=1.0),Translation=(X=0.0,Y=0.0,Z=8.0),Scale3D=(X=0.70,Y=1.0,Z=1.0)),upperarm_fwd_fr_19_#11=(Rotation=(X=0.0,Y=-0.0,Z=0.0,W=1.0),Translation=(X=0.0, ...(truncated)"
    ExecuteContext=@RigUnit_BeginExecution.ExecutePin
   )
   > Constraints = RigVMFunctionReferenceNode()
   > Variables = RigVMFunctionReferenceNode()
   > Skirt_Fix = RigVMFunctionReferenceNode()
   > IsCrouching_Fix = RigVMFunctionReferenceNode()
  ]}
]
hierarchy [
 root {loc=<0 0 0>, rot=<0 0 0 1>}
 pelvis {parent=root, loc=<0 -0.5301 98.0571>, rot=<0.0127 0.707 -0.0127 -0.707>}
 spine_01 {parent=pelvis, loc=<6.1208 -0.9328 0>, rot=<0 0 0.034 -0.9994>}
 spine_02 {parent=spine_01, loc=<6.5997 -0.2095 0>, rot=<0 0 0.0238 -0.9997>}
 spine_03 {parent=spine_02, loc=<6.4914 -0.1237 0>, rot=<0 0 0.0149 -0.9999>}
 spine_04 {parent=spine_03, loc=<6.9598 0.0794 0>, rot=<0 0 -0.0132 -0.9999>}
 spine_05 {parent=spine_04, loc=<6.9987 0 0>, rot=<0 0 -0.0015 -1>}
 neck_01 {parent=spine_05, loc=<11.3138 0.1529 0>, rot=<0 0 -0.2003 -0.9797>}
 neck_02 {parent=neck_01, loc=<4.5627 0 0>, rot=<0 0 0.0504 -0.9987>}
 head {parent=neck_02, loc=<5.3914 0 0>, rot=<0 0 0.115 -0.9934>}
 base_hair_1_c {parent=head, loc=<1.2116 -5.0574 0>, rot=<0.1334 -0.6944 0.6944 -0.1334>}
 dyn_hair_1_c {parent=base_hair_1_c, loc=<0 0 0>, rot=<0 0 0 1>}
 dyn_hair_2_c {parent=dyn_hair_1_c, loc=<7.3205 0 0>, rot=<0 0 0 1>}
 dyn_hair_3_c {parent=dyn_hair_2_c, loc=<7.3197 0 0>, rot=<0 0 0 1>}
 dyn_hair_4_c {parent=dyn_hair_3_c, loc=<7.2884 0 0>, rot=<0 0 0 1>}
 dyn_hair_5_c {parent=dyn_hair_4_c, loc=<7.2939 0 0>, rot=<0 0 0 1>}
 dyn_hair_6_c {parent=dyn_hair_5_c, loc=<7.3213 0 0>, rot=<0 0 0 1>}
 dyn_hair_7_c {parent=dyn_hair_6_c, loc=<7.3211 0 0>, rot=<0 0 0 1>}
 base_hair_1_l {parent=head, loc=<1.2116 -5.0574 -5.6>, rot=<0.1334 -0.6944 0.6944 -0.1334>}
 dyn_hair_1_l {parent=base_hair_1_l, loc=<0 0 0>, rot=<0 0 0 1>}
 dyn_hair_2_l {parent=dyn_hair_1_l, loc=<7.3205 0 0>, rot=<0 0 0 1>}
 dyn_hair_3_l {parent=dyn_hair_2_l, loc=<7.3197 0 0>, rot=<0 0 0 1>}
 dyn_hair_4_l {parent=dyn_hair_3_l, loc=<7.2884 0 0>, rot=<0 0 0 1>}
 dyn_hair_5_l {parent=dyn_hair_4_l, loc=<7.2939 0 0>, rot=<0 0 0 1>}
 dyn_hair_6_l {parent=dyn_hair_5_l, loc=<7.3213 0 0>, rot=<0 0 0 1>}
 dyn_hair_7_l {parent=dyn_hair_6_l, loc=<7.3211 0 0>, rot=<0 0 0 1>}
 base_hair_1_r {parent=head, loc=<1.2116 -5.0574 5.6>, rot=<0.1334 -0.6944 0.6944 -0.1334>}
 dyn_hair_1_r {parent=base_hair_1_r, loc=<0 0 0>, rot=<0 0 0 1>}
 dyn_hair_2_r {parent=dyn_hair_1_r, loc=<7.3206 0 -0.0001>, rot=<0 0 0 1>}
 dyn_hair_3_r {parent=dyn_hair_2_r, loc=<7.3192 0 0.0002>, rot=<0 0 0 1>}
 dyn_hair_4_r {parent=dyn_hair_3_r, loc=<7.2887 0 -0.0001>, rot=<0 0 0 1>}
 dyn_hair_5_r {parent=dyn_hair_4_r, loc=<7.2941 0 0.0001>, rot=<0 0 0 1>}
 dyn_hair_6_r {parent=dyn_hair_5_r, loc=<7.321 0 -0.0001>, rot=<0 0 0 1>}
 dyn_hair_7_r {parent=dyn_hair_6_r, loc=<7.321 0 0>, rot=<0 0 0 1>}
 base_front_1_l {parent=head, loc=<2.1092 2.2933 -12.0367>, rot=<0.089 -0.9878 0.1274 0.0104>}
 dyn_front_1_l {parent=base_front_1_l, loc=<0 0 0>, rot=<0 0 0 1>}
 dyn_front_2_l {parent=dyn_front_1_l, loc=<5.41 0 0>, rot=<0.0963 -0.2979 -0.2082 -0.9266>}
 dyn_front_3_l {parent=dyn_front_2_l, loc=<5.4113 0 0>, rot=<0.1712 0.3731 -0.0741 -0.9088>}
 dyn_front_4_l {parent=dyn_front_3_l, loc=<5.2821 0 0>, rot=<-0.0255 -0.0713 0.1674 -0.983>}
 dyn_front_5_l {parent=dyn_front_4_l, loc=<5.369 0 0>, rot=<-0.0358 -0.1067 -0.1478 -0.9826>}
 dyn_front_6_l {parent=dyn_front_5_l, loc=<5.3796 0 0>, rot=<0.038 0.1567 0.2296 -0.9598>}
 dyn_front_7_l {parent=dyn_front_6_l, loc=<5.4828 0 0>, rot=<0 0 0 1>}
 base_side_1_l {parent=head, loc=<2.1671 -4.7708 -10.2321>, rot=<0.1028 -0.9828 0.1529 0.0147>}
 dyn_side_1_l {parent=base_side_1_l, loc=<0 0 0>, rot=<0 0 0 1>}
 dyn_side_2_l {parent=dyn_side_1_l, loc=<3.7385 0 0>, rot=<-0.0094 0.0003 -0.23 -0.9731>}
 dyn_side_3_l {parent=dyn_side_2_l, loc=<3.7671 0 0>, rot=<-0.0734 0.0433 -0.2647 -0.9606>}
 dyn_side_4_l {parent=dyn_side_3_l, loc=<3.7063 0 0>, rot=<-0.0569 0.0604 -0.0135 -0.9965>}
 dyn_side_5_l {parent=dyn_side_4_l, loc=<3.7381 0 0>, rot=<-0.1153 0.0842 0.1947 -0.9704>}
 dyn_side_6_l {parent=dyn_side_5_l, loc=<3.7495 0 0>, rot=<0.5975 -0.3476 -0.0679 -0.7194>}
 dyn_side_7_l {parent=dyn_side_6_l, loc=<3.7186 0 0>, rot=<0 0 0 1>}
 base_earing_01 {parent=head, loc=<0.0321 0.1128 6.304>, rot=<-0.0007 0.1736 0.9848 -0.0041>}
 dyn_earing_01_r {parent=base_earing_01, loc=<0 0 0>, rot=<0 0 0 1>}
 dyn_earing_02_r {parent=dyn_earing_01_r, loc=<0.9998 -0.0581 0.1972>, rot=<0 0 0 1>}
 dyn_earing_03_r {parent=dyn_earing_02_r, loc=<1.0639 0 0>, rot=<0 0 0 1>}
 base_offset_hair_r {parent=head, loc=<-8.186 -1.9499 6.2396>, rot=<-0.0009 -0.9948 0.0603 -0.0822>}
 dyn_offset_hair_r {parent=base_offset_hair_r, loc=<0 0 0>, rot=<0 0 0 1>}
 FACIAL_C_FacialRoot {parent=head, loc=<2.1524 -0.746 -0.0002>, rot=<0.4879 -0.5118 -0.4879 -0.5118>}
 FACIAL_C_UnderChin {parent=FACIAL_C_FacialRoot, loc=<0.0001 8.7273 4.3111>, rot=<0.4954 0 0 -0.8687>}
 FACIAL_L_UnderChin {parent=FACIAL_C_FacialRoot, loc=<3.0211 7.4205 3.004>, rot=<0.3982 -0.3553 0.0805 -0.8419>}
 FACIAL_R_UnderChin {parent=FACIAL_C_FacialRoot, loc=<-3.0227 7.4201 3.0041>, rot=<0.3982 0.3553 -0.0805 -0.8419>}
 FACIAL_C_Jaw {parent=FACIAL_C_FacialRoot, loc=<-0.0003 2.5065 2.1443>, rot=<0.3372 0 0 -0.9414>}
 FACIAL_C_Jawline {parent=FACIAL_C_Jaw, loc=<-0.0003 2.3627 7.3008>, rot=<0.2586 0 0 -0.966>}
 FACIAL_L_Ear {parent=FACIAL_C_FacialRoot, loc=<6.4154 0.1413 0.7018>, rot=<-0.0314 -0.0005 -0.0262 -0.9992>}
 FACIAL_R_Ear {parent=FACIAL_C_FacialRoot, loc=<-6.4157 0.1391 0.6967>, rot=<-0.0314 0.0006 0.0262 -0.9992>}
 FACIAL_L_Jawline {parent=FACIAL_C_FacialRoot, loc=<4.336 5.6413 4.509>, rot=<0.2044 0.158 -0.0538 -0.9646>}
 FACIAL_R_Jawline {parent=FACIAL_C_FacialRoot, loc=<-4.3387 5.6407 4.5089>, rot=<0.2044 -0.158 0.0539 -0.9646>}
 head_null {parent=neck_02, type=Null, loc=<5.3914 0 0>, rot=<0 0 0.115 -0.9934>}
 clavicle_l {parent=spine_05, loc=<5.4831 -0.1053 -3.2946>, rot=<0.0465 -0.7289 -0.009 -0.6829>}
 upperarm_l {parent=clavicle_l, loc=<9.8658 -1.2136 -0.4105>, rot=<0.0164 -0.3245 -0.0269 -0.9454>}
 lowerarm_l {parent=upperarm_l, loc=<26.465 -0.8345 0.259>, rot=<0.0036 0.0161 -0.2298 -0.9731>}
 hand_l {parent=lowerarm_l, loc=<22.7355 -0.3203 -0.6105>, rot=<-0.6784 -0.0223 0.0735 -0.7307>}
 thumb_01_l {parent=hand_l, loc=<1.3118 -0.3191 -2.459>, rot=<0.6546 -0.4042 0.0664 -0.6354>}
 thumb_02_l {parent=thumb_01_l, loc=<5.9336 -0.0448 0.0401>, rot=<-0.0037 0.0213 0.298 -0.9543>}
 thumb_03_l {parent=thumb_02_l, loc=<2.2223 0.3682 -0.0907>, rot=<0.0118 -0.013 0.0066 -0.9998>}
 index_metacarpal_l {parent=hand_l, loc=<1.6535 0.0183 -2.96>, rot=<0.0158 -0.0124 -0.0173 -0.9996>}
 index_01_l {parent=index_metacarpal_l, loc=<8.3008 0.0357 -0.0524>, rot=<-0.0044 0.0485 0.0924 -0.9945>}
 index_02_l {parent=index_01_l, loc=<3.3587 -0.3004 0.0452>, rot=<-0.002 0.0027 0.3225 -0.9466>}
 index_03_l {parent=index_02_l, loc=<2.5578 -0.123 -0.0252>, rot=<-0.002 0.0067 0.2558 -0.9667>}
 middle_metacarpal_l {parent=hand_l, loc=<2.1856 -0.1725 -0.5984>, rot=<0.006 0.0168 -0.0329 -0.9993>}
 middle_01_l {parent=middle_metacarpal_l, loc=<7.3275 -0.2447 0.0129>, rot=<-0.0208 0.0042 0.0858 -0.9961>}
 middle_02_l {parent=middle_01_l, loc=<4.0123 0.1401 0.1175>, rot=<-0.0218 0.0258 0.3512 -0.9357>}
 middle_03_l {parent=middle_02_l, loc=<2.6077 0.0534 0.1373>, rot=<0.0082 0.0188 0.1708 -0.9851>}
 ring_metacarpal_l {parent=hand_l, loc=<2.2074 -0.3625 1.3638>, rot=<-0.0284 0.0722 -0.0458 -0.9959>}
 ring_01_l {parent=ring_metacarpal_l, loc=<7.0295 -0.6197 -0.1446>, rot=<-0.0489 -0.0252 0.1892 -0.9804>}
 ring_02_l {parent=ring_01_l, loc=<3.6174 0.4647 -0.1027>, rot=<0 0 0.2806 -0.9598>}
 ring_03_l {parent=ring_02_l, loc=<2.4152 0.114 0.0199>, rot=<0 0 0.1189 -0.9929>}
 pinky_metacarpal_l {parent=hand_l, loc=<1.7616 -0.9753 3.3814>, rot=<-0.0397 0.0814 -0.069 -0.9935>}
 pinky_01_l {parent=pinky_metacarpal_l, loc=<7.1709 -0.4831 -0.1094>, rot=<-0.0654 -0.0035 0.2242 -0.9723>}
 pinky_02_l {parent=pinky_01_l, loc=<2.8274 0.0893 -0.1043>, rot=<0.0037 0.0011 0.2854 -0.9584>}
 pinky_03_l {parent=pinky_02_l, loc=<2.2046 0.0307 -0.0078>, rot=<-0.0002 0.0098 0.1445 -0.9895>}
 hand_attach_l {parent=hand_l, loc=<8.5 -3 0>, rot=<0.7071 0.7071 0 0>}
 weapon_l {parent=hand_l, loc=<-0.1974 -3.4111 0.3811>, rot=<0 0 0 1>}
 deform_wrist_dn_base_l {parent=hand_l, loc=<0 -1.5326 0>, rot=<0 0 0 -1>}
 deform_wrist_dn_l {parent=deform_wrist_dn_base_l, loc=<0 0 0>, rot=<0 0 0 1>}
 deform_wrist_up_base_l {parent=hand_l, loc=<0 2.3009 0>, rot=<0 0 0 -1>}
 deform_wrist_up_l {parent=deform_wrist_up_base_l, loc=<0 0 0>, rot=<0 0 0 1>}
 lowerarm_twist_02_l {parent=lowerarm_l, loc=<10.6707 0 0>, rot=<0 0 0 1>}
 lowerarm_twist_01_l {parent=lowerarm_l, loc=<18.316 0 0>, rot=<0 0 0 1>}
 deform_elbow_in_base_l {parent=lowerarm_l, loc=<0.1834 3.6062 -0.2572>, rot=<0 0 0 -1>}
 deform_elbow_in_l {parent=deform_elbow_in_base_l, loc=<0 0 0>, rot=<0 0 0 1>}
 deform_elbow_out_base_l {parent=lowerarm_l, loc=<0 -4 0>, rot=<0 0 0 1>}
 deform_elbow_out_l {parent=deform_elbow_out_base_l, loc=<0 0 0>, rot=<0 0 0 1>}
 upperarm_twist_01_l {parent=upperarm_l, loc=<7.2135 -0.1134 0>, rot=<0 0 0 1>}
 deform_upperarm_bk_base_l {parent=upperarm_twist_01_l, loc=<-6.4053 -4.3746 0.314>, rot=<0 0 0 -1>}
 deform_upperarm_bk_l {parent=deform_upperarm_bk_base_l, loc=<0 0 0>, rot=<0 0 0 1>}
 deform_upperarm_fr_base_l {parent=upperarm_twist_01_l, loc=<-5.6312 5.092 0.3354>, rot=<0 0 0 -1>}
 deform_upperarm_fr_l {parent=deform_upperarm_fr_base_l, loc=<0 0 0>, rot=<0 0 0 1>}
 deform_upperarm_up_base_l {parent=upperarm_twist_01_l, loc=<-8.891 0.4825 3.9517>, rot=<0 0 0 -1>}
 deform_upperarm_up_l {parent=deform_upperarm_up_base_l, loc=<0 0 0>, rot=<0 0 0 1>}
 upperarm_twist_02_l {parent=upperarm_l, loc=<19.7998 -0.1134 0>, rot=<0 0 0 1>}
 upperarm_l_null {parent=clavicle_l, type=Null, loc=<9.8658 -1.2136 -0.4105>, rot=<0.0164 -0.3245 -0.0269 -0.9454>}
 clavicle_r {parent=spine_05, loc=<5.4831 -0.1053 3.2946>, rot=<0.7289 0.0465 -0.6829 0.009>}
 upperarm_r {parent=clavicle_r, loc=<-9.8658 1.2136 0.4105>, rot=<0.0164 -0.3245 -0.0269 -0.9454>}
 lowerarm_r {parent=upperarm_r, loc=<-26.465 0.8345 -0.259>, rot=<0.0036 0.0161 -0.2298 -0.9731>}
 hand_r {parent=lowerarm_r, loc=<-22.7355 0.3203 0.6106>, rot=<-0.6784 -0.0223 0.0735 -0.7307>}
 thumb_01_r {parent=hand_r, loc=<-1.3118 0.3191 2.459>, rot=<0.6546 -0.4042 0.0664 -0.6354>}
 thumb_02_r {parent=thumb_01_r, loc=<-5.9336 0.0448 -0.0401>, rot=<-0.0037 0.0213 0.298 -0.9543>}
 thumb_03_r {parent=thumb_02_r, loc=<-2.2223 -0.3682 0.0907>, rot=<0.0118 -0.013 0.0066 -0.9998>}
 index_metacarpal_r {parent=hand_r, loc=<-1.6535 -0.0183 2.96>, rot=<0.0158 -0.0124 -0.0173 -0.9996>}
 index_01_r {parent=index_metacarpal_r, loc=<-8.3008 -0.0357 0.0524>, rot=<-0.0044 0.0485 0.0924 -0.9945>}
 index_02_r {parent=index_01_r, loc=<-3.3587 0.3004 -0.0452>, rot=<-0.002 0.0027 0.3225 -0.9466>}
 index_03_r {parent=index_02_r, loc=<-2.5578 0.123 0.0252>, rot=<-0.002 0.0067 0.2558 -0.9667>}
 middle_metacarpal_r {parent=hand_r, loc=<-2.1856 0.1725 0.5984>, rot=<0.006 0.0168 -0.0329 -0.9993>}
 middle_01_r {parent=middle_metacarpal_r, loc=<-7.3275 0.2447 -0.0129>, rot=<-0.0208 0.0042 0.0858 -0.9961>}
 middle_02_r {parent=middle_01_r, loc=<-4.0123 -0.1401 -0.1175>, rot=<-0.0218 0.0258 0.3512 -0.9357>}
 middle_03_r {parent=middle_02_r, loc=<-2.6077 -0.0534 -0.1373>, rot=<0.0082 0.0188 0.1708 -0.9851>}
 ring_metacarpal_r {parent=hand_r, loc=<-2.2074 0.3625 -1.3638>, rot=<-0.0284 0.0722 -0.0458 -0.9959>}
 ring_01_r {parent=ring_metacarpal_r, loc=<-7.0295 0.6197 0.1446>, rot=<-0.0489 -0.0252 0.1892 -0.9804>}
 ring_02_r {parent=ring_01_r, loc=<-3.6174 -0.4647 0.1027>, rot=<0 0 0.2806 -0.9598>}
 ring_03_r {parent=ring_02_r, loc=<-2.4152 -0.114 -0.0199>, rot=<0 0 0.1189 -0.9929>}
 pinky_metacarpal_r {parent=hand_r, loc=<-1.7616 0.9753 -3.3814>, rot=<-0.0397 0.0814 -0.069 -0.9935>}
 pinky_01_r {parent=pinky_metacarpal_r, loc=<-7.1709 0.4831 0.1094>, rot=<-0.0654 -0.0035 0.2242 -0.9723>}
 pinky_02_r {parent=pinky_01_r, loc=<-2.8274 -0.0893 0.1043>, rot=<0.0037 0.0011 0.2854 -0.9584>}
 pinky_03_r {parent=pinky_02_r, loc=<-2.2046 -0.0307 0.0078>, rot=<-0.0002 0.0098 0.1445 -0.9895>}
 weapon_r {parent=hand_r, loc=<0.1974 3.4111 -0.3811>, rot=<0 0 0 1>}
 hand_attach_r {parent=hand_r, loc=<-8.5001 2.9999 0>, rot=<0 0 -0.7071 -0.7071>}
 deform_wrist_dn_base_r {parent=hand_r, loc=<-0.0001 1.5328 0>, rot=<0 0 0 -1>}
 deform_wrist_dn_r {parent=deform_wrist_dn_base_r, loc=<0 0 0>, rot=<0 0 0 1>}
 deform_wrist_up_base_r {parent=hand_r, loc=<-0.0002 -2.3007 -0.0001>, rot=<0 0 0 -1>}
 deform_wrist_up_r {parent=deform_wrist_up_base_r, loc=<0 0 0>, rot=<0 0 0 1>}
 lowerarm_twist_02_r {parent=lowerarm_r, loc=<-10.6707 0 0>, rot=<0 0 0 -1>}
 lowerarm_twist_01_r {parent=lowerarm_r, loc=<-18.316 0 0>, rot=<0 0 0 -1>}
 deform_elbow_in_base_r {parent=lowerarm_r, loc=<-0.1834 -3.6062 0.2572>, rot=<0 0 0 -1>}
 deform_elbow_in_r {parent=deform_elbow_in_base_r, loc=<0 0 0>, rot=<0 0 0 1>}
 deform_elbow_out_base_r {parent=lowerarm_r, loc=<0 4 -0.0001>, rot=<0 0 0 -1>}
 deform_elbow_out_r {parent=deform_elbow_out_base_r, loc=<0 0 0>, rot=<0 0 0 1>}
 upperarm_twist_01_r {parent=upperarm_r, loc=<-7.2135 0.1134 0>, rot=<0 0 0 -1>}
 deform_upperarm_bk_base_r {parent=upperarm_twist_01_r, loc=<6.4053 4.3746 -0.314>, rot=<0 0 0 -1>}
 deform_upperarm_bk_r {parent=deform_upperarm_bk_base_r, loc=<0 0 0>, rot=<0 0 0 1>}
 deform_upperarm_fr_base_r {parent=upperarm_twist_01_r, loc=<5.6312 -5.092 -0.3354>, rot=<0 0 0 -1>}
 deform_upperarm_fr_r {parent=deform_upperarm_fr_base_r, loc=<0 0 0>, rot=<0 0 0 1>}
 deform_upperarm_up_base_r {parent=upperarm_twist_01_r, loc=<8.891 -0.4825 -3.9517>, rot=<0 0 0 -1>}
 deform_upperarm_up_r {parent=deform_upperarm_up_base_r, loc=<0 0 0>, rot=<0 0 0 1>}
 base_rose_r {parent=upperarm_twist_01_r, loc=<12.5756 -1.7678 -1.3366>, rot=<0.9343 -0.0039 0.3563 -0.0065>}
 dyn_rose_r {parent=base_rose_r, loc=<0 0 0>, rot=<0 0 0 1>}
 upperarm_twist_02_r {parent=upperarm_r, loc=<-19.7998 0.1134 0>, rot=<0 0 0 -1>}
 upperarm_r_null {parent=clavicle_r, type=Null, loc=<-9.8658 1.2136 0.4105>, rot=<0.0164 -0.3245 -0.0269 -0.9454>}
 attach_backpack {parent=spine_05, loc=<-0.123 -13.473 0>, rot=<-0.0093 -0.707 0.0093 -0.707>}
 Backpack_BR {parent=attach_backpack, type=Null, loc=<0 1.5 0>, rot=<0 0 0 1>}
 attach_cape {parent=spine_05, loc=<9.017 -11.714 0>, rot=<0.0616 -0.7044 -0.0616 -0.7044>}
 cape {parent=attach_cape, type=Null, loc=<0 0 0>, rot=<0 0 0 1>}
 thigh_l {parent=pelvis, loc=<-1.0855 -0.2109 -8.7053>, rot=<0.0009 0.0355 0.0155 -0.9993>}
 calf_l {parent=thigh_l, loc=<-42.5284 -0.0019 -0.7297>, rot=<0.0006 -0.0161 -0.0391 -0.9991>}
 foot_l {parent=calf_l, loc=<-46.7401 -2.5014 -0.592>, rot=<0.0427 -0.0238 0.0353 -0.9982>}
 ball_l {parent=foot_l, loc=<-3.5658 10.8066 -0.1544>, rot=<0.0001 -0.0001 -0.7063 -0.7079>}
 calf_twist_02_l {parent=calf_l, loc=<-19.1546 -1.7943 -0.2638>, rot=<0.0237 0.0095 -0.0035 -0.9997>}
 calf_twist_01_l {parent=calf_l, loc=<-35.8483 -1.9197 -0.5794>, rot=<0.0237 0.0095 -0.0035 -0.9997>}
 deform_knee_in_base_l {parent=calf_l, loc=<0.8031 -5.0231 0.2456>, rot=<0 0 0 -1>}
 deform_knee_in_l {parent=deform_knee_in_base_l, loc=<0 0 0>, rot=<0 0 0 1>}
 deform_knee_out_base_l {parent=calf_l, loc=<0 4 0>, rot=<0 0 0 -1>}
 deform_knee_out_l {parent=deform_knee_out_base_l, loc=<0 0 0>, rot=<0 0 0 1>}
 foot_l_null {parent=calf_l, type=Null, loc=<-46.7402 -2.5014 -0.592>, rot=<0.0427 -0.0238 0.0353 -0.9982>}
 thigh_twist_01_l {parent=thigh_l, loc=<-27.952 0 0>, rot=<0.0013 0 0 -1>}
 Offset_skirt_side_l {parent=thigh_l, loc=<4.3923 0.0923 -5.1241>, rot=<0.0311 -0.9914 -0.126 -0.0158>}
 Offset_skirt_back_l {parent=thigh_l, loc=<5.8104 -7.3123 -1.3891>, rot=<0.0311 -0.9914 -0.126 -0.0158>}
 calf_l_null {parent=thigh_l, type=Null, loc=<-42.5284 -0.0019 -0.7297>, rot=<0.0006 -0.0161 -0.0391 -0.9991>}
 deform_groin_base_l {parent=thigh_l, loc=<0.4801 5.6776 -0.2486>, rot=<0 0 0 -1>}
 deform_groin_l {parent=deform_groin_base_l, loc=<0 0 0>, rot=<0 0 0 1>}
 skirt_l_root {parent=thigh_l, loc=<0 0 0>, rot=<-0.0024 -0.9994 -0.0003 0.0355>}
 base_skirt_side_l {parent=skirt_l_root, loc=<-4.7518 -2.6407 4.457>, rot=<0 0 0 -1>}
 dyn_skirt_side_l_01 {parent=base_skirt_side_l, loc=<0 0 0>, rot=<0 0 0 1>}
 dyn_skirt_side_l_02 {parent=dyn_skirt_side_l_01, loc=<11.5157 -3.1266 3.0437>, rot=<0 0 0 1>}
 dyn_skirt_side_l_03 {parent=dyn_skirt_side_l_02, loc=<11.7086 -0.7574 2.0883>, rot=<0 0 0 1>}
 dyn_skirt_side_l_04 {parent=dyn_skirt_side_l_03, loc=<20.7147 -0.3991 2.097>, rot=<0 0 0 1>}
 dyn_skirt_side_l_05 {parent=dyn_skirt_side_l_04, loc=<13.4916 2.3815 0.9699>, rot=<0 0 0 -1>}
 dyn_skirt_side_l_06 {parent=dyn_skirt_side_l_05, loc=<14.7287 1.1176 2.0853>, rot=<0 0 0 1>}
 base_skirt_back_l {parent=skirt_l_root, loc=<-6.3467 -8.469 -1.9566>, rot=<0 0 0 -1>}
 dyn_skirt_back_l_01 {parent=base_skirt_back_l, loc=<0 0 0>, rot=<0 0 0 1>}
 dyn_skirt_back_l_02 {parent=dyn_skirt_back_l_01, loc=<10.6078 -3.9558 1.5496>, rot=<0 0 0 1>}
 dyn_skirt_back_l_03 {parent=dyn_skirt_back_l_02, loc=<12.3766 -2.0555 0.5414>, rot=<0 0 0 1>}
 dyn_skirt_back_l_04 {parent=dyn_skirt_back_l_03, loc=<17.4459 -2.2763 1.4185>, rot=<0 0 0 1>}
 skirt_l_front_root {parent=thigh_l, loc=<0 0 0>, rot=<-0.0024 -0.9994 -0.0003 0.0355>}
 base_skirt_front_l_01 {parent=skirt_l_front_root, loc=<-5.1109 6.6858 1.2549>, rot=<0.7071 0 0 -0.7071>}
 dyn_skirt_front_l_01 {parent=base_skirt_front_l_01, loc=<0 0 0>, rot=<0 0 0 1>}
 dyn_skirt_front_l_02 {parent=dyn_skirt_front_l_01, loc=<13.5587 -0.0401 0.2854>, rot=<0 0 0 1>}
 dyn_skirt_front_l_03 {parent=dyn_skirt_front_l_02, loc=<6.58 0.0024 0.63>, rot=<0 0 0 1>}
 thigh_r {parent=pelvis, loc=<-1.0855 -0.2109 8.7053>, rot=<0.0355 -0.0009 0.9993 0.0155>}
 calf_r {parent=thigh_r, loc=<42.5284 0.0019 0.7297>, rot=<0.0006 -0.0161 -0.0391 -0.9991>}
 foot_r {parent=calf_r, loc=<46.7401 2.5014 0.592>, rot=<0.0427 -0.0238 0.0353 -0.9982>}
 ball_r {parent=foot_r, loc=<3.5658 -10.8066 0.1544>, rot=<0.0001 -0.0001 -0.7063 -0.7079>}
 calf_twist_02_r {parent=calf_r, loc=<19.1546 1.7943 0.2638>, rot=<0.0237 0.0095 -0.0035 -0.9997>}
 calf_twist_01_r {parent=calf_r, loc=<35.8483 1.9197 0.5794>, rot=<0.0237 0.0095 -0.0035 -0.9997>}
 deform_knee_in_base_r {parent=calf_r, loc=<-0.8031 5.0231 -0.2457>, rot=<0 0 0 -1>}
 deform_knee_in_r {parent=deform_knee_in_base_r, loc=<0 0 0>, rot=<0 0 0 1>}
 deform_knee_out_base_r {parent=calf_r, loc=<0 -4 0>, rot=<0 0 0 -1>}
 deform_knee_out_r {parent=deform_knee_out_base_r, loc=<0 0 0>, rot=<0 0 0 1>}
 foot_r_null {parent=calf_r, type=Null, loc=<46.7401 2.5014 0.592>, rot=<0.0427 -0.0238 0.0353 -0.9982>}
 thigh_twist_01_r {parent=thigh_r, loc=<27.952 0 0>, rot=<0.0013 0 0 -1>}
 Offset_skirt_back_r {parent=thigh_r, loc=<-5.8104 7.3123 1.3891>, rot=<0.9759 0.0145 0.0048 -0.2178>}
 Offset_skirt_side_r {parent=thigh_r, loc=<-4.3698 -0.0816 5.1088>, rot=<0.9759 0.0145 0.0048 -0.2178>}
 calf_r_null {parent=thigh_r, type=Null, loc=<42.5284 0.0019 0.7297>, rot=<0.0006 -0.0161 -0.0391 -0.9991>}
 deform_groin_base_r {parent=thigh_r, loc=<-0.4801 -5.6776 0.2486>, rot=<0 0 0 -1>}
 deform_groin_r {parent=deform_groin_base_r, loc=<0 0 0>, rot=<0 0 0 1>}
 skirt_r_root {parent=thigh_r, loc=<0 0 0>, rot=<0.9994 -0.0024 -0.0355 -0.0003>}
 base_skirt_back_r {parent=skirt_r_root, loc=<-6.3467 -8.469 1.9566>, rot=<0 0 0 -1>}
 dyn_skirt_back_r_01 {parent=base_skirt_back_r, loc=<0 0 0>, rot=<0 0 0 1>}
 dyn_skirt_back_r_02 {parent=dyn_skirt_back_r_01, loc=<10.6078 -3.9558 -1.5496>, rot=<0 0 0 1>}
 dyn_skirt_back_r_03 {parent=dyn_skirt_back_r_02, loc=<12.3766 -2.0555 -0.5414>, rot=<0.0031 -0.0247 0.0926 -0.9954>}
 dyn_skirt_back_r_04 {parent=dyn_skirt_back_r_03, loc=<17.6144 0.9914 -0.5515>, rot=<-0.0031 0.0247 -0.0926 -0.9954>}
 base_skirt_side_r {parent=skirt_r_root, loc=<-4.7518 -2.6407 -4.457>, rot=<0 0 0 1>}
 dyn_skirt_side_r_01 {parent=base_skirt_side_r, loc=<0 0 0>, rot=<0 0 0 1>}
 dyn_skirt_side_r_02 {parent=dyn_skirt_side_r_01, loc=<11.5157 -3.1266 -3.0437>, rot=<-0.026 0 0 -0.9997>}
 dyn_skirt_side_r_03 {parent=dyn_skirt_side_r_02, loc=<11.7086 -0.8651 -2.046>, rot=<0.1279 0 0 -0.9918>}
 dyn_skirt_side_r_04 {parent=dyn_skirt_side_r_03, loc=<20.7147 0.0348 -2.1343>, rot=<-0.102 0 0 -0.9948>}
 dyn_skirt_side_r_05 {parent=dyn_skirt_side_r_04, loc=<13.4916 2.3815 -0.9699>, rot=<0 0 0 1>}
 dyn_skirt_side_r_06 {parent=dyn_skirt_side_r_05, loc=<14.7287 1.1176 -2.0853>, rot=<0 0 0 1>}
 skirt_r_front_root {parent=thigh_r, loc=<0 0 0>, rot=<0.9994 -0.0024 -0.0355 -0.0003>}
 base_skirt_front_r_01 {parent=skirt_r_front_root, loc=<-5.1109 6.6858 -1.2549>, rot=<0.7071 0 0 -0.7071>}
 dyn_skirt_front_r_01 {parent=base_skirt_front_r_01, loc=<0 0 0>, rot=<0 0 0 1>}
 dyn_skirt_front_r_02 {parent=dyn_skirt_front_r_01, loc=<13.5587 -0.0401 0.2854>, rot=<0 0 0 1>}
 dyn_skirt_front_r_03 {parent=dyn_skirt_front_r_02, loc=<6.58 0.0024 0.63>, rot=<0 0 0 1>}
 base_skirt_back_m {parent=pelvis, loc=<5.0439 -9.2746 0>, rot=<-0.0179 -0.9998 0 0>}
 dyn_skirt_back_m_01 {parent=base_skirt_back_m, loc=<0 0 0>, rot=<0 0 0 -1>}
 dyn_skirt_back_m_02 {parent=dyn_skirt_back_m_01, loc=<10.3464 -4.7388 0.0001>, rot=<0.0181 -0.0018 -0.0016 -0.9998>}
 dyn_skirt_back_m_03 {parent=dyn_skirt_back_m_02, loc=<12.352 -2.6713 -0.0645>, rot=<-0.0181 0.0018 0.0016 -0.9998>}
 dyn_skirt_back_m_04 {parent=dyn_skirt_back_m_03, loc=<16.4263 -2.8525 0.088>, rot=<0 0 0 1>}
 thigh_l_null {parent=pelvis, type=Null, loc=<-1.0855 -0.2109 -8.7053>, rot=<0.0009 0.0355 0.0155 -0.9993>}
 thigh_r_null {parent=pelvis, type=Null, loc=<-1.0855 -0.2109 8.7053>, rot=<0.0355 -0.0009 0.9993 0.0155>}
 deform_glute_base_l {parent=pelvis, loc=<-7.3466 -9.0213 -9.1245>, rot=<0.0172 0.031 -0.4477 -0.8935>}
 deform_glute_l {parent=deform_glute_base_l, loc=<0 0 0>, rot=<0 0 0 1>}
 deform_glute_base_r {parent=pelvis, loc=<-7.3464 -9.0213 9.1245>, rot=<0.031 -0.0172 0.8935 -0.4477>}
 deform_glute_r {parent=deform_glute_base_r, loc=<0 0 0>, rot=<0 0 0 1>}
 base_skirt_front_m_01 {parent=pelvis, loc=<4.2998 7.353 0.0567>, rot=<0.0127 0.707 0.707 0.0127>}
 dyn_skirt_front_m_01 {parent=base_skirt_front_m_01, loc=<0 0 0>, rot=<0 0 0 1>}
 dyn_skirt_front_m_02 {parent=dyn_skirt_front_m_01, loc=<13.5587 -0.0401 0.2854>, rot=<0 0 0 1>}
 dyn_skirt_front_m_03 {parent=dyn_skirt_front_m_02, loc=<6.58 0.0024 0.63>, rot=<0 0 0 1>}
 base_offset_hip {parent=pelvis, loc=<1.4685 -1.8367 0>, rot=<-0.0127 -0.707 0.0127 -0.707>}
 dyn_offset_hip {parent=base_offset_hip, loc=<0 0 0>, rot=<0 0 0 1>}
 attach {parent=root, loc=<0 0 0>, rot=<0 0 0 1>}
 ik_hand_root {parent=root, loc=<0 0 0>, rot=<0 0 0 1>}
 ik_hand_gun {parent=ik_hand_root, loc=<-48.1743 3.3545 104.4381>, rot=<0.6276 0.1309 0.4183 -0.6434>}
 ik_hand_r {parent=ik_hand_gun, loc=<0 0 0>, rot=<0 0 0 1>}
 ik_hand_l {parent=ik_hand_gun, loc=<59.3271 67.6997 34.3524>, rot=<-0.6158 -0.7027 -0.3565 0>}
 ik_foot_root {parent=root, loc=<0 0 0>, rot=<0 0 0 1>}
 ik_foot_l {parent=ik_foot_root, loc=<14.8262 -7.3545 8.296>, rot=<0.0357 0.7045 0.0284 -0.7082>}
 ik_foot_r {parent=ik_foot_root, loc=<-14.8262 -7.3545 8.296>, rot=<0.7082 0.0284 -0.7045 0.0357>}
 root_fp {parent=root, loc=<0 0 0>, rot=<0 0 0 1>}
 attach_fp {parent=root_fp, loc=<0 0 0>, rot=<0 0 0 1>}
 camera_root {parent=root_fp, loc=<0 0 0>, rot=<0 0 0 1>}
 camera {parent=camera_root, loc=<0 0 0>, rot=<0 0 0 1>}
 attach_emote_01 {parent=root, loc=<0 0 0>, rot=<0 0 0 1>}
 attach_emote_02 {parent=root, loc=<0 0 0>, rot=<0 0 0 1>}
 attach_emote_03 {parent=root, loc=<0 0 0>, rot=<0 0 0 1>}
 warp_point_01 {parent=root, loc=<0 0 0>, rot=<0 0 0 1>}
 warp_point_02 {parent=root, loc=<0 0 0>, rot=<0 0 0 1>}
 warp_point_03 {parent=root, loc=<0 0 0>, rot=<0 0 0 1>}
]
variables [
 PelvisProx {type=double, default=0}
 Pub_EasternBlock Emote {type=bool, default=False}
 Pub_IsBalletSpin Emote {type=bool, default=False}
 Pub_IsCelebration Emote {type=bool, default=False}
 Pub_IsCrouching {type=bool, default=False}
 Pub_IsDBNO {type=bool, default=False}
 Pub_IsDBNO_Carried {type=bool, default=False}
 Pub_IsEnableDynamics {type=bool, default=False}
 Pub_IsFrontEnd {type=bool, default=False}
 Pub_IsGrindRail {type=bool, default=False}
 Pub_IsIdleMontage {type=double, default=0}
 Pub_IsJumping {type=bool, default=False}
 Pub_IsKneeSliding {type=bool, default=False}
 Pub_IsOpenParachute {type=bool, default=False}
 Pub_IsPistolEquipped {type=bool, default=False}
 Pub_IsSkydive {type=bool, default=False}
 Pub_IsTacticalSprint {type=bool, default=False}
 Pub_IsTakeOut_Emote {type=bool, default=False}
 Pub_IsTargeting {type=bool, default=False}
 Pub_IsbaseballSliding {type=bool, default=False}
 Pub_IswearingCape {type=bool, default=False}
 Pub_moveBackward {type=double, default=0}
 Pub_moveForward {type=double, default=0}
 Pub_strafeLeft float {type=double, default=0}
 Pub_strafeRight float {type=double, default=0}
 SPR L Leg Bkd {type=double, default=0}
 SPR L Leg Fwd {type=double, default=0}
 SPR L Leg In Twist {type=double, default=0}
 SPR L Leg Itw {type=double, default=0}
 SPR L Leg Otw {type=double, default=0}
 SPR L Leg Out Twist {type=double, default=0}
 SPR L calf Bkd {type=double, default=0}
 SPR L calf ExtremeBkd {type=double, default=0}
 SPR R Leg Bkd {type=double, default=0}
 SPR R Leg Fwd {type=double, default=0}
 SPR R Leg In Twist {type=double, default=0}
 SPR R Leg Itw {type=double, default=0}
 SPR R Leg Otw {type=double, default=0}
 SPR R Leg Out Twist {type=double, default=0}
 SPR R calf Bkd {type=double, default=0}
 SPR R calf ExtremeBkd {type=double, default=0}
]
