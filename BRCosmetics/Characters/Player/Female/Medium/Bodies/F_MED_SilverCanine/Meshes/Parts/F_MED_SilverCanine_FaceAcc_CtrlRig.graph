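# Control Rig for the SilverCanine female character's face accessory, driving its dynamic hair chains and front strand bones procedurally.
# Weighted parent constraints bind hair nulls and front bones to head, neck, spine, clavicles and arms.
# Swing/twist delta-rotation readers and spherical pose readers, scaled by interpolated targeting, weapon and swing states, apply corrective offsets.
# Twist distribution and B-spline IK then pose the left, center and right hair chains.
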ControlRigBlueprint /BRCosmetics/Characters/Player/Female/Medium/Bodies/F_MED_SilverCanine/Meshes/Parts/F_MED_SilverCanine_FaceAcc_CtrlRig
graphs [
 CollapseNode_ContainedGraph {nodes [
   n1 = RigVMFunctionEntryNode()
   constraints = RigVMCollapseNode(ExecutePin=@n1.ExecutePin)
   lerp bools = RigVMCollapseNode(ExecuteContext=@constraints.ExecutePin)
   > RigVMFunction_Sequence = RigVMAggregateNode()
   RigVMFunction_Sequence_1 = RigVMAggregateNode(ExecuteContext=@RigVMFunction_Sequence.A)
   n2 = &RigUnit_SphericalPoseReader(
    ActiveRegionScaleFactors=(PositiveWidth=1,NegativeWidth=1,PositiveHeight=1,NegativeHeight=1)
    ActiveRegionSize=0.05
    Debug=(bDrawDebug=false,bDraw2D=False,bDrawLocalAxes=False,DebugScale=25,DebugSegments=20,DebugThickness=0.25)
    DriverAxis=(X=1,Y=0,Z=0)
    DriverItem=(Type=Bone,Name="head")
    FalloffRegionScaleFactors=(PositiveWidth=1,NegativeWidth=1,PositiveHeight=1,NegativeHeight=1)
    FalloffSize=0.2
    FlipHeightScaling=False
    FlipWidthScaling=False
    OptionalParentItem=(Type=Bone,Name="spine_05")
    RotationOffset=(X=0,Y=-62,Z=-90)
    ExecutePin=@RigVMFunction_Sequence_1.A
   )
   n3 = RigVMVariableNode() {variable=isTargeting lerp}
   > RigVMVariableNode(Value=&RigVMFunction_MathDoubleMul(
     A=&RigVMFunction_MathDoubleMul(A=@n2.OutputParam, B=@n3.Value).Result
     B=RigVMVariableNode() {variable=isPistol Equipped lerp}.Value
    ).Result) {variable=isPistolASDUp Extreme}
   DeltaRotation v4 = RigVMFunctionReferenceNode(
    Rotation Order=ZXY
    child=(Type=Bone,Name="head")
    debugRotation=false
    outputMinMax=(X=0,Y=1)
    parent=(Type=Null,Name="head_null")
    premult=1
    rotationMinMax=(X=0,Y=-60)
    triggerAxis=Z
    twistAxis=(X=1,Y=0,Z=0)
    ExecuteContext=@RigVMFunction_Sequence_1.B
   )
   Remap = &RigVMFunction_MathDoubleRemap(SourceMaximum=1, SourceMinimum=0, TargetMaximum=0, TargetMinimum=1, Value=RigVMVariableNode() {variable=isPistolASDUp Extreme}.Value)
   n4 = &RigVMFunction_MathDoubleMul(B=@Remap.Result, A=@"DeltaRotation v4".outputWeight)
   Multiply = &RigVMFunction_MathDoubleMul(B=0.5, A=@n4.Result)
   > &RigUnit_ModifyTransforms(
    ItemToModify="((Item=(Type=Null,Name=\"dyn_hair_1_l_ctrl_null\"),Transform=(Rotation=(X=0,Y=0,Z=0,W=1),Translation=(X=0,Y=0,Z=5),Scale3D=(X=1,Y=1,Z=1))),(Item=(Type=Null,Name=\"dyn_hair_1_r_ctrl_null\"),Transform=(Rotation=(X=0,Y=0,Z=0,W=1),Translation=(X=0,Y=0,Z=5),Scale3D ...(truncated)"
    WeightMinimum=0
    Weight=@Multiply.Result
   )
   > &RigUnit_ModifyTransforms(
    ItemToModify="((Item=(Type=Null,Name=\"dyn_hair_3_l_ctrl_null\"),Transform=(Rotation=(X=0,Y=0,Z=0,W=1),Translation=(X=1,Y=0,Z=7),Scale3D=(X=1,Y=1,Z=1))),(Item=(Type=Null,Name=\"dyn_hair_3_r_ctrl_null\"),Transform=(Rotation=(X=0,Y=0,Z=0,W=1),Translation=(X=1,Y=0,Z=7),Scale3D ...(truncated)"
    WeightMinimum=0
    Weight=&RigVMFunction_MathDoubleMul(B=1, A=@n4.Result).Result
   )
   > DeltaRotation v4_3 = RigVMFunctionReferenceNode(
    Rotation Order=ZYX
    child=(Type=Bone,Name="head")
    debugRotation=false
    onlyTwist=true
    outputMinMax=(X=0,Y=1)
    parent=(Type=Null,Name="head_null")
    premult=1
    rotationMinMax=(X=0,Y=-60)
    twistAxis=(X=1,Y=0,Z=0)
   )
   > &RigUnit_ModifyTransforms(
    ItemToModify="((Item=(Type=Null,Name=\"dyn_hair_2_r_ctrl_null\"),Transform=(Rotation=(X=0,Y=0,Z=0,W=1),Translation=(X=0,Y=0,Z=3),Scale3D=(X=1,Y=1,Z=1))),(Item=(Type=Null,Name=\"dyn_hair_3_r_ctrl_null\"),Transform=(Rotation=(X=0,Y=0,Z=0,W=-1),Translation=(X=0,Y=0,Z=3),Scale3 ...(truncated)"
    WeightMinimum=0
    Weight=@"DeltaRotation v4_3".outputWeight
   )
   > DeltaRotation v4_5 = RigVMFunctionReferenceNode(
    Rotation Order=ZYX
    child=(Type=Bone,Name="head")
    debugRotation=false
    onlyTwist=true
    outputMinMax=(X=0,Y=1)
    parent=(Type=Null,Name="head_null")
    premult=1
    rotationMinMax=(X=0,Y=60)
    twistAxis=(X=1,Y=0,Z=0)
   )
   > &RigUnit_ModifyTransforms(
    ItemToModify="((Item=(Type=Null,Name=\"dyn_hair_1_r_ctrl_null\"),Transform=(Rotation=(X=0,Y=0,Z=0,W=1),Translation=(X=-0.2341,Y=-3.6072,Z=-0.9853),Scale3D=(X=1,Y=1,Z=1))),(Item=(Type=Null,Name=\"dyn_hair_2_r_ctrl_null\"),Transform=(Rotation=(X=0,Y=0,Z=0,W=-1),Translation=(X ...(truncated)"
    WeightMinimum=0
    Weight=@"DeltaRotation v4_5".outputWeight
   )
   > DeltaRotation v4_7 = RigVMFunctionReferenceNode(
    Rotation Order=ZYX
    child=(Type=Bone,Name="head")
    debugRotation=false
    outputMinMax=(X=0,Y=1)
    parent=(Type=Null,Name="head_null")
    premult=1
    rotationMinMax=(X=-60,Y=-80)
    triggerAxis=Z
    twistAxis=(X=1,Y=0,Z=0)
   )
   > RigVMVariableNode(Value=@"DeltaRotation v4_7".outputWeight) {variable=head rz n60n80 alpha}
   > &RigUnit_ModifyTransforms(
    ItemToModify="((Item=(Type=Null,Name=\"dyn_hair_3_r_ctrl_null\"),Transform=(Rotation=(X=0,Y=0,Z=0,W=-1),Translation=(X=2.2209,Y=-3.1709,Z=-5.7956),Scale3D=(X=1,Y=1,Z=1))),(Item=(Type=Null,Name=\"dyn_hair_2_l_ctrl_null\"),Transform=(Rotation=(X=0,Y=0,Z=0,W=-1),Translation=(X ...(truncated)"
    WeightMinimum=0
    Weight=@"DeltaRotation v4_7".outputWeight
   )
   DeltaRotation v4_1 = RigVMFunctionReferenceNode(
    Rotation Order=ZYX
    child=(Type=Bone,Name="spine_05")
    debugRotation=false
    onlyTwist=true
    outputMinMax=(X=0,Y=1)
    parent=(Type=Bone,Name="spine_03")
    premult=1
    rotationMinMax=(X=0,Y=35)
    twistAxis=(X=1,Y=0,Z=0)
    ExecuteContext=@RigVMFunction_Sequence_1.C
   )
   > &RigUnit_ModifyTransforms(
    ItemToModify=((Item=(Type=Null,Name="dyn_hair_4_r_ctrl_null"),Transform=(Rotation=(X=0,Y=0,Z=0,W=1),Translation=(X=0,Y=-2.112,Z=3.5606),Scale3D=(X=1,Y=1,Z=1))))
    WeightMinimum=0
    Weight=@"DeltaRotation v4_1".outputWeight
   )
   > &RigUnit_ModifyTransforms(
    ItemToModify="((Item=(Type=Null,Name=\"dyn_hair_3_r_ctrl_null\"),Transform=(Rotation=(X=0,Y=0,Z=0,W=1),Translation=(X=0,Y=0,Z=3),Scale3D=(X=1,Y=1,Z=1))),(Item=(Type=Null,Name=\"dyn_hair_2_r_ctrl_null\"),Transform=(Rotation=(X=0,Y=0,Z=0,W=1),Translation=(X=0,Y=0,Z=2.0512),Sc ...(truncated)"
    WeightMinimum=0
    Weight=@"DeltaRotation v4_1".outputWeight
   )
   > DeltaRotation v4_2 = RigVMFunctionReferenceNode(
    Rotation Order=ZYX
    child=(Type=Bone,Name="spine_05")
    debugRotation=false
    onlyTwist=true
    outputMinMax=(X=0,Y=1)
    parent=(Type=Bone,Name="spine_03")
    premult=0
    rotationMinMax=(X=0,Y=-35)
    twistAxis=(X=1,Y=0,Z=0)
   )
   > &RigUnit_ModifyTransforms(
    ItemToModify=((Item=(Type=Null,Name="dyn_hair_4_l_ctrl_null"),Transform=(Rotation=(X=0,Y=0,Z=0,W=1),Translation=(X=0,Y=-2.112,Z=3.5606),Scale3D=(X=1,Y=1,Z=1))))
    WeightMinimum=0
    Weight=@"DeltaRotation v4_2".outputWeight
   )
   > n5 = &RigUnit_SphericalPoseReader(
    ActiveRegionScaleFactors=(PositiveWidth=0.5,NegativeWidth=0.5,PositiveHeight=0.5,NegativeHeight=0.5)
    ActiveRegionSize=0.12
    Debug=(bDrawDebug=false,bDraw2D=False,bDrawLocalAxes=False,DebugScale=25,DebugSegments=20,DebugThickness=0.25)
    DriverAxis=(X=1,Y=0,Z=0)
    DriverItem=(Type=Bone,Name="clavicle_l")
    FalloffRegionScaleFactors=(PositiveWidth=0.5,NegativeWidth=0.5,PositiveHeight=0.5,NegativeHeight=0.5)
    FalloffSize=0.3
    FlipHeightScaling=False
    FlipWidthScaling=False
    OptionalParentItem=(Type=Bone,Name="None")
    RotationOffset=(X=0,Y=90,Z=140)
   )
   > &RigUnit_ModifyTransforms(
    ItemToModify="((Item=(Type=Null,Name=\"dyn_hair_3_l_ctrl_null\"),Transform=(Rotation=(X=0,Y=0,Z=0,W=-1),Translation=(X=0,Y=3,Z=2.1),Scale3D=(X=1,Y=1,Z=1))),(Item=(Type=Null,Name=\"dyn_hair_4_l_ctrl_null\"),Transform=(Rotation=(X=0,Y=0,Z=0,W=-1),Translation=(X=0,Y=3.8,Z=3.2) ...(truncated)"
    WeightMinimum=0
    Weight=@n5.OutputParam
   )
   > n6 = &RigUnit_SphericalPoseReader(
    ActiveRegionScaleFactors=(PositiveWidth=0.5,NegativeWidth=0.5,PositiveHeight=0.5,NegativeHeight=0.5)
    ActiveRegionSize=0.12
    Debug=(bDrawDebug=false,bDraw2D=False,bDrawLocalAxes=False,DebugScale=25,DebugSegments=20,DebugThickness=0.25)
    DriverAxis=(X=-1,Y=0,Z=0)
    DriverItem=(Type=Bone,Name="clavicle_r")
    FalloffRegionScaleFactors=(PositiveWidth=0.5,NegativeWidth=0.5,PositiveHeight=0.5,NegativeHeight=0.5)
    FalloffSize=0.3
    FlipHeightScaling=False
    FlipWidthScaling=False
    OptionalParentItem=(Type=Bone,Name="None")
    RotationOffset=(X=0,Y=-90,Z=140)
   )
   > &RigUnit_ModifyTransforms(
    ItemToModify=((Item=(Type=Null,Name="dyn_hair_3_r_ctrl_null"),Transform=(Translation=(X=0,Y=-3,Z=2.1),Scale3D=(X=1,Y=1,Z=1))),(Item=(Type=Null,Name="dyn_hair_4_r_ctrl_null"),Transform=(Rotation=(X=0,Y=0,Z=0,W=1),Translation=(X=0,Y=-3.8,Z=3.2),Scale3D=(X=1,Y=1,Z=1))))
    WeightMinimum=0
    Weight=@n6.OutputParam
   )
   > n7 = &RigUnit_SphericalPoseReader(
    ActiveRegionScaleFactors=(PositiveWidth=1,NegativeWidth=1,PositiveHeight=1,NegativeHeight=1)
    ActiveRegionSize=0.05
    Debug=(bDrawDebug=false,bDraw2D=False,bDrawLocalAxes=False,DebugScale=25,DebugSegments=20,DebugThickness=0.25)
    DriverAxis=(X=1,Y=0,Z=0)
    DriverItem=(Type=Bone,Name="head")
    FalloffRegionScaleFactors=(PositiveWidth=1,NegativeWidth=1,PositiveHeight=1,NegativeHeight=1)
    FalloffSize=0.2
    FlipHeightScaling=False
    FlipWidthScaling=False
    OptionalParentItem=(Type=Bone,Name="spine_05")
    RotationOffset=(X=0,Y=-50,Z=-15)
   )
   n8 = RigVMVariableNode() {variable=isTargeting lerp}
   > &RigUnit_ModifyTransforms(
    ItemToModify="((Item=(Type=Null,Name=\"dyn_hair_2_r_ctrl_null\"),Transform=(Rotation=(X=0,Y=0,Z=0,W=1),Translation=(X=-4.9,Y=-7.4781,Z=1.8842),Scale3D=(X=1,Y=1,Z=1))),(Item=(Type=Null,Name=\"dyn_hair_3_r_ctrl_null\"),Transform=(Rotation=(X=0,Y=0,Z=0,W=1),Translation=(X=-1.6 ...(truncated)"
    WeightMinimum=0
    Weight=&RigVMFunction_MathDoubleMul(
     A=&RigVMFunction_MathDoubleMul(A=@n7.OutputParam, B=@n8.Value).Result
     B=RigVMVariableNode() {variable=isAR Equipped lerp}.Value
    ).Result
   )
   > n9 = &RigUnit_SphericalPoseReader(
    ActiveRegionScaleFactors=(PositiveWidth=1,NegativeWidth=1,PositiveHeight=1,NegativeHeight=1)
    ActiveRegionSize=0.05
    Debug=(bDrawDebug=false,bDraw2D=False,bDrawLocalAxes=False,DebugScale=25,DebugSegments=20,DebugThickness=0.25)
    DriverAxis=(X=1,Y=0,Z=0)
    DriverItem=(Type=Bone,Name="head")
    FalloffRegionScaleFactors=(PositiveWidth=1,NegativeWidth=1,PositiveHeight=1,NegativeHeight=1)
    FalloffSize=0.2
    FlipHeightScaling=False
    FlipWidthScaling=False
    OptionalParentItem=(Type=Bone,Name="spine_05")
    RotationOffset=(X=0,Y=-65,Z=-52)
   )
   > &RigUnit_ModifyTransforms(
    ItemToModify="((Item=(Type=Null,Name=\"dyn_hair_1_r_ctrl_null\"),Transform=(Rotation=(X=0,Y=0,Z=0,W=-1),Translation=(X=0,Y=5,Z=0),Scale3D=(X=1,Y=1,Z=1))),(Item=(Type=Null,Name=\"dyn_hair_2_r_ctrl_null\"),Transform=(Rotation=(X=0,Y=0,Z=0,W=1),Translation=(X=0,Y=-7,Z=0),Scale ...(truncated)"
    WeightMinimum=0
    Weight=@n9.OutputParam
   )
   n10 = RigVMVariableNode() {variable=isTargeting lerp}
   > &RigUnit_ModifyTransforms(
    ItemToModify=((Item=(Type=Null,Name="dyn_hair_1_r_ctrl_null"),Transform=(Rotation=(X=0,Y=0,Z=0,W=1),Translation=(X=0,Y=0,Z=0),Scale3D=(X=1,Y=1,Z=1))))
    WeightMinimum=0
    Weight=&RigVMFunction_MathDoubleMul(
     A=&RigVMFunction_MathDoubleMul(A=@n9.OutputParam, B=@n10.Value).Result
     B=RigVMVariableNode() {variable=isPistol Equipped lerp}.Value
    ).Result
   )
   joints_twist_c = RigVMDispatchNode(Value=((Type=Bone,Name="dyn_hair_2_c"),(Type=Bone,Name="dyn_hair_3_c"),(Type=Bone,Name="dyn_hair_4_c"))) {template="DISPATCH_RigVMDispatch_Constant(io Value)"}
   TwistBones = RigVMFunctionReferenceNode(
    Axis to Twist=(bX=True,bY=true,bZ=true)
    Debug=false
    End=(Type=Bone,Name="neck_01")
    Start=(Type=Bone,Name="spine_02")
    ExecuteContext=@RigVMFunction_Sequence.B
    Bones to Twist=@joints_twist_c.Value
   )
   joints_twist_l = RigVMDispatchNode(Value=((Type=Bone,Name="dyn_hair_2_l"),(Type=Bone,Name="dyn_hair_3_l"),(Type=Bone,Name="dyn_hair_4_l"))) {template="DISPATCH_RigVMDispatch_Constant(io Value)"}
   > TwistBones_1 = RigVMFunctionReferenceNode(
    Axis to Twist=(bX=True,bY=true,bZ=true)
    Debug=false
    End=(Type=Bone,Name="neck_01")
    Start=(Type=Bone,Name="spine_02")
    Bones to Twist=@joints_twist_l.Value
   )
   joints_twist_r = RigVMDispatchNode(Value=((Type=Bone,Name="dyn_hair_2_r"),(Type=Bone,Name="dyn_hair_3_r"),(Type=Bone,Name="dyn_hair_4_r"))) {template="DISPATCH_RigVMDispatch_Constant(io Value)"}
   > TwistBones_2 = RigVMFunctionReferenceNode(
    Axis to Twist=(bX=True,bY=true,bZ=true)
    Debug=false
    End=(Type=Bone,Name="neck_01")
    Start=(Type=Bone,Name="spine_02")
    Bones to Twist=@joints_twist_r.Value
   )
   RigVMFunction_MathBoolNand2 = /Script/RigVM.RigVMFunction_MathBoolNand2(
    B=&RigVMFunction_MathBoolNot(Value=RigVMVariableNode() {variable=pub isWearingBackBling}.Value).Result
    A=RigVMVariableNode() {variable=pub isRigidBodies Enabled}.Value
   ) {is=struct RigVMUnitNode, template="Nand::Execute(in A,in B,out Result)"}
   RigVMFunction_ControlFlowBranch = &RigVMFunction_ControlFlowBranch(Condition=@RigVMFunction_MathBoolNand2.Result, ExecuteContext=@TwistBones_2.ExecuteContext)
   controls_ik_c_x7 = RigVMDispatchNode(Value=((Type=Control,Name="dyn_hair_1_c_ctrl"),(Type=Control,Name="dyn_hair_2_c_ctrl"),(Type=Control,Name="dyn_hair_3_c_ctrl"),(Type=Control,Name="dyn_hair_4_c_ctrl"))) {template="DISPATCH_RigVMDispatch_Constant(io Value)"}
   joints_ik_c_f_x7 = RigVMDispatchNode(Value=((Type=Bone,Name="dyn_hair_1_c"),(Type=Bone,Name="dyn_hair_2_c"),(Type=Bone,Name="dyn_hair_3_c"),(Type=Bone,Name="dyn_hair_4_c"))) {template="DISPATCH_RigVMDispatch_Constant(io Value)"}
   SplineIK_6 = RigVMFunctionReferenceNode(
    Primary Axis=(X=1,Y=0,Z=0)
    Samples Per Segment=1
    Secondary Spline Direction=(X=0,Y=0,Z=0)
    Spline Mode=BSpline
    Stretch=False
    Up Axis=(X=0,Y=0,Z=0)
    debug=false
    Controls=@controls_ik_c_x7.Value
    Bones=@joints_ik_c_f_x7.Value
    ExecuteContext=@RigVMFunction_ControlFlowBranch.True
   )
   controls_ik_l_f_x7 = RigVMDispatchNode(Value=((Type=Control,Name="dyn_hair_1_l_ctrl"),(Type=Control,Name="dyn_hair_2_l_ctrl"),(Type=Control,Name="dyn_hair_3_l_ctrl"),(Type=Control,Name="dyn_hair_4_l_ctrl"))) {template="DISPATCH_RigVMDispatch_Constant(io Value)"}
   joints_ik_l_f_x7 = RigVMDispatchNode(Value=((Type=Bone,Name="dyn_hair_1_l"),(Type=Bone,Name="dyn_hair_2_l"),(Type=Bone,Name="dyn_hair_3_l"),(Type=Bone,Name="dyn_hair_4_l"))) {template="DISPATCH_RigVMDispatch_Constant(io Value)"}
   > SplineIK_7 = RigVMFunctionReferenceNode(
    Primary Axis=(X=1,Y=0,Z=0)
    Samples Per Segment=1
    Secondary Spline Direction=(X=0,Y=0,Z=0)
    Spline Mode=BSpline
    Stretch=False
    Up Axis=(X=0,Y=0,Z=0)
    debug=false
    Controls=@controls_ik_l_f_x7.Value
    Bones=@joints_ik_l_f_x7.Value
   )
   controls_ik_r_x7 = RigVMDispatchNode(Value=((Type=Control,Name="dyn_hair_1_r_ctrl"),(Type=Control,Name="dyn_hair_2_r_ctrl"),(Type=Control,Name="dyn_hair_3_r_ctrl"),(Type=Control,Name="dyn_hair_4_r_ctrl"))) {template="DISPATCH_RigVMDispatch_Constant(io Value)"}
   joints_ik_r_f_x7 = RigVMDispatchNode(Value=((Type=Bone,Name="dyn_hair_1_r"),(Type=Bone,Name="dyn_hair_2_r"),(Type=Bone,Name="dyn_hair_3_r"),(Type=Bone,Name="dyn_hair_4_r"))) {template="DISPATCH_RigVMDispatch_Constant(io Value)"}
   > SplineIK_8 = RigVMFunctionReferenceNode(
    Primary Axis=(X=1,Y=0,Z=0)
    Samples Per Segment=1
    Secondary Spline Direction=(X=0,Y=0,Z=0)
    Spline Mode=BSpline
    Stretch=False
    Up Axis=(X=0,Y=0,Z=0)
    debug=false
    Controls=@controls_ik_r_x7.Value
    Bones=@joints_ik_r_f_x7.Value
   )
   controls_ik_c_x4 = RigVMDispatchNode(Value=((Type=Control,Name="dyn_hair_1_c_ctrl"),(Type=Control,Name="dyn_hair_2_c_ctrl"),(Type=Control,Name="dyn_hair_3_c_ctrl"),(Type=Control,Name="dyn_hair_4_c_ctrl"))) {template="DISPATCH_RigVMDispatch_Constant(io Value)"}
   joints_ik_c_x4 = RigVMDispatchNode(Value=((Type=Bone,Name="dyn_hair_1_c"),(Type=Bone,Name="dyn_hair_2_c"),(Type=Bone,Name="dyn_hair_3_c"),(Type=Bone,Name="dyn_hair_4_c"))) {template="DISPATCH_RigVMDispatch_Constant(io Value)"}
   SplineIK = RigVMFunctionReferenceNode(
    Primary Axis=(X=1,Y=0,Z=0)
    Samples Per Segment=1
    Secondary Spline Direction=(X=0,Y=0,Z=0)
    Spline Mode=BSpline
    Stretch=False
    Up Axis=(X=0,Y=0,Z=0)
    debug=false
    Controls=@controls_ik_c_x4.Value
    Bones=@joints_ik_c_x4.Value
    ExecuteContext=@RigVMFunction_ControlFlowBranch.False
   )
   controls_ik_l_x4 = RigVMDispatchNode(Value=((Type=Control,Name="dyn_hair_1_l_ctrl"),(Type=Control,Name="dyn_hair_2_l_ctrl"),(Type=Control,Name="dyn_hair_3_l_ctrl"),(Type=Control,Name="dyn_hair_4_l_ctrl"))) {template="DISPATCH_RigVMDispatch_Constant(io Value)"}
   joints_ik_l_x4 = RigVMDispatchNode(Value=((Type=Bone,Name="dyn_hair_1_l"),(Type=Bone,Name="dyn_hair_2_l"),(Type=Bone,Name="dyn_hair_3_l"),(Type=Bone,Name="dyn_hair_4_l"))) {template="DISPATCH_RigVMDispatch_Constant(io Value)"}
   > SplineIK_1 = RigVMFunctionReferenceNode(
    Primary Axis=(X=1,Y=0,Z=0)
    Samples Per Segment=1
    Secondary Spline Direction=(X=0,Y=0,Z=0)
    Spline Mode=BSpline
    Stretch=False
    Up Axis=(X=0,Y=0,Z=0)
    debug=false
    Controls=@controls_ik_l_x4.Value
    Bones=@joints_ik_l_x4.Value
   )
   controls_ik_r_x4 = RigVMDispatchNode(Value=((Type=Control,Name="dyn_hair_1_r_ctrl"),(Type=Control,Name="dyn_hair_2_r_ctrl"),(Type=Control,Name="dyn_hair_3_r_ctrl"),(Type=Control,Name="dyn_hair_4_r_ctrl"))) {template="DISPATCH_RigVMDispatch_Constant(io Value)"}
   joints_ik_r_x4 = RigVMDispatchNode(Value=((Type=Bone,Name="dyn_hair_1_r"),(Type=Bone,Name="dyn_hair_2_r"),(Type=Bone,Name="dyn_hair_3_r"),(Type=Bone,Name="dyn_hair_4_r"))) {template="DISPATCH_RigVMDispatch_Constant(io Value)"}
   > SplineIK_2 = RigVMFunctionReferenceNode(
    Primary Axis=(X=1,Y=0,Z=0)
    Samples Per Segment=1
    Secondary Spline Direction=(X=0,Y=0,Z=0)
    Spline Mode=BSpline
    Stretch=False
    Up Axis=(X=0,Y=0,Z=0)
    debug=false
    Controls=@controls_ik_r_x4.Value
    Bones=@joints_ik_r_x4.Value
   )
   RigVMFunctionReturnNode(ExecutePin=@RigVMFunction_Sequence.D)
   AlterTransform = RigVMFunctionReferenceNode(
    inheritsTransform=false
    transforms="((Item=(Type=Bone,Name=\"dyn_hair_1_r\"),Transform=(Rotation=(X=0,Y=0,Z=0,W=1),Translation=(X=0,Y=0,Z=0),Scale3D=(X=1,Y=1,Z=1))),(Item=(Type=Bone,Name=\"dyn_hair_1_l\"),Transform=(Rotation=(X=0,Y=0,Z=0,W=1),Translation=(X=2,Y=0,Z=8),Scale3D=(X=1,Y=1,Z=1))),(It ...(truncated)"
    weights=RigVMVariableNode() {variable=head rz n60n80 alpha}.Value
    ExecuteContext=@RigVMFunction_Sequence.C
   )
   > DeltaRotation v4_4 = RigVMFunctionReferenceNode(
    Rotation Order=ZYX
    child=(Type=Bone,Name="head")
    debugRotation=false
    outputMinMax=(X=0,Y=1)
    parent=(Type=Null,Name="head_null")
    premult=1
    rotationMinMax=(X=-30,Y=-45)
    triggerAxis=Z
    twistAxis=(X=1,Y=0,Z=0)
   )
   > AlterTransform_1 = RigVMFunctionReferenceNode(
    inheritsTransform=false
    transforms="((Item=(Type=Bone,Name=\"dyn_hair_1_r\"),Transform=(Rotation=(X=0,Y=0,Z=0,W=1),Translation=(X=1,Y=0,Z=2),Scale3D=(X=1,Y=1,Z=1))),(Item=(Type=Bone,Name=\"dyn_hair_1_l\"),Transform=(Rotation=(X=0,Y=0,Z=0,W=1),Translation=(X=1,Y=0,Z=2),Scale3D=(X=1,Y=1,Z=1))),(It ...(truncated)"
    weights=@"DeltaRotation v4_4".outputWeight
   )
   > n11 = &RigUnit_SphericalPoseReader(
    ActiveRegionScaleFactors=(PositiveWidth=1,NegativeWidth=1,PositiveHeight=1,NegativeHeight=1)
    ActiveRegionSize=0.05
    Debug=(bDrawDebug=false,bDraw2D=False,bDrawLocalAxes=False,DebugScale=25,DebugSegments=20,DebugThickness=0.25)
    DriverAxis=(X=1,Y=0,Z=0)
    DriverItem=(Type=Bone,Name="head")
    FalloffRegionScaleFactors=(PositiveWidth=1,NegativeWidth=1,PositiveHeight=1,NegativeHeight=1)
    FalloffSize=0.2
    FlipHeightScaling=False
    FlipWidthScaling=False
    OptionalParentItem=(Type=Bone,Name="spine_05")
    RotationOffset=(X=0,Y=-62,Z=-89)
   )
   > AlterTransform_2 = RigVMFunctionReferenceNode(
    inheritsTransform=false
    transforms="((Item=(Type=Bone,Name=\"dyn_hair_1_l\"),Transform=(Rotation=(X=0,Y=0,Z=0,W=1),Translation=(X=8,Y=0,Z=0),Scale3D=(X=1,Y=1,Z=1))),(Item=(Type=Bone,Name=\"dyn_hair_1_c\"),Transform=(Rotation=(X=0,Y=0,Z=0,W=1),Translation=(X=8,Y=0,Z=0),Scale3D=(X=1,Y=1,Z=1))),(It ...(truncated)"
    weights=&RigVMFunction_MathDoubleMul(A=@n11.OutputParam, B=RigVMVariableNode() {variable=isTargeting lerp}.Value).Result
   )
   n12 = &RigUnit_SphericalPoseReader(
    ActiveRegionScaleFactors=(PositiveWidth=0.5,NegativeWidth=1,PositiveHeight=1,NegativeHeight=1)
    ActiveRegionSize=0.1
    Debug=(bDrawDebug=false,bDraw2D=False,bDrawLocalAxes=False,DebugScale=25,DebugSegments=20,DebugThickness=0.25)
    DriverAxis=(X=1,Y=0,Z=0)
    DriverItem=(Type=Bone,Name="clavicle_l")
    ExecutePin=()
    FalloffRegionScaleFactors=(PositiveWidth=0.5,NegativeWidth=1,PositiveHeight=1,NegativeHeight=1)
    FalloffSize=0.2
    FlipHeightScaling=False
    FlipWidthScaling=False
    OptionalParentItem=(Type=Bone,Name="None")
    RotationOffset=(X=-20,Y=-50,Z=0)
   )
   > &RigUnit_ModifyTransforms(
    ItemToModify="((Item=(Type=Null,Name=\"dyn_hair_2_l_ctrl_null\"),Transform=(Rotation=(X=0,Y=0,Z=0,W=1),Translation=(X=0,Y=-6.5129,Z=17.4268),Scale3D=(X=1,Y=1,Z=1))),(Item=(Type=Null,Name=\"dyn_hair_3_l_ctrl_null\"),Transform=(Rotation=(X=0,Y=0,Z=0,W=1),Translation=(X=-2.002 ...(truncated)"
    WeightMinimum=0
    Weight=&RigVMFunction_MathDoubleMul(A=@n12.OutputParam, B=RigVMVariableNode() {variable=isPASwing lerp}.Value).Result
   )
   &RigUnit_ModifyTransforms(
    ExecutePin=()
    ItemToModify="((Item=(Type=Null,Name=\"dyn_hair_2_l_ctrl_null\"),Transform=(Rotation=(X=0,Y=0,Z=0,W=1),Translation=(X=2.22,Y=0,Z=8.25),Scale3D=(X=1,Y=1,Z=1))),(Item=(Type=Null,Name=\"dyn_hair_2_r_ctrl_null\"),Transform=(Rotation=(X=0,Y=0,Z=0,W=1),Translation=(X=2.22,Y=0,Z=8 ...(truncated)"
    Weight=1
    WeightMinimum=0
   )
  ]}
 CollapseNode_ContainedGraph {nodes [
   n1 = RigVMFunctionEntryNode()
   &RigUnit_ParentConstraint(
    AdvancedSettings=(RotationOrderForFilter=XZY)
    Child=(Type=Null,Name="dyn_hair_1_c_ctrl_null")
    Filter=(TranslationFilter=(bX=True,bY=True,bZ=True),RotationFilter=(bX=True,bY=True,bZ=True),ScaleFilter=(bX=False,bY=False,bZ=False))
    Parents=((Item=(Type=Bone,Name="neck_02"),Weight=0.0434),(Item=(Type=Bone,Name="head"),Weight=0.9566))
    Weight=1
    ExecutePin=@n1.ExecutePin
   )
   > &RigUnit_ParentConstraint(
    AdvancedSettings=(RotationOrderForFilter=XZY)
    Child=(Type=Null,Name="dyn_hair_2_c_ctrl_null")
    Filter=(TranslationFilter=(bX=True,bY=True,bZ=True),RotationFilter=(bX=True,bY=True,bZ=True),ScaleFilter=(bX=False,bY=False,bZ=False))
    Parents="((Item=(Type=Bone,Name=\"clavicle_r\"),Weight=0.015),(Item=(Type=Bone,Name=\"clavicle_l\"),Weight=0.015),(Item=(Type=Bone,Name=\"spine_04\"),Weight=0.0681),(Item=(Type=Bone,Name=\"head\"),Weight=0.0921),(Item=(Type=Bone,Name=\"neck_02\"),Weight=0.1321),(Item=(Type=B ...(truncated)"
    Weight=1
   )
   > &RigUnit_ParentConstraint(
    AdvancedSettings=(RotationOrderForFilter=XZY)
    Child=(Type=Null,Name="dyn_hair_3_c_ctrl_null")
    Filter=(TranslationFilter=(bX=True,bY=True,bZ=True),RotationFilter=(bX=True,bY=True,bZ=True),ScaleFilter=(bX=False,bY=False,bZ=False))
    Parents=((Item=(Type=Bone,Name="clavicle_l"),Weight=0.0369),(Item=(Type=Bone,Name="clavicle_r"),Weight=0.0369),(Item=(Type=Bone,Name="spine_05"),Weight=0.9261))
    Weight=1
   )
   > &RigUnit_ParentConstraint(
    AdvancedSettings=(RotationOrderForFilter=XZY)
    Child=(Type=Null,Name="dyn_hair_4_c_ctrl_null")
    Filter=(TranslationFilter=(bX=True,bY=True,bZ=True),RotationFilter=(bX=True,bY=True,bZ=True),ScaleFilter=(bX=False,bY=False,bZ=False))
    Parents=((Item=(Type=Bone,Name="spine_04"),Weight=0.0415),(Item=(Type=Bone,Name="clavicle_l"),Weight=0.0758),(Item=(Type=Bone,Name="clavicle_r"),Weight=0.0758),(Item=(Type=Bone,Name="spine_05"),Weight=0.807))
    Weight=1
   )
   > &RigUnit_ParentConstraint(
    AdvancedSettings=(RotationOrderForFilter=XZY)
    Child=(Type=Null,Name="dyn_hair_1_l_ctrl_null")
    Filter=(TranslationFilter=(bX=True,bY=True,bZ=True),RotationFilter=(bX=True,bY=True,bZ=True),ScaleFilter=(bX=False,bY=False,bZ=False))
    Parents=((Item=(Type=Bone,Name="head"),Weight=1.0))
    Weight=1
   )
   > &RigUnit_ParentConstraint(
    AdvancedSettings=(RotationOrderForFilter=XZY)
    Child=(Type=Null,Name="dyn_hair_2_l_ctrl_null")
    Filter=(TranslationFilter=(bX=True,bY=True,bZ=True),RotationFilter=(bX=True,bY=True,bZ=True),ScaleFilter=(bX=False,bY=False,bZ=False))
    Parents="((Item=(Type=Bone,Name=\"clavicle_l\"),Weight=0.0503),(Item=(Type=Bone,Name=\"spine_04\"),Weight=0.0613),(Item=(Type=Bone,Name=\"head\"),Weight=0.1054),(Item=(Type=Bone,Name=\"neck_02\"),Weight=0.1567),(Item=(Type=Bone,Name=\"spine_05\"),Weight=0.2205),(Item=(Type=B ...(truncated)"
    Weight=1
   )
   > &RigUnit_ParentConstraint(
    AdvancedSettings=(RotationOrderForFilter=XZY)
    Child=(Type=Null,Name="dyn_hair_3_l_ctrl_null")
    Filter=(TranslationFilter=(bX=True,bY=True,bZ=True),RotationFilter=(bX=True,bY=True,bZ=True),ScaleFilter=(bX=False,bY=False,bZ=False))
    Parents=((Item=(Type=Bone,Name="clavicle_l"),Weight=0.4559),(Item=(Type=Bone,Name="spine_05"),Weight=0.5441))
    Weight=1
   )
   > &RigUnit_ParentConstraint(
    AdvancedSettings=(RotationOrderForFilter=XZY)
    Child=(Type=Null,Name="dyn_hair_4_l_ctrl_null")
    Filter=(TranslationFilter=(bX=True,bY=True,bZ=True),RotationFilter=(bX=True,bY=True,bZ=True),ScaleFilter=(bX=False,bY=False,bZ=False))
    Parents=((Item=(Type=Bone,Name="upperarm_l"),Weight=0.0282),(Item=(Type=Bone,Name="spine_04"),Weight=0.0451),(Item=(Type=Bone,Name="spine_05"),Weight=0.377),(Item=(Type=Bone,Name="clavicle_l"),Weight=0.5498))
    Weight=1
   )
   > &RigUnit_ParentConstraint(
    AdvancedSettings=(RotationOrderForFilter=XZY)
    Child=(Type=Null,Name="dyn_hair_1_r_ctrl_null")
    Filter=(TranslationFilter=(bX=True,bY=True,bZ=True),RotationFilter=(bX=True,bY=True,bZ=True),ScaleFilter=(bX=False,bY=False,bZ=False))
    Parents=((Item=(Type=Bone,Name="head"),Weight=1.0))
    Weight=1
   )
   > &RigUnit_ParentConstraint(
    AdvancedSettings=(RotationOrderForFilter=XZY)
    Child=(Type=Null,Name="dyn_hair_2_r_ctrl_null")
    Filter=(TranslationFilter=(bX=True,bY=True,bZ=True),RotationFilter=(bX=True,bY=True,bZ=True),ScaleFilter=(bX=False,bY=False,bZ=False))
    Parents="((Item=(Type=Bone,Name=\"clavicle_r\"),Weight=0.0503),(Item=(Type=Bone,Name=\"spine_04\"),Weight=0.0613),(Item=(Type=Bone,Name=\"head\"),Weight=0.1054),(Item=(Type=Bone,Name=\"neck_02\"),Weight=0.1567),(Item=(Type=Bone,Name=\"spine_05\"),Weight=0.2204),(Item=(Type=B ...(truncated)"
    Weight=1
   )
   > &RigUnit_ParentConstraint(
    AdvancedSettings=(RotationOrderForFilter=XZY)
    Child=(Type=Null,Name="dyn_hair_3_r_ctrl_null")
    Filter=(TranslationFilter=(bX=True,bY=True,bZ=True),RotationFilter=(bX=True,bY=True,bZ=True),ScaleFilter=(bX=False,bY=False,bZ=False))
    Parents=((Item=(Type=Bone,Name="clavicle_r"),Weight=0.4559),(Item=(Type=Bone,Name="spine_05"),Weight=0.5441))
    Weight=1
   )
   > n2 = &RigUnit_ParentConstraint(
    AdvancedSettings=(RotationOrderForFilter=XZY)
    Child=(Type=Null,Name="dyn_hair_4_r_ctrl_null")
    Filter=(TranslationFilter=(bX=True,bY=True,bZ=True),RotationFilter=(bX=True,bY=True,bZ=True),ScaleFilter=(bX=False,bY=False,bZ=False))
    Parents=((Item=(Type=Bone,Name="upperarm_r"),Weight=0.0282),(Item=(Type=Bone,Name="spine_04"),Weight=0.0451),(Item=(Type=Bone,Name="spine_05"),Weight=0.377),(Item=(Type=Bone,Name="clavicle_r"),Weight=0.5498))
    Weight=1
   )
   RigVMFunctionReturnNode(ExecutePin=@n2.ExecutePin)
  ]}
 CollapseNode_ContainedGraph {nodes [
   RigVMFunctionEntryNode()
   n1 = &RigVMFunction_AlphaInterp(
    InterpSpeedDecreasing=15
    InterpSpeedIncreasing=15
    Scale=1
    bClampResult=False
    Value=&RigVMFunction_MathBoolToFloat(Value=RigVMVariableNode() {variable=pub isTargeting}.Value).Result
   )
   > RigVMVariableNode(Value=@n1.Result) {variable=isTargeting lerp}
   RigVMFunction_MathBoolOr = RigVMAggregateNode(
    A=RigVMVariableNode() {variable=pub isAR Equipped}.Value
    B=RigVMVariableNode() {variable=pub isSMG90}.Value
    C=RigVMVariableNode() {variable=pub isOutLaw}.Value
   )
   n2 = &RigVMFunction_AlphaInterp(
    InterpSpeedDecreasing=15
    InterpSpeedIncreasing=15
    Scale=1
    bClampResult=False
    Value=&RigVMFunction_MathBoolToFloat(Value=@RigVMFunction_MathBoolOr.Result).Result
   )
   > RigVMVariableNode(Value=@n2.Result) {variable=isAR Equipped lerp}
   n3 = &RigVMFunction_AlphaInterp(
    InterpSpeedDecreasing=15
    InterpSpeedIncreasing=15
    Scale=1
    bClampResult=False
    Value=&RigVMFunction_MathBoolToFloat(Value=RigVMVariableNode() {variable=pub isPistol Equipped}.Value).Result
   )
   > RigVMVariableNode(Value=@n3.Result) {variable=isPistol Equipped lerp}
   n4 = &RigVMFunction_AlphaInterp(
    InterpSpeedDecreasing=15
    InterpSpeedIncreasing=15
    Scale=1
    bClampResult=False
    Value=&RigVMFunction_MathBoolToFloat(Value=RigVMVariableNode() {variable=pub PA Swing}.Value).Result
   )
   > RigVMVariableNode(Value=@n4.Result) {variable=isPASwing lerp}
   > RigVMFunctionReturnNode()
  ]}
 Multiply_4_ContainedGraph {nodes [
   n1 = RigVMFunctionEntryNode()
   n2 = &RigVMFunction_MathDoubleMul(A=@n1.A, B=@n1.B)
   RigVMFunctionReturnNode(Result=&RigVMFunction_MathDoubleMul(B=@n1.D, A=&RigVMFunction_MathDoubleMul(B=@n1.C, A=@n2.Result).Result).Result)
  ]}
 New Function_ContainedGraph {nodes [
   n1 = RigVMFunctionEntryNode()
   RigVMFunction_MathBoolToInteger = &RigVMFunction_MathBoolToInteger(Value=@n1.onlyTwist)
   RigVMFunction_MathQuaternionSwingTwist = &RigVMFunction_MathQuaternionSwingTwist(
    TwistAxis=@n1.twistAxis
    Input=&RigUnit_GetRelativeTransformForItem(bChildInitial=False, bParentInitial=False, Child=@n1.child, Parent=@n1.parent).RelativeTransform.Rotation
   )
   n2 = RigVMDispatchNode(
    Values=((),())
    Index=@RigVMFunction_MathBoolToInteger.Result
    Values.0=@RigVMFunction_MathQuaternionSwingTwist.Swing
    Values.1=@RigVMFunction_MathQuaternionSwingTwist.Twist
   ) {template="DISPATCH_RigVMDispatch_SelectInt32(in Index,in Values,out Result)"}
   RigVMFunction_MathQuaternionToEuler = &RigVMFunction_MathQuaternionToEuler(Value=@n2.Result, RotationOrder=@n1.Rotation Order)
   > DISPATCH_RigVMDispatch_Print = RigVMDispatchNode(ScreenColor=(R=1,G=1,B=1,A=1), ScreenDuration=1.5, Enabled=@n1.debugRotation, Value=@RigVMFunction_MathQuaternionToEuler.Result) {template="DISPATCH_RigVMDispatch_Print(in Prefix,in Value,in Enabled,in ScreenDuration,in ScreenColor)"}
   Cast = RigVMDispatchNode(Value=@n1.triggerAxis) {template="DISPATCH_RigVMDispatch_CastEnumToInt(in Value,out Result)"}
   DISPATCH_RigVMDispatch_SelectInt32 = RigVMDispatchNode(
    Values=(0,0,0)
    Index=@Cast.Result
    Values.0=@RigVMFunction_MathQuaternionToEuler.Result.X
    Values.1=@RigVMFunction_MathQuaternionToEuler.Result.Y
    Values.2=@RigVMFunction_MathQuaternionToEuler.Result.Z
   ) {template="DISPATCH_RigVMDispatch_SelectInt32(in Index,in Values,out Result)"}
   Remap = &RigVMFunction_MathDoubleRemap(
    bClamp=False
    Value=@DISPATCH_RigVMDispatch_SelectInt32.Result
    TargetMinimum=@n1.outputMinMax.X
    TargetMaximum=@n1.outputMinMax.Y
    SourceMinimum=@n1.rotationMinMax.X
    SourceMaximum=@n1.rotationMinMax.Y
   )
   Absolute = &RigVMFunction_MathDoubleAbs(Value=@Remap.Result)
   Multiply = &RigVMFunction_MathDoubleMul(
    A=&RigVMFunction_MathDoubleClamp(Value=RigVMDispatchNode(True=@Absolute.Result, Condition=@n1.absoluteValue, False=@Remap.Result) {template="DISPATCH_RigVMDispatch_If(in Condition,in True,in False,out Result)"}.Result).Result
    B=@n1.premult
   )
   > RigVMFunctionReturnNode(outputQuat=@n2.Result, outputWeight=@Multiply.Result)
  ]}
 New Function_ContainedGraph {nodes [
   n1 = RigVMFunctionEntryNode()
   > For_Each = RigVMDispatchNode(Array=@n1.transforms) {template="DISPATCH_RigVMDispatch_ArrayIterator(in Array,out Element,out Index,out Count,out Ratio)"}
   > &RigUnit_OffsetTransformForItem(Item=@For_Each.Element.Item, OffsetTransform=@For_Each.Element.Transform, Weight=@n1.weights, bPropagateToChildren=@n1.inheritsTransform)
   RigVMFunctionReturnNode(ExecuteContext=@For_Each.Completed)
  ]}
 New Function_ContainedGraph {nodes [
   n1 = RigVMFunctionEntryNode()
   RigVMFunction_Sequence = &RigVMFunction_Sequence(ExecuteContext=@n1.ExecuteContext)
   &RigUnit_ParentConstraint(
    AdvancedSettings=(RotationOrderForFilter=XZY)
    Child=(Type=Bone,Name="dyn_front_1_l")
    Parents=((Item=(Type=Bone,Name="head"),Weight=0.1191),(Item=(Type=Bone,Name="neck_01"),Weight=0.1657),(Item=(Type=Bone,Name="neck_02"),Weight=0.2707),(Item=(Type=Bone,Name="spine_05"),Weight=0.4444))
    Weight=1
    ExecutePin=@RigVMFunction_Sequence.A
   )
   > &RigUnit_ParentConstraint(
    AdvancedSettings=(RotationOrderForFilter=XZY)
    Child=(Type=Bone,Name="dyn_front_2_l")
    Parents=((Item=(Type=Bone,Name="head"),Weight=0.1191),(Item=(Type=Bone,Name="neck_01"),Weight=0.1657),(Item=(Type=Bone,Name="neck_02"),Weight=0.2707),(Item=(Type=Bone,Name="spine_05"),Weight=0.4444))
    Weight=1
   )
   > &RigUnit_ParentConstraint(
    AdvancedSettings=(RotationOrderForFilter=XZY)
    Child=(Type=Bone,Name="dyn_front_3_l")
    Parents=((Item=(Type=Bone,Name="neck_02"),Weight=0.0003),(Item=(Type=Bone,Name="spine_04"),Weight=0.013),(Item=(Type=Bone,Name="clavicle_l"),Weight=0.2303),(Item=(Type=Bone,Name="spine_05"),Weight=0.7563))
    Weight=1
   )
   > &RigUnit_ParentConstraint(
    AdvancedSettings=(RotationOrderForFilter=XZY)
    Child=(Type=Bone,Name="dyn_front_4_l")
    Parents=((Item=(Type=Bone,Name="upperarm_twist_01_l"),Weight=0.0222),(Item=(Type=Bone,Name="spine_04"),Weight=0.255),(Item=(Type=Bone,Name="spine_05"),Weight=0.314),(Item=(Type=Bone,Name="clavicle_l"),Weight=0.4088))
    Weight=1
   )
   > &RigUnit_ParentConstraint(
    AdvancedSettings=(RotationOrderForFilter=XZY)
    Child=(Type=Bone,Name="dyn_front_5_l")
    Parents=((Item=(Type=Bone,Name="upperarm_twist_01_l"),Weight=0.0333),(Item=(Type=Bone,Name="spine_05"),Weight=0.2881),(Item=(Type=Bone,Name="spine_04"),Weight=0.3088),(Item=(Type=Bone,Name="clavicle_l"),Weight=0.3699))
    Weight=1
   )
   > &RigUnit_ParentConstraint(
    AdvancedSettings=(RotationOrderForFilter=XZY)
    Child=(Type=Bone,Name="dyn_front_6_l")
    Parents=((Item=(Type=Bone,Name="clavicle_l"),Weight=0.0067),(Item=(Type=Bone,Name="spine_05"),Weight=0.1118),(Item=(Type=Bone,Name="spine_04"),Weight=0.8815))
    Weight=1
   )
   > &RigUnit_ParentConstraint(
    AdvancedSettings=(RotationOrderForFilter=XZY)
    Child=(Type=Bone,Name="dyn_front_7_l")
    Parents=((Item=(Type=Bone,Name="spine_04"),Weight=1.0))
    Weight=1
   )
   > &RigUnit_ParentConstraint(
    AdvancedSettings=(RotationOrderForFilter=XZY)
    Child=(Type=Bone,Name="base_offset_hair_r")
    Parents=((Item=(Type=Bone,Name="dyn_hair_3_r"),Weight=0.0273),(Item=(Type=Bone,Name="dyn_hair_2_r"),Weight=0.1799),(Item=(Type=Bone,Name="head"),Weight=0.3088),(Item=(Type=Bone,Name="dyn_hair_1_r"),Weight=0.484))
    Weight=1
   )
   RigVMFunctionReturnNode(ExecuteContext=@RigVMFunction_Sequence.B)
  ]}
 New Function_ContainedGraph {nodes [
   n1 = RigVMFunctionEntryNode()
   RigVMFunction_Sequence = &RigVMFunction_Sequence(ExecuteContext=@n1.ExecuteContext)
   DeltaRotation v4 = RigVMFunctionReferenceNode(
    Rotation Order=ZYX
    child=(Type=Bone,Name="head")
    debugRotation=false
    onlyTwist=true
    outputMinMax=(X=0,Y=1)
    parent=(Type=Null,Name="head_null")
    premult=1
    rotationMinMax=(X=0,Y=-50)
    twistAxis=(X=1,Y=0,Z=0)
    ExecuteContext=@RigVMFunction_Sequence.A
   )
   AlphaInterp = &RigVMFunction_AlphaInterp(InterpSpeedDecreasing=5, InterpSpeedIncreasing=5, Scale=1, bClampResult=False, Value=@"DeltaRotation v4".outputWeight)
   > &RigUnit_ModifyTransforms(
    ItemToModify="((Item=(Type=Bone,Name=\"dyn_front_1_l\"),Transform=(Rotation=(X=0,Y=0,Z=0,W=-1),Translation=(X=0.5,Y=0,Z=2),Scale3D=(X=1,Y=1,Z=1))),(Item=(Type=Bone,Name=\"dyn_front_2_l\"),Transform=(Rotation=(X=0,Y=0,Z=0,W=1),Translation=(X=0,Y=0,Z=3),Scale3D=(X=1,Y=1,Z=1)) ...(truncated)"
    WeightMinimum=0
    Weight=@AlphaInterp.Result
   )
   > DeltaRotation v4_1 = RigVMFunctionReferenceNode(
    Rotation Order=ZYX
    child=(Type=Bone,Name="upperarm_l")
    debugRotation=false
    outputMinMax=(X=0,Y=1)
    parent=(Type=Null,Name="upperarm_l_null")
    premult=1
    rotationMinMax=(X=0,Y=-28)
    triggerAxis=Y
    twistAxis=(X=1,Y=0,Z=0)
   )
   n2 = &RigVMFunction_MathBoolToFloat(Value=RigVMVariableNode() {variable=pub isParachuteOpened}.Value)
   Multiply = &RigVMFunction_MathDoubleMul(A=@"DeltaRotation v4_1".outputWeight, B=@n2.Result)
   > &RigUnit_ModifyTransforms(
    ItemToModify=((Item=(Type=Bone,Name="dyn_hair_2_l"),Transform=(Rotation=(X=0,Y=-0.1348,Z=0,W=-0.9909),Translation=(X=0.2956,Y=0.0001,Z=3.5287),Scale3D=(X=1,Y=1,Z=1))))
    WeightMinimum=0
    Weight=@Multiply.Result
   )
   > DeltaRotation v4_2 = RigVMFunctionReferenceNode(
    Rotation Order=ZYX
    child=(Type=Bone,Name="upperarm_r")
    debugRotation=false
    outputMinMax=(X=0,Y=1)
    parent=(Type=Null,Name="upperarm_r_null")
    premult=1
    rotationMinMax=(X=0,Y=-44)
    triggerAxis=Y
    twistAxis=(X=1,Y=0,Z=0)
   )
   > &RigUnit_ModifyTransforms(
    ItemToModify="((Item=(Type=Bone,Name=\"dyn_hair_2_r\"),Transform=(Rotation=(X=0,Y=0.0621,Z=0,W=0.9981),Translation=(X=-3.3132,Y=-0.0013,Z=2.6761),Scale3D=(X=1,Y=1,Z=1))),(Item=(Type=Bone,Name=\"dyn_offset_hair_r\"),Transform=(Rotation=(X=0,Y=0,Z=0,W=1),Translation=(X=0,Y=-5 ...(truncated)"
    WeightMinimum=0
    Weight=&RigVMFunction_MathDoubleMul(A=@"DeltaRotation v4_2".outputWeight, B=@n2.Result).Result
   )
   > DeltaRotation v4_3 = RigVMFunctionReferenceNode(
    Rotation Order=ZYX
    child=(Type=Bone,Name="head")
    debugRotation=false
    outputMinMax=(X=0,Y=1)
    parent=(Type=Null,Name="head_null")
    premult=1
    rotationMinMax=(X=0,Y=8)
    triggerAxis=Z
    twistAxis=(X=1,Y=0,Z=0)
   )
   > &RigUnit_ModifyTransforms(
    ItemToModify=((Item=(Type=Bone,Name="dyn_hair_4_c"),Transform=(Rotation=(X=0,Y=-0.1736,Z=0,W=0.9848),Translation=(X=0,Y=0,Z=0),Scale3D=(X=1,Y=1,Z=1))))
    WeightMinimum=0
    Weight=@"DeltaRotation v4_3".outputWeight
   )
   > DeltaRotation v4_4 = RigVMFunctionReferenceNode(
    Rotation Order=ZYX
    child=(Type=Bone,Name="upperarm_l")
    debugRotation=false
    outputMinMax=(X=0,Y=1)
    parent=(Type=Null,Name="upperarm_l_null")
    premult=1
    rotationMinMax=(X=0,Y=54)
    triggerAxis=Z
    twistAxis=(X=1,Y=0,Z=0)
   )
   n3 = RigVMVariableNode() {variable=pub isPistol Equipped}
   > &RigUnit_ModifyTransforms(
    ItemToModify="((Item=(Type=Bone,Name=\"dyn_front_3_l\"),Transform=(Rotation=(X=-0.0076,Y=0.0868,Z=0.0868,W=0.9924),Translation=(X=0,Y=0,Z=0),Scale3D=(X=1,Y=1,Z=1))),(Item=(Type=Bone,Name=\"dyn_front_5_l\"),Transform=(Rotation=(X=-0.0226,Y=0.0842,Z=0.2578,W=0.9623),Translati ...(truncated)"
    WeightMinimum=0
    Weight=&RigVMFunction_MathDoubleMul(A=@"DeltaRotation v4_4".outputWeight, B=&RigVMFunction_MathBoolToFloat(Value=@n3.Value).Result).Result
   )
   > DeltaRotation v4_5 = RigVMFunctionReferenceNode(
    Rotation Order=ZYX
    child=(Type=Bone,Name="head")
    debugRotation=false
    outputMinMax=(X=0,Y=1)
    parent=(Type=Null,Name="head_null")
    premult=1
    rotationMinMax=(X=0,Y=-25)
    triggerAxis=Z
    twistAxis=(X=1,Y=0,Z=0)
   )
   Multiply_3 = RigVMAggregateNode(
    A=@"DeltaRotation v4_5".outputWeight
    D=RigVMDispatchNode(False=1, True=0, Condition=RigVMVariableNode() {variable=pub isPistol Equipped}.Value) {template="DISPATCH_RigVMDispatch_If(in Condition,in True,in False,out Result)"}.Result
    B=&RigVMFunction_MathBoolToFloat(Value=RigVMVariableNode() {variable=pub isTargeting}.Value).Result
    C=&RigVMFunction_MathBoolToFloat(Value=RigVMVariableNode() {variable=pub isCrouching}.Value).Result
   )
   n4 = &RigVMFunction_AlphaInterp(InterpSpeedDecreasing=5, InterpSpeedIncreasing=5, Scale=1, bClampResult=False, Value=@Multiply_3.Result)
   > &RigUnit_ModifyTransforms(
    ItemToModify="((Item=(Type=Bone,Name=\"dyn_hair_1_r\"),Transform=(Rotation=(X=-0.1092,Y=-0.0128,Z=-0.1155,W=-0.9872),Translation=(X=3.3529,Y=-2.9752,Z=4.4195),Scale3D=(X=1,Y=1,Z=1))),(Item=(Type=Bone,Name=\"dyn_hair_2_r\"),Transform=(Rotation=(X=0,Y=0.3175,Z=0,W=0.9483),Tra ...(truncated)"
    WeightMinimum=0
    Weight=@n4.Result
   )
   RigVMFunction_MathBoolToFloat = &RigVMFunction_MathBoolToFloat(Value=RigVMVariableNode() {variable=pub isDBNO}.Value)
   n5 = &RigVMFunction_AlphaInterp(InterpSpeedDecreasing=5, InterpSpeedIncreasing=5, Scale=1, bClampResult=False, Value=@RigVMFunction_MathBoolToFloat.Result)
   > &RigUnit_ModifyTransforms(
    ItemToModify=((Item=(Type=Bone,Name="dyn_front_3_l"),Transform=(Rotation=(X=-0.0996,Y=0.1422,Z=0.5649,W=0.8067),Translation=(X=0,Y=0,Z=0),Scale3D=(X=1,Y=1,Z=1))))
    WeightMinimum=0
    Weight=@n5.Result
   )
   > DeltaRotation v4_6 = RigVMFunctionReferenceNode(
    Rotation Order=ZYX
    child=(Type=Bone,Name="clavicle_l")
    debugRotation=false
    outputMinMax=(X=0,Y=1)
    parent=(Type=Null,Name="clavicle_l_null")
    premult=1
    rotationMinMax=(X=0,Y=-20)
    triggerAxis=Z
    twistAxis=(X=1,Y=0,Z=0)
   )
   n6 = RigVMVariableNode() {variable=isPASwing lerp}
   n7 = &RigVMFunction_AlphaInterp(
    InterpSpeedDecreasing=5
    InterpSpeedIncreasing=5
    Scale=1
    bClampResult=False
    Value=&RigVMFunction_MathDoubleMul(A=@"DeltaRotation v4_6".outputWeight, B=@n6.Value).Result
   )
   > &RigUnit_ModifyTransforms(
    ItemToModify="((Item=(Type=Bone,Name=\"dyn_hair_2_l\"),Transform=(Rotation=(X=0.0331,Y=-0.7152,Z=-0.0677,W=0.6948),Translation=(X=-0.0676,Y=0.2044,Z=-5.0573),Scale3D=(X=1,Y=1,Z=1))),(Item=(Type=Bone,Name=\"dyn_hair_3_l\"),Transform=(Rotation=(X=0,Y=0.3518,Z=0,W=0.9361),Tran ...(truncated)"
    WeightMinimum=0
    Weight=@n7.Result
   )
   > DeltaRotation v4_7 = RigVMFunctionReferenceNode(
    Rotation Order=ZYX
    child=(Type=Bone,Name="clavicle_l")
    debugRotation=false
    outputMinMax=(X=0,Y=1)
    parent=(Type=Null,Name="clavicle_l_null")
    premult=1
    rotationMinMax=(X=0,Y=10)
    triggerAxis=Z
    twistAxis=(X=1,Y=0,Z=0)
   )
   n8 = &RigVMFunction_AlphaInterp(
    InterpSpeedDecreasing=5
    InterpSpeedIncreasing=5
    Scale=1
    bClampResult=False
    Value=&RigVMFunction_MathDoubleMul(A=@"DeltaRotation v4_7".outputWeight, B=@n6.Value).Result
   )
   > &RigUnit_ModifyTransforms(
    ItemToModify="((Item=(Type=Bone,Name=\"dyn_hair_2_l\"),Transform=(Rotation=(X=0,Y=0,Z=0,W=1),Translation=(X=0,Y=0,Z=0),Scale3D=(X=1,Y=1,Z=1))),(Item=(Type=Bone,Name=\"dyn_hair_3_l\"),Transform=(Rotation=(X=0,Y=-0.2357,Z=0,W=0.9718),Translation=(X=0,Y=0,Z=0),Scale3D=(X=1,Y=1 ...(truncated)"
    WeightMinimum=0
    Weight=@n8.Result
   )
   > DeltaRotation v4_8 = RigVMFunctionReferenceNode(
    Rotation Order=ZYX
    child=(Type=Bone,Name="head")
    debugRotation=false
    outputMinMax=(X=0,Y=1)
    parent=(Type=Null,Name="head_null")
    premult=1
    rotationMinMax=(X=0,Y=5)
    triggerAxis=Y
    twistAxis=(X=1,Y=0,Z=0)
   )
   > &RigUnit_ModifyTransforms(
    ItemToModify=((Item=(Type=Bone,Name="dyn_earing_01_r"),Transform=(Rotation=(X=-0.0114,Y=-0.13,Z=-0.0864,W=0.9877),Translation=(X=0,Y=0,Z=0),Scale3D=(X=1,Y=1,Z=1))))
    WeightMinimum=0
    Weight=@"DeltaRotation v4_8".outputWeight
   )
   RigVMFunctionReturnNode(ExecuteContext=@RigVMFunction_Sequence.B)
  ]}
 RigVMFunctionLibrary {nodes [
   AlterTransform = RigVMCollapseNode(ExecuteContext=(), inheritsTransform=false, transforms=(), weights=1)
   Constraints = RigVMCollapseNode(ExecuteContext=())
   DeltaRotation v4 = RigVMCollapseNode(
    ExecuteContext=()
    Rotation Order=ZYX
    child=(Type=Bone,Name="None")
    outputMinMax=(X=0,Y=1)
    parent=(Type=Bone,Name="None")
    premult=1
    rotationMinMax=(X=0,Y=0)
    twistAxis=(X=1,Y=0,Z=0)
   )
   Extra_Fixes = RigVMCollapseNode(ExecuteContext=())
  ]}
 RigVMFunction_MathBoolOr_1_ContainedGraph {nodes [
   n1 = RigVMFunctionEntryNode()
   RigVMFunction_MathBoolOr = &RigVMFunction_MathBoolOr(A=@n1.A, B=@n1.B)
   RigVMFunctionReturnNode(Result=&RigVMFunction_MathBoolOr(B=@n1.C, A=@RigVMFunction_MathBoolOr.Result).Result)
  ]}
 RigVMFunction_Sequence_1_ContainedGraph {nodes [
   n1 = RigVMFunctionEntryNode()
   RigVMFunction_Sequence = &RigVMFunction_Sequence(ExecuteContext=@n1.ExecuteContext)
   n2 = &RigVMFunction_Sequence(ExecuteContext=@RigVMFunction_Sequence.B)
   n3 = &RigVMFunction_Sequence(ExecuteContext=@n2.B)
   RigVMFunctionReturnNode(A=@RigVMFunction_Sequence.A, B=@n2.A, C=@n3.A, D=@n3.B)
  ]}
 RigVMFunction_Sequence_2_ContainedGraph {nodes [
   n1 = RigVMFunctionEntryNode()
   n2 = &RigVMFunction_Sequence(ExecuteContext=@n1.ExecuteContext)
   n3 = &RigVMFunction_Sequence(ExecuteContext=@n2.B)
   RigVMFunctionReturnNode(A=@n2.A, B=@n3.A, C=@n3.B)
  ]}
 RigVMModel {nodes [
   RigUnit_BeginExecution = &RigUnit_BeginExecution()
   hair = RigVMCollapseNode(ExecutePin=@RigUnit_BeginExecution.ExecutePin)
   Constraints = RigVMFunctionReferenceNode(ExecuteContext=@hair.ExecutePin)
   > Extra_Fixes = RigVMFunctionReferenceNode()
  ]}
 RigVMModel Rig {nodes [
   n1 = &RigUnit_PrepareForExecution()
   ItemArray = RigVMDispatchNode(Value=((Type=Bone,Name="head"))) {template="DISPATCH_RigVMDispatch_Constant(io Value)"}
   DISPATCH_RigVMDispatch_ArrayIterator = RigVMDispatchNode(Array=@ItemArray.Value, ExecuteContext=@n1.ExecutePin) {template="DISPATCH_RigVMDispatch_ArrayIterator(in Array,out Element,out Index,out Count,out Ratio)"}
   Concat = /Script/RigVM.RigVMFunction_NameConcat(B=_null, A=@DISPATCH_RigVMDispatch_ArrayIterator.Element.Name) {is=struct RigVMUnitNode, template="Concat::Execute(in A,in B,out Result)"}
   > &RigUnit_HierarchyAddNull(
    Space=LocalSpace
    Name=@Concat.Result
    Transform=&RigUnit_GetTransform(Item=(Type=Bone,Name="clavicle_l"), Space=LocalSpace, bInitial=true, Item.Name=@DISPATCH_RigVMDispatch_ArrayIterator.Element.Name).Transform
    Parent=/Script/ControlRig.RigUnit_HierarchyGetParent(bDefaultParent=True, Child=@DISPATCH_RigVMDispatch_ArrayIterator.Element) {is=struct RigVMUnitNode}.Parent
   )
   > &RigUnit_HierarchyAddNull(
    Name=head_sp_null
    Parent=(Type=Bone,Name="spine_05")
    Space=LocalSpace
    Transform=&RigUnit_GetRelativeTransformForItem(Child=(Type=Bone,Name="head"), Parent=(Type=Bone,Name="spine_05"), bChildInitial=true, bParentInitial=true).RelativeTransform
   )
  ]}
]
hierarchy [
 dyn_hair_4_r_ctrl_null {type=Null, loc=<-5.6 -15.1378 132.962>, rot=<-0.5109 -0.5109 0.4888 -0.4888>}
 dyn_hair_4_r_ctrl {
  controlType=EULER_TRANSFORM
  parent=dyn_hair_4_r_ctrl_null
  type=Control
  loc=<-63.6076 -15.1378 117.4697>
  rot=<0.5978 0.4037 -0.3777 0.5805>
 }
 dyn_hair_3_r_ctrl_null {type=Null, loc=<-5.6 -12.601 139.795>, rot=<-0.4037 -0.4037 0.5805 -0.5805>}
 dyn_hair_3_r_ctrl {controlType=EULER_TRANSFORM, parent=dyn_hair_3_r_ctrl_null, type=Control, loc=<-26.1997 -12.601 137.9241>, rot=<0.431 0.3755 -0.5606 0.5992>}
 dyn_hair_2_r_ctrl_null {type=Null, loc=<-5.6 -9.4102 146.382>, rot=<-0.3755 -0.3755 0.5992 -0.5992>}
 dyn_hair_2_r_ctrl {controlType=EULER_TRANSFORM, parent=dyn_hair_2_r_ctrl_null, type=Control, loc=<-1.2049 -9.4102 146.9463>, rot=<0.3513 0.3991 -0.6137 0.5837>}
 dyn_hair_1_r_ctrl_null {type=Null, loc=<-5.6 -6.7546 153.204>, rot=<0.3991 0.3991 -0.5837 0.5837>}
 dyn_hair_1_r_ctrl {controlType=EULER_TRANSFORM, parent=dyn_hair_1_r_ctrl_null, type=Control, loc=<0 0 0>, rot=<0 0 0 1>}
 dyn_hair_4_l_ctrl_null {type=Null, loc=<5.6 -15.1378 132.962>, rot=<-0.5109 -0.5109 0.4888 -0.4888>}
 dyn_hair_4_l_ctrl {
  controlType=EULER_TRANSFORM
  parent=dyn_hair_4_l_ctrl_null
  type=Control
  loc=<-53.2898 -15.1378 121.8274>
  rot=<0.5978 0.4037 -0.3777 0.5805>
 }
 dyn_hair_3_l_ctrl_null {type=Null, loc=<5.6 -12.601 139.7946>, rot=<-0.4037 -0.4037 0.5805 -0.5805>}
 dyn_hair_3_l_ctrl {controlType=EULER_TRANSFORM, parent=dyn_hair_3_l_ctrl_null, type=Control, loc=<-15.0513 -12.601 138.9927>, rot=<0.431 0.3755 -0.5606 0.5992>}
 dyn_hair_2_l_ctrl_null {type=Null, loc=<5.6 -9.4102 146.3822>, rot=<-0.3755 -0.3755 0.5992 -0.5992>}
 dyn_hair_2_l_ctrl {controlType=EULER_TRANSFORM, parent=dyn_hair_2_l_ctrl_null, type=Control, loc=<9.9597 -9.4102 146.0531>, rot=<0.3513 0.3991 -0.6137 0.5837>}
 dyn_hair_1_l_ctrl_null {type=Null, loc=<5.6 -6.7546 153.204>, rot=<0.3991 0.3991 -0.5837 0.5837>}
 dyn_hair_1_l_ctrl {controlType=EULER_TRANSFORM, parent=dyn_hair_1_l_ctrl_null, type=Control, loc=<0 0 0>, rot=<0 0 0 1>}
 dyn_hair_4_c_ctrl_null {type=Null, loc=<0 -15.1378 132.962>, rot=<-0.5109 -0.5109 0.4888 -0.4888>}
 dyn_hair_4_c_ctrl {
  controlType=EULER_TRANSFORM
  parent=dyn_hair_4_c_ctrl_null
  type=Control
  loc=<-58.4486 -15.1378 119.6485>
  rot=<0.5978 0.4037 -0.3777 0.5805>
 }
 dyn_hair_3_c_ctrl_null {type=Null, loc=<0 -12.601 139.7946>, rot=<-0.4037 -0.4037 0.5805 -0.5805>}
 dyn_hair_3_c_ctrl {controlType=EULER_TRANSFORM, parent=dyn_hair_3_c_ctrl_null, type=Control, loc=<-20.6258 -12.601 138.4583>, rot=<0.431 0.3755 -0.5606 0.5992>}
 dyn_hair_2_c_ctrl_null {type=Null, loc=<0 -9.4102 146.3822>, rot=<0.3755 0.3755 -0.5992 0.5992>}
 dyn_hair_2_c_ctrl {controlType=EULER_TRANSFORM, parent=dyn_hair_2_c_ctrl_null, type=Control, loc=<4.3775 -9.4102 146.4998>, rot=<0.3513 0.3991 -0.6137 0.5837>}
 dyn_hair_1_c_ctrl_null {type=Null, loc=<0 -6.7546 153.204>, rot=<-0.3991 -0.3991 0.5837 -0.5837>}
 dyn_hair_1_c_ctrl {controlType=EULER_TRANSFORM, parent=dyn_hair_1_c_ctrl_null, type=Control, loc=<0 0 0>, rot=<0 0 0 1>}
 root {loc=<0 0 0>, rot=<0 0 0 1>}
 pelvis {parent=root, loc=<0 -0.5301 98.0571>, rot=<0.0127 0.707 -0.0127 -0.707>}
 spine_01 {parent=pelvis, loc=<6.1208 -0.9328 0>, rot=<0 0 0.034 -0.9994>}
 spine_02 {parent=spine_01, loc=<6.5997 -0.2095 0>, rot=<0 0 0.0238 -0.9997>}
 spine_03 {parent=spine_02, loc=<6.4914 -0.1237 0>, rot=<0 0 0.0149 -0.9999>}
 spine_04 {parent=spine_03, loc=<6.9598 0.0794 0>, rot=<0 0 -0.0132 -0.9999>}
 spine_05 {parent=spine_04, loc=<6.9987 0 0>, rot=<0 0 -0.0015 -1>}
 neck_01 {parent=spine_05, loc=<11.3138 0.1529 0>, rot=<0 0 -0.2003 -0.9797>}
 neck_02 {parent=neck_01, loc=<4.5627 0 0>, rot=<0 0 0.0504 -0.9987>}
 head {parent=neck_02, loc=<5.3914 0 0>, rot=<0 0 0.115 -0.9934>}
 base_hair_1_c {parent=head, loc=<1.2116 -5.0574 0>, rot=<0.1334 -0.6944 0.6944 -0.1334>}
 dyn_hair_1_c {parent=base_hair_1_c, loc=<0 0 0>, rot=<0 0 0 1>}
 dyn_hair_2_c {parent=dyn_hair_1_c, loc=<7.3205 0 0>, rot=<0 0 0 1>}
 dyn_hair_3_c {parent=dyn_hair_2_c, loc=<7.3197 0 0>, rot=<0 0 0 1>}
 dyn_hair_4_c {parent=dyn_hair_3_c, loc=<7.2884 0 0>, rot=<0 0 0 1>}
 dyn_hair_5_c {parent=dyn_hair_4_c, loc=<7.2939 0 0>, rot=<0 0 0 1>}
 dyn_hair_6_c {parent=dyn_hair_5_c, loc=<7.3213 0 0>, rot=<0 0 0 1>}
 dyn_hair_7_c {parent=dyn_hair_6_c, loc=<7.3211 0 0>, rot=<0 0 0 1>}
 base_hair_1_l {parent=head, loc=<1.2116 -5.0574 -5.6>, rot=<0.1334 -0.6944 0.6944 -0.1334>}
 dyn_hair_1_l {parent=base_hair_1_l, loc=<0 0 0>, rot=<0 0 0 1>}
 dyn_hair_2_l {parent=dyn_hair_1_l, loc=<7.3205 0 0>, rot=<0 0 0 1>}
 dyn_hair_3_l {parent=dyn_hair_2_l, loc=<7.3197 0 0>, rot=<0 0 0 1>}
 dyn_hair_4_l {parent=dyn_hair_3_l, loc=<7.2884 0 0>, rot=<0 0 0 1>}
 dyn_hair_5_l {parent=dyn_hair_4_l, loc=<7.2939 0 0>, rot=<0 0 0 1>}
 dyn_hair_6_l {parent=dyn_hair_5_l, loc=<7.3213 0 0>, rot=<0 0 0 1>}
 dyn_hair_7_l {parent=dyn_hair_6_l, loc=<7.3211 0 0>, rot=<0 0 0 1>}
 base_hair_1_r {parent=head, loc=<1.2116 -5.0574 5.6>, rot=<0.1334 -0.6944 0.6944 -0.1334>}
 dyn_hair_1_r {parent=base_hair_1_r, loc=<0 0 0>, rot=<0 0 0 1>}
 dyn_hair_2_r {parent=dyn_hair_1_r, loc=<7.3206 0 -0.0001>, rot=<0 0 0 1>}
 dyn_hair_3_r {parent=dyn_hair_2_r, loc=<7.3192 0 0.0002>, rot=<0 0 0 1>}
 dyn_hair_4_r {parent=dyn_hair_3_r, loc=<7.2887 0 -0.0001>, rot=<0 0 0 1>}
 dyn_hair_5_r {parent=dyn_hair_4_r, loc=<7.2941 0 0.0001>, rot=<0 0 0 1>}
 dyn_hair_6_r {parent=dyn_hair_5_r, loc=<7.321 0 -0.0001>, rot=<0 0 0 1>}
 dyn_hair_7_r {parent=dyn_hair_6_r, loc=<7.321 0 0>, rot=<0 0 0 1>}
 base_side_1_l {parent=head, loc=<2.1671 -4.7708 -10.2321>, rot=<0.1028 -0.9828 0.1529 0.0147>}
 dyn_side_1_l {parent=base_side_1_l, loc=<0 0 0>, rot=<0 0 0 1>}
 dyn_side_2_l {parent=dyn_side_1_l, loc=<3.7385 0 0>, rot=<-0.0094 0.0003 -0.23 -0.9731>}
 dyn_side_3_l {parent=dyn_side_2_l, loc=<3.7671 0 0>, rot=<-0.0734 0.0433 -0.2647 -0.9606>}
 dyn_side_4_l {parent=dyn_side_3_l, loc=<3.7063 0 0>, rot=<-0.0569 0.0604 -0.0135 -0.9965>}
 dyn_side_5_l {parent=dyn_side_4_l, loc=<3.7381 0 0>, rot=<-0.1153 0.0842 0.1947 -0.9704>}
 dyn_side_6_l {parent=dyn_side_5_l, loc=<3.7495 0 0>, rot=<0.5975 -0.3476 -0.0679 -0.7194>}
 dyn_side_7_l {parent=dyn_side_6_l, loc=<3.7186 0 0>, rot=<0 0 0 1>}
 base_front_1_l {parent=head, loc=<2.1092 2.2933 -12.0367>, rot=<0.089 -0.9878 0.1274 0.0104>}
 dyn_front_1_l {parent=base_front_1_l, loc=<0 0 0>, rot=<0 0 0 1>}
 dyn_front_2_l {parent=dyn_front_1_l, loc=<5.41 0 0>, rot=<0.0963 -0.2979 -0.2082 -0.9266>}
 dyn_front_3_l {parent=dyn_front_2_l, loc=<5.4113 0 0>, rot=<0.1712 0.3731 -0.0741 -0.9088>}
 dyn_front_4_l {parent=dyn_front_3_l, loc=<5.2821 0 0>, rot=<-0.0255 -0.0713 0.1674 -0.983>}
 dyn_front_5_l {parent=dyn_front_4_l, loc=<5.369 0 0>, rot=<-0.0358 -0.1067 -0.1478 -0.9826>}
 dyn_front_6_l {parent=dyn_front_5_l, loc=<5.3796 0 0>, rot=<0.038 0.1567 0.2296 -0.9598>}
 dyn_front_7_l {parent=dyn_front_6_l, loc=<5.4828 0 0>, rot=<0 0 0 1>}
 base_earing_01 {parent=head, loc=<0.0321 0.1128 6.304>, rot=<-0.0007 0.1736 0.9848 -0.0041>}
 dyn_earing_01_r {parent=base_earing_01, loc=<0 0 0>, rot=<0 0 0 1>}
 dyn_earing_02_r {parent=dyn_earing_01_r, loc=<0.9998 -0.0581 0.1972>, rot=<0 0 0 1>}
 dyn_earing_03_r {parent=dyn_earing_02_r, loc=<1.0639 0 0>, rot=<0 0 0 1>}
 base_offset_hair_r {parent=head, loc=<-8.186 -1.9499 6.2396>, rot=<-0.0009 -0.9948 0.0603 -0.0822>}
 dyn_offset_hair_r {parent=base_offset_hair_r, loc=<0 0 0>, rot=<0 0 0 1>}
 clavicle_l {parent=spine_05, loc=<5.4831 -0.1053 -3.2946>, rot=<0.0465 -0.7289 -0.009 -0.6829>}
 upperarm_l {parent=clavicle_l, loc=<9.8658 -1.2136 -0.4105>, rot=<0.0164 -0.3245 -0.0269 -0.9454>}
 upperarm_twist_01_l {parent=upperarm_l, loc=<7.2135 -0.1134 0>, rot=<0 0 0 1>}
 upperarm_l_null {parent=clavicle_l, type=Null, loc=<9.8658 -1.2136 -0.4105>, rot=<0.0164 -0.3245 -0.0269 -0.9454>}
 clavicle_r {parent=spine_05, loc=<5.4831 -0.1053 3.2946>, rot=<0.7289 0.0465 -0.6829 0.009>}
 upperarm_r {parent=clavicle_r, loc=<-9.8658 1.2136 0.4105>, rot=<0.0164 -0.3245 -0.0269 -0.9454>}
 upperarm_twist_01_r {parent=upperarm_r, loc=<-7.2135 0.1134 0>, rot=<0 0 0 -1>}
 upperarm_r_null {parent=clavicle_r, type=Null, loc=<-9.8658 1.2136 0.4105>, rot=<0.0164 -0.3245 -0.0269 -0.9454>}
 clavicle_l_null {parent=spine_05, type=Null, loc=<5.483 -0.1053 -3.2946>, rot=<0.0465 -0.7289 -0.009 -0.6829>}
 clavicle_r_null {parent=spine_05, type=Null, loc=<5.4831 -0.1053 3.2946>, rot=<0.7289 0.0465 -0.6829 0.009>}
]
variables [
 head rz n60n80 alpha {type=double, default=0}
 isAR Equipped lerp {type=double, default=0}
 isPASwing lerp {type=double, default=0}
 isPistol Equipped lerp {type=double, default=0}
 isPistolASDUp Extreme {type=double, default=0}
 isTargeting lerp {type=double, default=0}
 pub PA Swing {type=bool, default=False}
 pub isAR Equipped {type=bool, default=False}
 pub isCrouching {type=bool, default=False}
 pub isDBNO {type=bool, default=False}
 pub isOutLaw {type=bool, default=False}
 pub isParachuteOpened {type=bool, default=False}
 pub isPistol Equipped {type=bool, default=False}
 pub isRigidBodies Enabled {type=bool, default=False}
 pub isSMG90 {type=bool, default=False}
 pub isTargeting {type=bool, default=False}
 pub isWearingBackBling {type=bool, default=False}
]
