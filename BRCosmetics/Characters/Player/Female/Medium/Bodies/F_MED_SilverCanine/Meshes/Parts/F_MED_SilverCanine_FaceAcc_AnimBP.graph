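# Post-process-style AnimBP for the SilverCanine female face accessory, adding secondary motion to front, hair and earring bones.
# Pose copied from the parent mesh feeds a Control Rig driven by weapon and movement flags.
# The cached result runs through serial single-bone AnimDynamics nodes, simulating hair bones twice with differing angular limits.
# The event graph updates gravity override, front-end settings, backbling flags, fall-based hair overlap and equipped-weapon flags.
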AnimBlueprint /BRCosmetics/Characters/Player/Female/Medium/Bodies/F_MED_SilverCanine/Meshes/Parts/F_MED_SilverCanine_FaceAcc_AnimBP
target_skeleton=&/F_MED_SilverCanine_FaceAcc_Skeleton
graphs [
 AnimGraph {nodes [
   Output Pose(Result=Use cached pose 'CR_Cache_Dynamics'()) {layerGroup=None}
   n1 = Control Rig!(
    Source=Copy Pose From Mesh() {bUseAttachedParent=true}
    Alpha
    pub isRigidBodies Enabled:bool=Get bAnimDynamicsEnabled() -> (bAnimDynamicsEnabled:bool)
    pub PA Swing:bool=Get bIsPlayingMeleeAnim()
    pub isWearingBackBling:bool=Get isWearingBlackbling() -> (isWearingBlackbling:bool)
    pub isTargeting:bool=Get bIsTargeting()
    pub isAR Equipped:bool=Get isAR Equipped() -> (isAR Equipped:bool)
    pub isPistol Equipped:bool=Get isPistol Equipped() -> (isPistol Equipped:bool)
    pub isParachuteOpened:bool=Get bIsParachuteOpened()
    pub isCrouching:bool=Get bIsCrouching()
    pub isSMG90:bool=Get isSMG90() -> (isSMG90:bool)
    pub isOutLaw:bool=Get isOutLaw() -> (isOutLaw:bool)
    pub isDBNO:bool=Get bIsDBNO()
   ) -> (Pose) {controlRigAssetReference {blueprintRigClass=&/F_MED_SilverCanine_FaceAcc_CtrlRig.F_MED_SilverCanine_FaceAcc_CtrlRig_C}}
   Save cached pose 'CR_Cache_Input'(Pose=@n1) {cacheName=CR_Cache_Input}
   n2 = Anim Dynamics - Bone: dyn_front_1_l!(
    ComponentPose=Local To Component(LocalPose=Use cached pose 'CR_Cache_Input'())
    Alpha=1
    GravityOverride:Vector="0,0,-980"
    SimSpaceSettings:Anim Phys Sim Space Settings Structure=(SimSpaceAngularAlpha=0,MaxAngularVelocity=10000,MaxAngularAcceleration=10000,ExternalAngularVelocity=(X=0,Y=0,Z=0))
    ExternalForce="0,0,0"
   ) -> (Pose) {
    angularDampingOverride=1
    angularSpringConstant=20000
    bAngularSpring=true
    bOverrideAngularDamping=true
    bUseGravityOverride=true
    gravityScale=0.35
    simulationSpace=RootRelative
    boundBone {boneName=dyn_front_1_l}
    componentAppliedLinearAccClamp=<10000 10000 10000>
    componentLinearAccScale=<1 1 1>
    componentLinearVelScale=<1 1 1>
    gravityOverride {z=-980}
    physicsBodyDefinitions [{
      boundBone {boneName=dyn_front_1_l}
      boxExtents=<35 50 50>
      constraintSetup {angularLimitsMax=<3 0 10>, angularLimitsMin=<-3 -5 0>, angularTarget=<1 0 0>}
      localJointOffset=<4 0 0>
     }]
   }
   n3 = Anim Dynamics - Bone: dyn_front_2_l!(
    ComponentPose=@n2
    Alpha=1
    GravityOverride:Vector="0,0,-980"
    SimSpaceSettings:Anim Phys Sim Space Settings Structure=(SimSpaceAngularAlpha=0,MaxAngularVelocity=10000,MaxAngularAcceleration=10000,ExternalAngularVelocity=(X=0,Y=0,Z=0))
    ExternalForce="0,0,0"
   ) -> (Pose) {
    angularDampingOverride=1
    angularSpringConstant=20000
    bAngularSpring=true
    bOverrideAngularDamping=true
    bUseGravityOverride=true
    gravityScale=0.35
    simulationSpace=RootRelative
    boundBone {boneName=dyn_front_2_l}
    componentAppliedLinearAccClamp=<10000 10000 10000>
    componentLinearAccScale=<1 1 1>
    componentLinearVelScale=<1 1 1>
    gravityOverride {z=-980}
    physicsBodyDefinitions [{
      boundBone {boneName=dyn_front_2_l}
      boxExtents=<35 50 50>
      constraintSetup {angularLimitsMax=<3 0 10>, angularLimitsMin=<-3 -5 0>, angularTarget=<1 0 0>}
      localJointOffset=<4 0 0>
     }]
   }
   n4 = Anim Dynamics - Bone: dyn_front_3_l!(
    ComponentPose=@n3
    Alpha=1
    GravityOverride:Vector="0,0,-980"
    SimSpaceSettings:Anim Phys Sim Space Settings Structure=(SimSpaceAngularAlpha=0,MaxAngularVelocity=10000,MaxAngularAcceleration=10000,ExternalAngularVelocity=(X=0,Y=0,Z=0))
    ExternalForce="0,0,0"
   ) -> (Pose) {
    angularDampingOverride=1
    angularSpringConstant=20000
    bAngularSpring=true
    bOverrideAngularDamping=true
    bUseGravityOverride=true
    gravityScale=0.35
    simulationSpace=RootRelative
    boundBone {boneName=dyn_front_3_l}
    componentAppliedLinearAccClamp=<10000 10000 10000>
    componentLinearAccScale=<1 1 1>
    componentLinearVelScale=<1 1 1>
    gravityOverride {z=-980}
    physicsBodyDefinitions [{
      boundBone {boneName=dyn_front_3_l}
      boxExtents=<35 50 50>
      constraintSetup {angularLimitsMax=<3 0 10>, angularLimitsMin=<-3 -5 0>, angularTarget=<1 0 0>}
      localJointOffset=<4 0 0>
     }]
   }
   n5 = Anim Dynamics - Bone: dyn_front_4_l!(
    ComponentPose=@n4
    Alpha=1
    GravityOverride:Vector="0,0,-980"
    SimSpaceSettings:Anim Phys Sim Space Settings Structure=(SimSpaceAngularAlpha=0,MaxAngularVelocity=10000,MaxAngularAcceleration=10000,ExternalAngularVelocity=(X=0,Y=0,Z=0))
    ExternalForce="0,0,0"
   ) -> (Pose) {
    angularDampingOverride=1
    angularSpringConstant=20000
    bAngularSpring=true
    bOverrideAngularDamping=true
    bUseGravityOverride=true
    gravityScale=0.35
    simulationSpace=RootRelative
    boundBone {boneName=dyn_front_4_l}
    componentAppliedLinearAccClamp=<10000 10000 10000>
    componentLinearAccScale=<1 1 1>
    componentLinearVelScale=<1 1 1>
    gravityOverride {z=-980}
    physicsBodyDefinitions [{
      boundBone {boneName=dyn_front_4_l}
      boxExtents=<35 50 50>
      constraintSetup {angularLimitsMax=<3 0 10>, angularLimitsMin=<-3 -5 0>, angularTarget=<1 0 0>}
      localJointOffset=<4 0 0>
     }]
   }
   n6 = Anim Dynamics - Bone: dyn_front_5_l!(
    ComponentPose=@n5
    Alpha=1
    GravityOverride:Vector="0,0,-980"
    SimSpaceSettings:Anim Phys Sim Space Settings Structure=(SimSpaceAngularAlpha=0,MaxAngularVelocity=10000,MaxAngularAcceleration=10000,ExternalAngularVelocity=(X=0,Y=0,Z=0))
    ExternalForce="0,0,0"
   ) -> (Pose) {
    angularDampingOverride=1
    angularSpringConstant=20000
    bAngularSpring=true
    bOverrideAngularDamping=true
    bUseGravityOverride=true
    gravityScale=0.35
    simulationSpace=RootRelative
    boundBone {boneName=dyn_front_5_l}
    componentAppliedLinearAccClamp=<10000 10000 10000>
    componentLinearAccScale=<1 1 1>
    componentLinearVelScale=<1 1 1>
    gravityOverride {z=-980}
    physicsBodyDefinitions [{
      boundBone {boneName=dyn_front_5_l}
      boxExtents=<35 50 50>
      constraintSetup {angularLimitsMax=<3 0 10>, angularLimitsMin=<-3 -5 0>, angularTarget=<1 0 0>}
      localJointOffset=<4 0 0>
     }]
   }
   n7 = Anim Dynamics - Bone: dyn_front_6_l!(
    ComponentPose=@n6
    Alpha=1
    GravityOverride:Vector="0,0,-980"
    SimSpaceSettings:Anim Phys Sim Space Settings Structure=(SimSpaceAngularAlpha=0,MaxAngularVelocity=10000,MaxAngularAcceleration=10000,ExternalAngularVelocity=(X=0,Y=0,Z=0))
    ExternalForce="0,0,0"
   ) -> (Pose) {
    angularDampingOverride=1
    angularSpringConstant=20000
    bAngularSpring=true
    bOverrideAngularDamping=true
    bUseGravityOverride=true
    gravityScale=0.35
    simulationSpace=RootRelative
    boundBone {boneName=dyn_front_6_l}
    componentAppliedLinearAccClamp=<10000 10000 10000>
    componentLinearAccScale=<1 1 1>
    componentLinearVelScale=<1 1 1>
    gravityOverride {z=-980}
    physicsBodyDefinitions [{
      boundBone {boneName=dyn_front_6_l}
      boxExtents=<35 50 50>
      constraintSetup {angularLimitsMax=<3 0 10>, angularLimitsMin=<-3 -5 0>, angularTarget=<1 0 0>}
      localJointOffset=<4 0 0>
     }]
   }
   n8 = Anim Dynamics - Bone: dyn_hair_1_c!(
    ComponentPose=@n7
    bAlphaBoolEnabled:bool=True
    GravityOverride:Vector="0,0,-980"
    SimSpaceSettings:Anim Phys Sim Space Settings Structure=(SimSpaceAngularAlpha=0,MaxAngularVelocity=10000,MaxAngularAcceleration=10000,ExternalAngularVelocity=(X=0,Y=0,Z=0))
    ExternalForce="0,0,0"
   ) -> (Pose) {
    alphaInputType=Bool
    angularDampingOverride=1
    angularSpringConstant=20000
    bAngularSpring=true
    bOverrideAngularDamping=true
    bUseGravityOverride=true
    gravityScale=0.35
    simulationSpace=RootRelative
    boundBone {boneName=dyn_hair_1_c}
    componentAppliedLinearAccClamp=<10000 10000 10000>
    componentLinearAccScale=<1 1 1>
    componentLinearVelScale=<1 1 1>
    gravityOverride {z=-980}
    physicsBodyDefinitions [{
      boundBone {boneName=dyn_hair_1_c}
      boxExtents=<30 40 40>
      constraintSetup {angularLimitsMax=<2 0 5>, angularLimitsMin=<-2 0 -5>, angularTarget=<1 0 0>}
      localJointOffset=<4 0 0>
     }]
   }
   n9 = Anim Dynamics - Bone: dyn_hair_2_c!(
    ComponentPose=@n8
    bAlphaBoolEnabled:bool=True
    GravityOverride:Vector="0,0,-980"
    SimSpaceSettings:Anim Phys Sim Space Settings Structure=(SimSpaceAngularAlpha=0,MaxAngularVelocity=10000,MaxAngularAcceleration=10000,ExternalAngularVelocity=(X=0,Y=0,Z=0))
    ExternalForce="0,0,0"
   ) -> (Pose) {
    alphaInputType=Bool
    angularDampingOverride=1
    angularSpringConstant=20000
    bAngularSpring=true
    bOverrideAngularDamping=true
    bUseGravityOverride=true
    gravityScale=0.35
    simulationSpace=RootRelative
    boundBone {boneName=dyn_hair_2_c}
    componentAppliedLinearAccClamp=<10000 10000 10000>
    componentLinearAccScale=<1 1 1>
    componentLinearVelScale=<1 1 1>
    gravityOverride {z=-980}
    physicsBodyDefinitions [{
      boundBone {boneName=dyn_hair_2_c}
      boxExtents=<30 40 40>
      constraintSetup {angularLimitsMax=<2 0 5>, angularLimitsMin=<-2 0 -5>, angularTarget=<1 0 0>}
      localJointOffset=<4 0 0>
     }]
   }
   n10 = Anim Dynamics - Bone: dyn_hair_3_c!(
    ComponentPose=@n9
    bAlphaBoolEnabled:bool=True
    GravityOverride:Vector="0,0,-980"
    SimSpaceSettings:Anim Phys Sim Space Settings Structure=(SimSpaceAngularAlpha=0,MaxAngularVelocity=10000,MaxAngularAcceleration=10000,ExternalAngularVelocity=(X=0,Y=0,Z=0))
    ExternalForce="0,0,0"
   ) -> (Pose) {
    alphaInputType=Bool
    angularDampingOverride=1
    angularSpringConstant=20000
    bAngularSpring=true
    bOverrideAngularDamping=true
    bUseGravityOverride=true
    gravityScale=0.35
    simulationSpace=RootRelative
    boundBone {boneName=dyn_hair_3_c}
    componentAppliedLinearAccClamp=<10000 10000 10000>
    componentLinearAccScale=<1 1 1>
    componentLinearVelScale=<1 1 1>
    gravityOverride {z=-980}
    physicsBodyDefinitions [{
      boundBone {boneName=dyn_hair_3_c}
      boxExtents=<30 40 40>
      constraintSetup {angularLimitsMax=<2 0 5>, angularLimitsMin=<-2 0 -5>, angularTarget=<1 0 0>}
      localJointOffset=<4 0 0>
     }]
   }
   n11 = Anim Dynamics - Bone: dyn_hair_4_c!(
    ComponentPose=@n10
    bAlphaBoolEnabled:bool=True
    GravityOverride:Vector="0,0,-980"
    SimSpaceSettings:Anim Phys Sim Space Settings Structure=(SimSpaceAngularAlpha=0,MaxAngularVelocity=10000,MaxAngularAcceleration=10000,ExternalAngularVelocity=(X=0,Y=0,Z=0))
    ExternalForce="0,0,0"
   ) -> (Pose) {
    alphaInputType=Bool
    angularDampingOverride=1
    angularSpringConstant=20000
    bAngularSpring=true
    bOverrideAngularDamping=true
    bUseGravityOverride=true
    gravityScale=0.35
    simulationSpace=RootRelative
    boundBone {boneName=dyn_hair_4_c}
    componentAppliedLinearAccClamp=<10000 10000 10000>
    componentLinearAccScale=<1 1 1>
    componentLinearVelScale=<1 1 1>
    gravityOverride {z=-980}
    physicsBodyDefinitions [{
      boundBone {boneName=dyn_hair_4_c}
      boxExtents=<30 40 40>
      constraintSetup {angularLimitsMax=<2 0 5>, angularLimitsMin=<-2 0 -5>, angularTarget=<1 0 0>}
      localJointOffset=<4 0 0>
     }]
   }
   n12 = Anim Dynamics - Bone: dyn_hair_1_l!(
    ComponentPose=@n11
    bAlphaBoolEnabled:bool=True
    GravityOverride:Vector="0,0,-980"
    SimSpaceSettings:Anim Phys Sim Space Settings Structure=(SimSpaceAngularAlpha=0,MaxAngularVelocity=10000,MaxAngularAcceleration=10000,ExternalAngularVelocity=(X=0,Y=0,Z=0))
    ExternalForce="0,0,0"
   ) -> (Pose) {
    alphaInputType=Bool
    angularDampingOverride=1
    angularSpringConstant=20000
    bAngularSpring=true
    bOverrideAngularDamping=true
    bUseGravityOverride=true
    gravityScale=0.35
    simulationSpace=RootRelative
    boundBone {boneName=dyn_hair_1_l}
    componentAppliedLinearAccClamp=<10000 10000 10000>
    componentLinearAccScale=<1 1 1>
    componentLinearVelScale=<1 1 1>
    gravityOverride {z=-980}
    physicsBodyDefinitions [{
      boundBone {boneName=dyn_hair_1_l}
      boxExtents=<30 40 40>
      constraintSetup {angularLimitsMax=<2 0 5>, angularLimitsMin=<-2 0 -5>, angularTarget=<1 0 0>}
      localJointOffset=<4 0 0>
     }]
   }
   n13 = Anim Dynamics - Bone: dyn_hair_2_l!(
    ComponentPose=@n12
    bAlphaBoolEnabled:bool=True
    GravityOverride:Vector="0,0,-980"
    SimSpaceSettings:Anim Phys Sim Space Settings Structure=(SimSpaceAngularAlpha=0,MaxAngularVelocity=10000,MaxAngularAcceleration=10000,ExternalAngularVelocity=(X=0,Y=0,Z=0))
    ExternalForce="0,0,0"
   ) -> (Pose) {
    alphaInputType=Bool
    angularDampingOverride=1
    angularSpringConstant=20000
    bAngularSpring=true
    bOverrideAngularDamping=true
    bUseGravityOverride=true
    gravityScale=0.35
    simulationSpace=RootRelative
    boundBone {boneName=dyn_hair_2_l}
    componentAppliedLinearAccClamp=<10000 10000 10000>
    componentLinearAccScale=<1 1 1>
    componentLinearVelScale=<1 1 1>
    gravityOverride {z=-980}
    physicsBodyDefinitions [{
      boundBone {boneName=dyn_hair_2_l}
      boxExtents=<30 40 40>
      constraintSetup {angularLimitsMax=<2 0 5>, angularLimitsMin=<-2 0 -5>, angularTarget=<1 0 0>}
      localJointOffset=<4 0 0>
     }]
   }
   n14 = Anim Dynamics - Bone: dyn_hair_3_l!(
    ComponentPose=@n13
    bAlphaBoolEnabled:bool=True
    GravityOverride:Vector="0,0,-980"
    SimSpaceSettings:Anim Phys Sim Space Settings Structure=(SimSpaceAngularAlpha=0,MaxAngularVelocity=10000,MaxAngularAcceleration=10000,ExternalAngularVelocity=(X=0,Y=0,Z=0))
    ExternalForce="0,0,0"
   ) -> (Pose) {
    alphaInputType=Bool
    angularDampingOverride=1
    angularSpringConstant=20000
    bAngularSpring=true
    bOverrideAngularDamping=true
    bUseGravityOverride=true
    gravityScale=0.35
    simulationSpace=RootRelative
    boundBone {boneName=dyn_hair_3_l}
    componentAppliedLinearAccClamp=<10000 10000 10000>
    componentLinearAccScale=<1 1 1>
    componentLinearVelScale=<1 1 1>
    gravityOverride {z=-980}
    physicsBodyDefinitions [{
      boundBone {boneName=dyn_hair_3_l}
      boxExtents=<30 40 40>
      constraintSetup {angularLimitsMax=<2 0 5>, angularLimitsMin=<-2 0 -5>, angularTarget=<1 0 0>}
      localJointOffset=<4 0 0>
     }]
   }
   n15 = Anim Dynamics - Bone: dyn_hair_4_l!(
    ComponentPose=@n14
    bAlphaBoolEnabled:bool=True
    GravityOverride:Vector="0,0,-980"
    SimSpaceSettings:Anim Phys Sim Space Settings Structure=(SimSpaceAngularAlpha=0,MaxAngularVelocity=10000,MaxAngularAcceleration=10000,ExternalAngularVelocity=(X=0,Y=0,Z=0))
    ExternalForce="0,0,0"
   ) -> (Pose) {
    alphaInputType=Bool
    angularDampingOverride=1
    angularSpringConstant=20000
    bAngularSpring=true
    bOverrideAngularDamping=true
    bUseGravityOverride=true
    gravityScale=0.35
    simulationSpace=RootRelative
    boundBone {boneName=dyn_hair_4_l}
    componentAppliedLinearAccClamp=<10000 10000 10000>
    componentLinearAccScale=<1 1 1>
    componentLinearVelScale=<1 1 1>
    gravityOverride {z=-980}
    physicsBodyDefinitions [{
      boundBone {boneName=dyn_hair_4_l}
      boxExtents=<30 40 40>
      constraintSetup {angularLimitsMax=<2 0 5>, angularLimitsMin=<-2 0 -5>, angularTarget=<1 0 0>}
      localJointOffset=<4 0 0>
     }]
   }
   n16 = Anim Dynamics - Bone: dyn_hair_1_r!(
    ComponentPose=@n15
    bAlphaBoolEnabled:bool=True
    GravityOverride:Vector="0,0,-980"
    SimSpaceSettings:Anim Phys Sim Space Settings Structure=(SimSpaceAngularAlpha=0,MaxAngularVelocity=10000,MaxAngularAcceleration=10000,ExternalAngularVelocity=(X=0,Y=0,Z=0))
    ExternalForce="0,0,0"
   ) -> (Pose) {
    alphaInputType=Bool
    angularDampingOverride=1
    angularSpringConstant=20000
    bAngularSpring=true
    bOverrideAngularDamping=true
    bUseGravityOverride=true
    gravityScale=0.35
    simulationSpace=RootRelative
    boundBone {boneName=dyn_hair_1_r}
    componentAppliedLinearAccClamp=<10000 10000 10000>
    componentLinearAccScale=<1 1 1>
    componentLinearVelScale=<1 1 1>
    gravityOverride {z=-980}
    physicsBodyDefinitions [{
      boundBone {boneName=dyn_hair_1_r}
      boxExtents=<30 40 40>
      constraintSetup {angularLimitsMax=<2 0 5>, angularLimitsMin=<-2 0 -5>, angularTarget=<1 0 0>}
      localJointOffset=<4 0 0>
     }]
   }
   n17 = Anim Dynamics - Bone: dyn_hair_2_r!(
    ComponentPose=@n16
    bAlphaBoolEnabled:bool=True
    GravityOverride:Vector="0,0,-980"
    SimSpaceSettings:Anim Phys Sim Space Settings Structure=(SimSpaceAngularAlpha=0,MaxAngularVelocity=10000,MaxAngularAcceleration=10000,ExternalAngularVelocity=(X=0,Y=0,Z=0))
    ExternalForce="0,0,0"
   ) -> (Pose) {
    alphaInputType=Bool
    angularDampingOverride=1
    angularSpringConstant=20000
    bAngularSpring=true
    bOverrideAngularDamping=true
    bUseGravityOverride=true
    gravityScale=0.35
    simulationSpace=RootRelative
    boundBone {boneName=dyn_hair_2_r}
    componentAppliedLinearAccClamp=<10000 10000 10000>
    componentLinearAccScale=<1 1 1>
    componentLinearVelScale=<1 1 1>
    gravityOverride {z=-980}
    physicsBodyDefinitions [{
      boundBone {boneName=dyn_hair_2_r}
      boxExtents=<30 40 40>
      constraintSetup {angularLimitsMax=<2 0 5>, angularLimitsMin=<-2 0 -5>, angularTarget=<1 0 0>}
      localJointOffset=<4 0 0>
     }]
   }
   n18 = Anim Dynamics - Bone: dyn_hair_3_r!(
    ComponentPose=@n17
    bAlphaBoolEnabled:bool=True
    GravityOverride:Vector="0,0,-980"
    SimSpaceSettings:Anim Phys Sim Space Settings Structure=(SimSpaceAngularAlpha=0,MaxAngularVelocity=10000,MaxAngularAcceleration=10000,ExternalAngularVelocity=(X=0,Y=0,Z=0))
    ExternalForce="0,0,0"
   ) -> (Pose) {
    alphaInputType=Bool
    angularDampingOverride=1
    angularSpringConstant=20000
    bAngularSpring=true
    bOverrideAngularDamping=true
    bUseGravityOverride=true
    gravityScale=0.35
    simulationSpace=RootRelative
    boundBone {boneName=dyn_hair_3_r}
    componentAppliedLinearAccClamp=<10000 10000 10000>
    componentLinearAccScale=<1 1 1>
    componentLinearVelScale=<1 1 1>
    gravityOverride {z=-980}
    physicsBodyDefinitions [{
      boundBone {boneName=dyn_hair_3_r}
      boxExtents=<30 40 40>
      constraintSetup {angularLimitsMax=<2 0 5>, angularLimitsMin=<-2 0 -5>, angularTarget=<1 0 0>}
      localJointOffset=<4 0 0>
     }]
   }
   n19 = Anim Dynamics - Bone: dyn_hair_4_r!(
    ComponentPose=@n18
    bAlphaBoolEnabled:bool=True
    GravityOverride:Vector="0,0,-980"
    SimSpaceSettings:Anim Phys Sim Space Settings Structure=(SimSpaceAngularAlpha=0,MaxAngularVelocity=10000,MaxAngularAcceleration=10000,ExternalAngularVelocity=(X=0,Y=0,Z=0))
    ExternalForce="0,0,0"
   ) -> (Pose) {
    alphaInputType=Bool
    angularDampingOverride=1
    angularSpringConstant=20000
    bAngularSpring=true
    bOverrideAngularDamping=true
    bUseGravityOverride=true
    gravityScale=0.35
    simulationSpace=RootRelative
    boundBone {boneName=dyn_hair_4_r}
    componentAppliedLinearAccClamp=<10000 10000 10000>
    componentLinearAccScale=<1 1 1>
    componentLinearVelScale=<1 1 1>
    gravityOverride {z=-980}
    physicsBodyDefinitions [{
      boundBone {boneName=dyn_hair_4_r}
      boxExtents=<30 40 40>
      constraintSetup {angularLimitsMax=<2 0 5>, angularLimitsMin=<-2 0 -5>, angularTarget=<1 0 0>}
      localJointOffset=<4 0 0>
     }]
   }
   n20 = Anim Dynamics - Bone: dyn_hair_1_c(
    ComponentPose=@n19
    bAlphaBoolEnabled=True
    GravityOverride="0,0,-980"
    SimSpaceSettings=(SimSpaceAngularAlpha=0,MaxAngularVelocity=10000,MaxAngularAcceleration=10000,ExternalAngularVelocity=(X=0,Y=0,Z=0))
    ExternalForce="0,0,0"
   ) {
    alphaInputType=Bool
    angularDampingOverride=1
    angularSpringConstant=20000
    bAngularSpring=true
    bOverrideAngularDamping=true
    bUseGravityOverride=true
    gravityScale=0.35
    simulationSpace=RootRelative
    boundBone {boneName=dyn_hair_1_c}
    componentAppliedLinearAccClamp=<10000 10000 10000>
    componentLinearAccScale=<1 1 1>
    componentLinearVelScale=<1 1 1>
    gravityOverride {z=-980}
    physicsBodyDefinitions [{
      boundBone {boneName=dyn_hair_1_c}
      boxExtents=<30 40 40>
      constraintSetup {angularLimitsMax=<2 0 5>, angularLimitsMin=<-2 -15 -5>, angularTarget=<1 0 0>}
      localJointOffset=<4 0 0>
     }]
   }
   n21 = Anim Dynamics - Bone: dyn_hair_2_c(
    ComponentPose=@n20
    bAlphaBoolEnabled=True
    GravityOverride="0,0,-980"
    SimSpaceSettings=(SimSpaceAngularAlpha=0,MaxAngularVelocity=10000,MaxAngularAcceleration=10000,ExternalAngularVelocity=(X=0,Y=0,Z=0))
    ExternalForce="0,0,0"
   ) {
    alphaInputType=Bool
    angularDampingOverride=1
    angularSpringConstant=20000
    bAngularSpring=true
    bOverrideAngularDamping=true
    bUseGravityOverride=true
    gravityScale=0.35
    simulationSpace=RootRelative
    boundBone {boneName=dyn_hair_2_c}
    componentAppliedLinearAccClamp=<10000 10000 10000>
    componentLinearAccScale=<1 1 1>
    componentLinearVelScale=<1 1 1>
    gravityOverride {z=-980}
    physicsBodyDefinitions [{
      boundBone {boneName=dyn_hair_2_c}
      boxExtents=<30 40 40>
      constraintSetup {angularLimitsMax=<2 0 5>, angularLimitsMin=<-2 -15 -5>, angularTarget=<1 0 0>}
      localJointOffset=<4 0 0>
     }]
   }
   n22 = Anim Dynamics - Bone: dyn_hair_3_c(
    ComponentPose=@n21
    bAlphaBoolEnabled=True
    GravityOverride="0,0,-980"
    SimSpaceSettings=(SimSpaceAngularAlpha=0,MaxAngularVelocity=10000,MaxAngularAcceleration=10000,ExternalAngularVelocity=(X=0,Y=0,Z=0))
    ExternalForce="0,0,0"
   ) {
    alphaInputType=Bool
    angularDampingOverride=1
    angularSpringConstant=20000
    bAngularSpring=true
    bOverrideAngularDamping=true
    bUseGravityOverride=true
    gravityScale=0.35
    simulationSpace=RootRelative
    boundBone {boneName=dyn_hair_3_c}
    componentAppliedLinearAccClamp=<10000 10000 10000>
    componentLinearAccScale=<1 1 1>
    componentLinearVelScale=<1 1 1>
    gravityOverride {z=-980}
    physicsBodyDefinitions [{
      boundBone {boneName=dyn_hair_3_c}
      boxExtents=<30 40 40>
      constraintSetup {angularLimitsMax=<2 0 5>, angularLimitsMin=<-2 -15 -5>, angularTarget=<1 0 0>}
      localJointOffset=<4 0 0>
     }]
   }
   n23 = Anim Dynamics - Bone: dyn_hair_4_c(
    ComponentPose=@n22
    bAlphaBoolEnabled=True
    GravityOverride="0,0,-980"
    SimSpaceSettings=(SimSpaceAngularAlpha=0,MaxAngularVelocity=10000,MaxAngularAcceleration=10000,ExternalAngularVelocity=(X=0,Y=0,Z=0))
    ExternalForce="0,0,0"
   ) {
    alphaInputType=Bool
    angularDampingOverride=1
    angularSpringConstant=20000
    bAngularSpring=true
    bOverrideAngularDamping=true
    bUseGravityOverride=true
    gravityScale=0.35
    simulationSpace=RootRelative
    boundBone {boneName=dyn_hair_4_c}
    componentAppliedLinearAccClamp=<10000 10000 10000>
    componentLinearAccScale=<1 1 1>
    componentLinearVelScale=<1 1 1>
    gravityOverride {z=-980}
    physicsBodyDefinitions [{
      boundBone {boneName=dyn_hair_4_c}
      boxExtents=<30 40 40>
      constraintSetup {angularLimitsMax=<2 0 5>, angularLimitsMin=<-2 -15 -5>, angularTarget=<1 0 0>}
      localJointOffset=<4 0 0>
     }]
   }
   n24 = Anim Dynamics - Bone: dyn_hair_1_l(
    ComponentPose=@n23
    bAlphaBoolEnabled=True
    GravityOverride="0,0,-980"
    SimSpaceSettings=(SimSpaceAngularAlpha=0,MaxAngularVelocity=10000,MaxAngularAcceleration=10000,ExternalAngularVelocity=(X=0,Y=0,Z=0))
    ExternalForce="0,0,0"
   ) {
    alphaInputType=Bool
    angularDampingOverride=1
    angularSpringConstant=20000
    bAngularSpring=true
    bOverrideAngularDamping=true
    bUseGravityOverride=true
    gravityScale=0.35
    simulationSpace=RootRelative
    boundBone {boneName=dyn_hair_1_l}
    componentAppliedLinearAccClamp=<10000 10000 10000>
    componentLinearAccScale=<1 1 1>
    componentLinearVelScale=<1 1 1>
    gravityOverride {z=-980}
    physicsBodyDefinitions [{
      boundBone {boneName=dyn_hair_1_l}
      boxExtents=<30 40 40>
      constraintSetup {angularLimitsMax=<2 0 5>, angularLimitsMin=<-2 -15 -5>, angularTarget=<1 0 0>}
      localJointOffset=<4 0 0>
     }]
   }
   n25 = Anim Dynamics - Bone: dyn_hair_2_l(
    ComponentPose=@n24
    bAlphaBoolEnabled=True
    GravityOverride="0,0,-980"
    SimSpaceSettings=(SimSpaceAngularAlpha=0,MaxAngularVelocity=10000,MaxAngularAcceleration=10000,ExternalAngularVelocity=(X=0,Y=0,Z=0))
    ExternalForce="0,0,0"
   ) {
    alphaInputType=Bool
    angularDampingOverride=1
    angularSpringConstant=20000
    bAngularSpring=true
    bOverrideAngularDamping=true
    bUseGravityOverride=true
    gravityScale=0.35
    simulationSpace=RootRelative
    boundBone {boneName=dyn_hair_2_l}
    componentAppliedLinearAccClamp=<10000 10000 10000>
    componentLinearAccScale=<1 1 1>
    componentLinearVelScale=<1 1 1>
    gravityOverride {z=-980}
    physicsBodyDefinitions [{
      boundBone {boneName=dyn_hair_2_l}
      boxExtents=<30 40 40>
      constraintSetup {angularLimitsMax=<2 0 5>, angularLimitsMin=<-2 -15 -5>, angularTarget=<1 0 0>}
      localJointOffset=<4 0 0>
     }]
   }
   n26 = Anim Dynamics - Bone: dyn_hair_3_l(
    ComponentPose=@n25
    bAlphaBoolEnabled=True
    GravityOverride="0,0,-980"
    SimSpaceSettings=(SimSpaceAngularAlpha=0,MaxAngularVelocity=10000,MaxAngularAcceleration=10000,ExternalAngularVelocity=(X=0,Y=0,Z=0))
    ExternalForce="0,0,0"
   ) {
    alphaInputType=Bool
    angularDampingOverride=1
    angularSpringConstant=20000
    bAngularSpring=true
    bOverrideAngularDamping=true
    bUseGravityOverride=true
    gravityScale=0.35
    simulationSpace=RootRelative
    boundBone {boneName=dyn_hair_3_l}
    componentAppliedLinearAccClamp=<10000 10000 10000>
    componentLinearAccScale=<1 1 1>
    componentLinearVelScale=<1 1 1>
    gravityOverride {z=-980}
    physicsBodyDefinitions [{
      boundBone {boneName=dyn_hair_3_l}
      boxExtents=<30 40 40>
      constraintSetup {angularLimitsMax=<2 0 5>, angularLimitsMin=<-2 -15 -5>, angularTarget=<1 0 0>}
      localJointOffset=<4 0 0>
     }]
   }
   n27 = Anim Dynamics - Bone: dyn_hair_4_l(
    ComponentPose=@n26
    bAlphaBoolEnabled=True
    GravityOverride="0,0,-980"
    SimSpaceSettings=(SimSpaceAngularAlpha=0,MaxAngularVelocity=10000,MaxAngularAcceleration=10000,ExternalAngularVelocity=(X=0,Y=0,Z=0))
    ExternalForce="0,0,0"
   ) {
    alphaInputType=Bool
    angularDampingOverride=1
    angularSpringConstant=20000
    bAngularSpring=true
    bOverrideAngularDamping=true
    bUseGravityOverride=true
    gravityScale=0.35
    simulationSpace=RootRelative
    boundBone {boneName=dyn_hair_4_l}
    componentAppliedLinearAccClamp=<10000 10000 10000>
    componentLinearAccScale=<1 1 1>
    componentLinearVelScale=<1 1 1>
    gravityOverride {z=-980}
    physicsBodyDefinitions [{
      boundBone {boneName=dyn_hair_4_l}
      boxExtents=<30 40 40>
      constraintSetup {angularLimitsMax=<2 0 5>, angularLimitsMin=<-2 -15 -5>, angularTarget=<1 0 0>}
      localJointOffset=<4 0 0>
     }]
   }
   n28 = Anim Dynamics - Bone: dyn_hair_1_r(
    ComponentPose=@n27
    bAlphaBoolEnabled=True
    GravityOverride="0,0,-980"
    SimSpaceSettings=(SimSpaceAngularAlpha=0,MaxAngularVelocity=10000,MaxAngularAcceleration=10000,ExternalAngularVelocity=(X=0,Y=0,Z=0))
    ExternalForce="0,0,0"
   ) {
    alphaInputType=Bool
    angularDampingOverride=1
    angularSpringConstant=20000
    bAngularSpring=true
    bOverrideAngularDamping=true
    bUseGravityOverride=true
    gravityScale=0.35
    simulationSpace=RootRelative
    boundBone {boneName=dyn_hair_1_r}
    componentAppliedLinearAccClamp=<10000 10000 10000>
    componentLinearAccScale=<1 1 1>
    componentLinearVelScale=<1 1 1>
    gravityOverride {z=-980}
    physicsBodyDefinitions [{
      boundBone {boneName=dyn_hair_1_r}
      boxExtents=<30 40 40>
      constraintSetup {angularLimitsMax=<2 0 5>, angularLimitsMin=<-2 -15 -5>, angularTarget=<1 0 0>}
      localJointOffset=<4 0 0>
     }]
   }
   n29 = Anim Dynamics - Bone: dyn_hair_2_r(
    ComponentPose=@n28
    bAlphaBoolEnabled=True
    GravityOverride="0,0,-980"
    SimSpaceSettings=(SimSpaceAngularAlpha=0,MaxAngularVelocity=10000,MaxAngularAcceleration=10000,ExternalAngularVelocity=(X=0,Y=0,Z=0))
    ExternalForce="0,0,0"
   ) {
    alphaInputType=Bool
    angularDampingOverride=1
    angularSpringConstant=20000
    bAngularSpring=true
    bOverrideAngularDamping=true
    bUseGravityOverride=true
    gravityScale=0.35
    simulationSpace=RootRelative
    boundBone {boneName=dyn_hair_2_r}
    componentAppliedLinearAccClamp=<10000 10000 10000>
    componentLinearAccScale=<1 1 1>
    componentLinearVelScale=<1 1 1>
    gravityOverride {z=-980}
    physicsBodyDefinitions [{
      boundBone {boneName=dyn_hair_2_r}
      boxExtents=<30 40 40>
      constraintSetup {angularLimitsMax=<2 0 5>, angularLimitsMin=<-2 -15 -5>, angularTarget=<1 0 0>}
      localJointOffset=<4 0 0>
     }]
   }
   n30 = Anim Dynamics - Bone: dyn_hair_3_r(
    ComponentPose=@n29
    bAlphaBoolEnabled=True
    GravityOverride="0,0,-980"
    SimSpaceSettings=(SimSpaceAngularAlpha=0,MaxAngularVelocity=10000,MaxAngularAcceleration=10000,ExternalAngularVelocity=(X=0,Y=0,Z=0))
    ExternalForce="0,0,0"
   ) {
    alphaInputType=Bool
    angularDampingOverride=1
    angularSpringConstant=20000
    bAngularSpring=true
    bOverrideAngularDamping=true
    bUseGravityOverride=true
    gravityScale=0.35
    simulationSpace=RootRelative
    boundBone {boneName=dyn_hair_3_r}
    componentAppliedLinearAccClamp=<10000 10000 10000>
    componentLinearAccScale=<1 1 1>
    componentLinearVelScale=<1 1 1>
    gravityOverride {z=-980}
    physicsBodyDefinitions [{
      boundBone {boneName=dyn_hair_3_r}
      boxExtents=<30 40 40>
      constraintSetup {angularLimitsMax=<2 0 5>, angularLimitsMin=<-2 -15 -5>, angularTarget=<1 0 0>}
      localJointOffset=<4 0 0>
     }]
   }
   n31 = Anim Dynamics - Bone: dyn_hair_4_r(
    ComponentPose=@n30
    bAlphaBoolEnabled=True
    GravityOverride="0,0,-980"
    SimSpaceSettings=(SimSpaceAngularAlpha=0,MaxAngularVelocity=10000,MaxAngularAcceleration=10000,ExternalAngularVelocity=(X=0,Y=0,Z=0))
    ExternalForce="0,0,0"
   ) {
    alphaInputType=Bool
    angularDampingOverride=1
    angularSpringConstant=20000
    bAngularSpring=true
    bOverrideAngularDamping=true
    bUseGravityOverride=true
    gravityScale=0.35
    simulationSpace=RootRelative
    boundBone {boneName=dyn_hair_4_r}
    componentAppliedLinearAccClamp=<10000 10000 10000>
    componentLinearAccScale=<1 1 1>
    componentLinearVelScale=<1 1 1>
    gravityOverride {z=-980}
    physicsBodyDefinitions [{
      boundBone {boneName=dyn_hair_4_r}
      boxExtents=<30 40 40>
      constraintSetup {angularLimitsMax=<2 0 5>, angularLimitsMin=<-2 -15 -5>, angularTarget=<1 0 0>}
      localJointOffset=<4 0 0>
     }]
   }
   n32 = Anim Dynamics - Bone: dyn_earing_01_r!(
    ComponentPose=@n31
    Alpha=1
    GravityOverride:Vector="0,0,-980"
    SimSpaceSettings:Anim Phys Sim Space Settings Structure=(SimSpaceAngularAlpha=0,MaxAngularVelocity=10000,MaxAngularAcceleration=10000,ExternalAngularVelocity=(X=0,Y=0,Z=0))
    ExternalForce="0,0,0"
   ) -> (Pose) {
    angularDampingOverride=1
    angularSpringConstant=20000
    bAngularSpring=true
    bOverrideAngularDamping=true
    bUseGravityOverride=true
    gravityScale=0.35
    simulationSpace=RootRelative
    boundBone {boneName=dyn_earing_01_r}
    componentAppliedLinearAccClamp=<10000 10000 10000>
    componentLinearAccScale=<1 1 1>
    componentLinearVelScale=<1 1 1>
    gravityOverride {z=-980}
    physicsBodyDefinitions [{
      boundBone {boneName=dyn_earing_01_r}
      boxExtents=<30 40 40>
      constraintSetup {angularLimitsMax=<15 15 30>, angularLimitsMin=<-15 -15 -20>, angularTarget=<1 0 0>}
      localJointOffset=<4 0 0>
     }]
   }
   n33 = Anim Dynamics - Bone: dyn_earing_02_r!(
    ComponentPose=@n32
    Alpha=1
    GravityOverride:Vector="0,0,-980"
    SimSpaceSettings:Anim Phys Sim Space Settings Structure=(SimSpaceAngularAlpha=0,MaxAngularVelocity=10000,MaxAngularAcceleration=10000,ExternalAngularVelocity=(X=0,Y=0,Z=0))
    ExternalForce="0,0,0"
   ) -> (Pose) {
    angularDampingOverride=1
    angularSpringConstant=20000
    bAngularSpring=true
    bOverrideAngularDamping=true
    bUseGravityOverride=true
    gravityScale=0.35
    simulationSpace=RootRelative
    boundBone {boneName=dyn_earing_02_r}
    componentAppliedLinearAccClamp=<10000 10000 10000>
    componentLinearAccScale=<1 1 1>
    componentLinearVelScale=<1 1 1>
    gravityOverride {z=-980}
    physicsBodyDefinitions [{
      boundBone {boneName=dyn_earing_02_r}
      boxExtents=<30 40 40>
      constraintSetup {angularLimitsMax=<15 15 30>, angularLimitsMin=<-15 -15 -20>, angularTarget=<1 0 0>}
      localJointOffset=<4 0 0>
     }]
   }
   n34 = Anim Dynamics - Bone: dyn_earing_03_r!(
    ComponentPose=@n33
    Alpha=1
    GravityOverride:Vector="0,0,-980"
    SimSpaceSettings:Anim Phys Sim Space Settings Structure=(SimSpaceAngularAlpha=0,MaxAngularVelocity=10000,MaxAngularAcceleration=10000,ExternalAngularVelocity=(X=0,Y=0,Z=0))
    ExternalForce="0,0,0"
   ) -> (Pose) {
    angularDampingOverride=1
    angularSpringConstant=20000
    bAngularSpring=true
    bOverrideAngularDamping=true
    bUseGravityOverride=true
    gravityScale=0.35
    simulationSpace=RootRelative
    boundBone {boneName=dyn_earing_03_r}
    componentAppliedLinearAccClamp=<10000 10000 10000>
    componentLinearAccScale=<1 1 1>
    componentLinearVelScale=<1 1 1>
    gravityOverride {z=-980}
    physicsBodyDefinitions [{
      boundBone {boneName=dyn_earing_03_r}
      boxExtents=<30 40 40>
      constraintSetup {angularLimitsMax=<15 15 30>, angularLimitsMin=<-15 -15 -20>, angularTarget=<1 0 0>}
      localJointOffset=<4 0 0>
     }]
   }
   Save cached pose 'CR_Cache_Dynamics'(Pose=Component To Local(ComponentPose=@n34)) {cacheName=CR_Cache_Dynamics}
  ]}
 EventGraph {nodes [
   Event BlueprintUpdateAnimation()
   > n1 = Sequence!(execute) -> (then_0, then_1, then_2:Exec, then_3:Exec, then_4:Exec)
   > n2 = "Gravity Override State Function (Parameter Asset) (Nativized)"(
    CharacterPartAnimInstance=Self-Reference()
    GameTimeInSeconds=GetGameTimeInSeconds()
    Parameters=&/F_MED_SilverCanine_FaceAcc_Parameters
   )
   > Set GravityOverride(GravityOverride=@n2)
   > Set RBAN Settings Out()
   > n3 = Branch(Condition=Get isWearingBlackbling())
   > Set DynamicsWithBackpack(DynamicsWithBackpack:bool=true) -> (Output_Get:bool)
   > Set DynamicsWithoutBackpack(DynamicsWithoutBackpack:bool=false) -> (Output_Get:bool)
   Set DynamicsWithBackpack(execute=@n3.else, DynamicsWithBackpack=false)
   > Set DynamicsWithoutBackpack(DynamicsWithoutBackpack=true)
   n4 = Branch(execute=@n1.then_4, Condition=Get bIsInFrontEnd())
   > Set AlphaFE(AlphaFE:double=0.6) -> (Output_Get:double)
   > Set Sim Space Settings(Sim Space Settings:Anim Phys Sim Space Settings Structure=Get Sim Space Settings_FE() -> (Sim Space Settings_FE:Anim Phys Sim Space Settings Structure)) -> (Output_Get:Anim Phys Sim Space Settings Structure)
   Set AlphaFE(execute=@n4.else, AlphaFE=1)
   > Set Sim Space Settings(Sim Space Settings=Get Sim Space Settings_BR() -> (Sim Space Settings_BR:Anim Phys Sim Space Settings Structure))
   n5 = Branch(execute=@n1.then_2, Condition=Get bIsPlayingMeleeAnim())
   > Set alpha hairRigidBody(alpha hairRigidBody:double=0.6) -> (Output_Get:double)
   n6 = Branch(execute=@n5.else, Condition=Get isTacticalJumpingFalling() -> (isTacticalJumpingFalling:bool))
   > Set alpha hairRigidBody(alpha hairRigidBody=0.8)
   Set alpha hairRigidBody(execute=@n6.else, alpha hairRigidBody=1)
   n7 = Get bIsJumping()
   n8 = Get bIsTacticalSprinting()
   n9 = Get bIsFalling()
   Set isTacticalJumpingFalling(execute=@n1.then_1, isTacticalJumpingFalling:bool=OR Boolean(A=AND Boolean(A=@n7, B=@n8), B=AND Boolean(A=@n8, B=@n9))) -> (Output_Get:bool)
   n10 = float * float!(
    A="Clamp (Float)"(Value=Get PawnVelocityZ(), Min=-1000, Max=0)
    B:bool=Get bRigidBodyEnabled()
    C:bool=NOT Boolean(A=Get isWearingBlackbling())
    D:bool=NOT Boolean(A=Get bIsSkydiveDiveMode())
   ) -> (ReturnValue) {is=title K2Node_PromotableOperator}
   n11 = "Atan2 (Degrees)"!(Y:double=@n10, X:double=3800) -> (ReturnValue:double)
   Set hairBaseOverlap rotator(execute=@n1.then_3, hairBaseOverlap rotator:Rotator=MakeRotator(Pitch=float * float!(A=@n11, B:double=-1) -> (ReturnValue))) -> (Output_Get:Rotator)
   > Set hairBaseOverlap vector(hairBaseOverlap vector:Vector=MakeVector(Y=0, Z=Remap Float(Input=@n10, OldRange=(X=0,Y=-1000), NewRange=(X=0,Y=2), Clamp=true) -> (Output=~))) -> (Output_Get:Vector)
   Event OnParachuteOpened()
   > Switch on EGliderType(Selection=Get PlayerGliderType() -> (PlayerGliderType=HangGlider))
   > Set isUsingHangGlidingInParachute(isUsingHangGlidingInParachute:bool=true) -> (Output_Get:bool)
   Event OnParachuteClosed()
   > Set isUsingHangGlidingInParachute(isUsingHangGlidingInParachute=false)
   Event OnCharacterCustomizationLoaded()
   n12 = CheckForBackbling()
   > Set isWearingBlackbling(isWearingBlackbling:bool=OR Boolean!(A=@n12, B=@n12.bIsBackblingACape, C:bool=@n12.bIsBackblingLong) -> (ReturnValue)) -> (Output_Get:bool)
   > n13 = Branch(Condition=Get bIsInFrontEnd())
   > Set SimSpaceSettings output(SimSpaceSettings output:Anim Phys Sim Space Settings Structure=Get SimSpaceSettings frontEnd() -> (SimSpaceSettings frontEnd:Anim Phys Sim Space Settings Structure)) -> (Output_Get:Anim Phys Sim Space Settings Structure)
   Set SimSpaceSettings output(
    execute=@n13.else
    SimSpaceSettings output=Get SimSpaceSettings default() -> (SimSpaceSettings default:Anim Phys Sim Space Settings Structure)
   )
   Event OnWeaponChanged()
   > Set isAR Equipped(isAR Equipped:bool=false) -> (Output_Get:bool)
   > Set isPistol Equipped(isPistol Equipped:bool=false) -> (Output_Get:bool)
   > Set isSMG90(isSMG90:bool=false) -> (Output_Get:bool)
   > Set isOutLaw(isOutLaw:bool=false) -> (Output_Get:bool)
   > n14 = Switch on EFortWeaponCoreAnimation!(execute, Selection:EFortWeaponCoreAnimation Enum=Get WeaponCoreAnim!() -> (WeaponCoreAnim)) -> (
    Melee:Exec
    Pistol:Exec
    Shotgun:Exec
    PaperBlueprint:Exec
    Rifle:Exec
    MeleeOneHand:Exec
    MachinePistol:Exec
    RocketLauncher:Exec
    GrenadeLauncher:Exec
    GoingCommando:Exec
    AssaultRifle:Exec
    TacticalShotgun:Exec
    SniperRifle:Exec
    TrapPlacement:Exec
    ShoulderLauncher:Exec
    AbilityDecoTool:Exec
    Crossbow:Exec
    C4:Exec
    RemoteControl:Exec
    DualWield:Exec
    AR_BullPup:Exec
    AR_ForwardGrip:Exec
    MedPackPaddles:Exec
    SMG_P90:Exec
    AR_DrumGun:Exec
    Consumable_Small:Exec
    Consumable_Large:Exec
    Balloon:Exec
    MountedTurret:Exec
    CreativeTool:Exec
    ExplosiveBow:Exec
    AshtonIndigo:Exec
    AshtonChicago:Exec
    MeleeDualWield:Exec
    Unarmed:Exec
   )
   Set isAR Equipped(execute=@n14.AssaultRifle, isAR Equipped=true)
   Set isPistol Equipped(execute=@n14.Pistol, isPistol Equipped=true)
   Set isSMG90(execute=@n14.SMG_P90, isSMG90=true)
   Set isOutLaw(execute=@n14.TacticalShotgun, isOutLaw=true)
  ]}
]
variables [
 alpha hairRigidBody {type=number}
 AlphaFE {type=number}
 DynamicsWithBackpack {type=boolean}
 DynamicsWithoutBackpack {type=boolean}
 Gravity Override Params {
  type=FortGravityOverrideParameterSet
  default=(Skydive_GoingUp=(GravityOverride=(WindFrequency=(X=1.500000,Y=7.000000,Z=8.000000),WindAmplitude=(X=200.000000,Y=-500.000000,Z=400.000000),GravityOverride=(X=0.000000,Y=-300.000000,Z=1500.000000),JointName="root")),Skydive_FloatingDown=(GravityOverride=(WindFrequency=(X=1.500000,Y=7.000000,Z=8.000000),WindAmplitude=(X=200.000000,Y=-500.000000,Z=400.000000),WindOffset=(X=0.000000,Y=-600.000000,Z=400.000000),GravityOverride=(X=0.000000,Y=-700.000000,Z=600.000000),JointName="root")),Skydive_DivingDown=(GravityOverride=(WindFrequency=(X=1.500000,Y=12.000000,Z=10.000000),WindAmplitude=(X=50.000000,Y=300.000000,Z=100.000000),GravityOverride=(X=-1500.000000,Y=0.000000,Z=0.000000),JointName="pelvis")),Skydive_Parachuting=(GravityOverride=(WindFrequency=(X=1.500000,Y=7.000000,Z=8.000000),WindAmplitude=(X=200.000000,Y=10.000000,Z=50.000000),GravityOverride=(X=0.000000,Y=0.000000,Z=-1200.000000),JointName="root")),OnGround_Standing=(GravityOverride=(GravityOverride=(X=0.000000,Y=0.000000,Z=-980.000000),JointName="root")),OnGround_Moving=(GravityOverride=(GravityOverride=(X=0.000000,Y=0.000000,Z=-980.000000),JointName="root")),OnGround_Moving_HighRotationRate=(GravityOverride=(GravityOverride=(X=0.000000,Y=0.000000,Z=-980.000000),JointName="root")),OnGround_Crouch_Moving=(GravityOverride=(GravityOverride=(X=0.000000,Y=0.000000,Z=-980.000000),JointName="root")),Emote_or_Melee=(GravityOverride=(GravityOverride=(X=0.000000,Y=0.000000,Z=-980.000000),JointName="root")),Swimming=(GravityOverride=(GravityOverride=(X=-800.000000,Y=0.000000,Z=0.000000),JointName="pelvis")),Front_End=(GravityOverride=(GravityOverride=(X=0.000000,Y=0.000000,Z=-980.000000),JointName="root")),DBNO=(GravityOverride=(GravityOverride=(X=-980.000000,Y=400.000000,Z=0.000000),JointName="pelvis")),DBNO_Carried=(GravityOverride=(GravityOverride=(X=-980.000000,Y=500.000000,Z=-500.000000),JointName="pelvis")),OnGround_Falling=(GravityOverride=(GravityOverride=(X=0.000000,Y=0.000000,Z=-980.000000),JointName="root")),Ziplining=(GravityOverride=(GravityOverride=(X=0.000000,Y=0.000000,Z=-980.000000),JointName="root")),Skydive_Surfing=(GravityOverride=(WindFrequency=(X=1.500000,Y=7.000000,Z=8.000000),WindAmplitude=(X=200.000000,Y=-500.000000,Z=400.000000),GravityOverride=(X=500.000000,Y=-300.000000,Z=-980.000000),JointName="root")),Riding_Standing=(GravityOverride=(GravityOverride=(X=0.000000,Y=-350.000000,Z=-980.000000),JointName="root")),Riding_Moving=(GravityOverride=(GravityOverride=(X=0.000000,Y=-450.000000,Z=-450.000000),JointName="root")),Riding_Moving_Targeting=(GravityOverride=(GravityOverride=(X=0.000000,Y=-200.000000,Z=-980.000000),JointName="root")),Tactical_Sprint=(GravityOverride=(GravityOverride=(X=0.000000,Y=-450.000000,Z=-500.000000),JointName="root")),Rail_Grinding=(GravityOverride=(GravityOverride=(X=0.000000,Y=0.000000,Z=-980.000000),JointName="root")),)
 }
 GravityOverride {type=Vector}
 hairBaseOverlap rotator {type=Rotator}
 hairBaseOverlap vector {type=Vector}
 isAR Equipped {type=boolean}
 isOutLaw {type=boolean}
 isPistol Equipped {type=boolean}
 isSMG90 {type=boolean}
 isTacticalJumpingFalling {type=boolean}
 isUsingHangGlidingInParachute {type=boolean}
 isWearingBlackbling {type=boolean}
 RBAN Params {
  type=FortRigidBodyAnimNodeParameterSet
  default=(Skydive_GoingUp=(RigidBodyAnimNode=(SimSpaceSettings=(WorldAlpha=1.000000,VelocityScaleZ=0.000000,MaxLinearVelocity=700.000000,MaxAngularVelocity=3.000000,MaxLinearAcceleration=300.000000,MaxAngularAcceleration=1000.000000,ExternalLinearDragV=(X=0.000000,Y=0.500000,Z=0.000000)))),Skydive_FloatingDown=(RigidBodyAnimNode=(ComponentLinearAccScale=(X=0.000000,Y=0.500000,Z=0.000000),ComponentLinearVelScale=(X=0.000000,Y=0.500000,Z=0.000000),SimSpaceSettings=(WorldAlpha=1.000000,VelocityScaleZ=0.000000,MaxLinearVelocity=700.000000,MaxAngularVelocity=3.000000,MaxLinearAcceleration=300.000000,MaxAngularAcceleration=1000.000000,ExternalLinearDragV=(X=0.000000,Y=0.500000,Z=0.000000)))),Skydive_DivingDown=(RigidBodyAnimNode=(ComponentLinearAccScale=(X=0.100000,Y=0.000000,Z=0.000000),ComponentLinearVelScale=(X=0.100000,Y=0.000000,Z=0.000000),SimSpaceSettings=(WorldAlpha=1.000000,VelocityScaleZ=0.000000,MaxLinearVelocity=700.000000,MaxAngularVelocity=3.000000,MaxLinearAcceleration=300.000000,MaxAngularAcceleration=1000.000000,ExternalLinearDragV=(X=0.000000,Y=0.500000,Z=0.000000)))),Skydive_Parachuting=(RigidBodyAnimNode=(ComponentLinearAccScale=(X=0.000000,Y=0.100000,Z=0.300000),ComponentLinearVelScale=(X=0.000000,Y=0.100000,Z=0.300000),SimSpaceSettings=(WorldAlpha=1.000000,VelocityScaleZ=0.000000,MaxLinearVelocity=700.000000,MaxAngularVelocity=3.000000,MaxLinearAcceleration=300.000000,MaxAngularAcceleration=1000.000000,ExternalLinearDragV=(X=0.000000,Y=0.500000,Z=0.000000)))),OnGround_Standing=(RigidBodyAnimNode=(ComponentLinearAccScale=(X=0.300000,Y=0.000000,Z=0.000000),ComponentLinearVelScale=(X=0.300000,Y=0.000000,Z=0.000000),SimSpaceSettings=(WorldAlpha=0.300000,VelocityScaleZ=0.050000,DampingAlpha=0.000000,MaxLinearVelocity=400.000000,MaxAngularVelocity=4.000000,MaxLinearAcceleration=200.000000,MaxAngularAcceleration=500.000000,ExternalLinearDragV=(X=0.000000,Y=0.500000,Z=0.000000)))),OnGround_Moving=(RigidBodyAnimNode=(ComponentLinearAccScale=(X=0.300000,Y=0.000000,Z=0.000000),ComponentLinearVelScale=(X=0.300000,Y=0.000000,Z=0.000000),SimSpaceSettings=(WorldAlpha=0.600000,VelocityScaleZ=0.050000,DampingAlpha=0.000000,MaxLinearVelocity=500.000000,MaxAngularVelocity=50.000000,MaxLinearAcceleration=300.000000,MaxAngularAcceleration=500.000000,ExternalLinearDragV=(X=0.000000,Y=0.500000,Z=0.000000)))),OnGround_Moving_HighRotationRate=(RigidBodyAnimNode=(ComponentLinearAccScale=(X=0.300000,Y=0.000000,Z=0.000000),ComponentLinearVelScale=(X=0.300000,Y=0.000000,Z=0.000000),SimSpaceSettings=(WorldAlpha=1.000000,VelocityScaleZ=0.000000,MaxLinearVelocity=700.000000,MaxAngularVelocity=3.000000,MaxLinearAcceleration=300.000000,MaxAngularAcceleration=1000.000000,ExternalLinearDragV=(X=0.000000,Y=0.500000,Z=0.000000)))),OnGround_Crouch_Moving=(RigidBodyAnimNode=(SimSpaceSettings=(WorldAlpha=1.000000,VelocityScaleZ=0.000000,MaxLinearVelocity=700.000000,MaxAngularVelocity=3.000000,MaxLinearAcceleration=300.000000,MaxAngularAcceleration=1000.000000,ExternalLinearDragV=(X=0.000000,Y=0.500000,Z=0.000000)))),Emote_or_Melee=(RigidBodyAnimNode=(SimSpaceSettings=(WorldAlpha=1.000000,VelocityScaleZ=0.000000,MaxLinearVelocity=700.000000,MaxAngularVelocity=3.000000,MaxLinearAcceleration=300.000000,MaxAngularAcceleration=1000.000000,ExternalLinearDragV=(X=0.000000,Y=0.500000,Z=0.000000)))),Swimming=(RigidBodyAnimNode=(SimSpaceSettings=(WorldAlpha=1.000000,VelocityScaleZ=0.000000,MaxLinearVelocity=700.000000,MaxAngularVelocity=3.000000,MaxLinearAcceleration=300.000000,MaxAngularAcceleration=1000.000000,ExternalLinearDragV=(X=0.000000,Y=0.500000,Z=0.000000)))),Front_End=(RigidBodyAnimNode=(SimSpaceSettings=(WorldAlpha=1.000000,VelocityScaleZ=0.000000,MaxLinearVelocity=700.000000,MaxAngularVelocity=3.000000,MaxLinearAcceleration=300.000000,MaxAngularAcceleration=1000.000000,ExternalLinearDragV=(X=0.000000,Y=0.500000,Z=0.000000)))),DBNO=(RigidBodyAnimNode=(SimSpaceSettings=(WorldAlpha=1.000000,VelocityScaleZ=0.000000,MaxLinearVelocity=700.000000,MaxAngularVelocity=3.000000,MaxLinearAcceleration=300.000000,MaxAngularAcceleration=1000.000000,ExternalLinearDragV=(X=0.000000,Y=0.500000,Z=0.000000)))),DBNO_Carried=(RigidBodyAnimNode=(SimSpaceSettings=(WorldAlpha=1.000000,VelocityScaleZ=0.000000,MaxLinearVelocity=700.000000,MaxAngularVelocity=3.000000,MaxLinearAcceleration=300.000000,MaxAngularAcceleration=1000.000000,ExternalLinearDragV=(X=0.000000,Y=0.500000,Z=0.000000)))),OnGround_Falling=(RigidBodyAnimNode=(SimSpaceSettings=(WorldAlpha=1.000000,VelocityScaleZ=0.000000,MaxLinearVelocity=700.000000,MaxAngularVelocity=3.000000,MaxLinearAcceleration=300.000000,MaxAngularAcceleration=1000.000000,ExternalLinearDragV=(X=0.000000,Y=0.500000,Z=0.000000)))),Ziplining=(RigidBodyAnimNode=(SimSpaceSettings=(WorldAlpha=1.000000,VelocityScaleZ=0.000000,MaxLinearVelocity=700.000000,MaxAngularVelocity=3.000000,MaxLinearAcceleration=300.000000,MaxAngularAcceleration=1000.000000,ExternalLinearDragV=(X=0.000000,Y=0.500000,Z=0.000000)))),Skydive_Surfing=(RigidBodyAnimNode=(SimSpaceSettings=(WorldAlpha=1.000000,VelocityScaleZ=0.000000,MaxLinearVelocity=700.000000,MaxAngularVelocity=3.000000,MaxLinearAcceleration=300.000000,MaxAngularAcceleration=1000.000000,ExternalLinearDragV=(X=0.000000,Y=0.500000,Z=0.000000)))),Riding_Standing=(RigidBodyAnimNode=(SimSpaceSettings=(WorldAlpha=1.000000,VelocityScaleZ=0.000000,MaxLinearVelocity=700.000000,MaxAngularVelocity=3.000000,MaxLinearAcceleration=300.000000,MaxAngularAcceleration=1000.000000,ExternalLinearDragV=(X=0.000000,Y=0.500000,Z=0.000000)))),Riding_Moving=(RigidBodyAnimNode=(SimSpaceSettings=(WorldAlpha=1.000000,VelocityScaleZ=0.000000,MaxLinearVelocity=700.000000,MaxAngularVelocity=3.000000,MaxLinearAcceleration=300.000000,MaxAngularAcceleration=1000.000000,ExternalLinearDragV=(X=0.000000,Y=0.500000,Z=0.000000)))),Riding_Moving_Targeting=(RigidBodyAnimNode=(SimSpaceSettings=(WorldAlpha=1.000000,VelocityScaleZ=0.000000,MaxLinearVelocity=700.000000,MaxAngularVelocity=3.000000,MaxLinearAcceleration=300.000000,MaxAngularAcceleration=1000.000000,ExternalLinearDragV=(X=0.000000,Y=0.500000,Z=0.000000)))),Tactical_Sprint=(RigidBodyAnimNode=(ComponentLinearAccScale=(X=0.300000,Y=0.000000,Z=0.000000),ComponentLinearVelScale=(X=0.300000,Y=0.000000,Z=0.000000),SimSpaceSettings=(WorldAlpha=1.000000,VelocityScaleZ=0.000000,MaxLinearVelocity=700.000000,MaxAngularVelocity=3.000000,MaxLinearAcceleration=300.000000,MaxAngularAcceleration=1000.000000,ExternalLinearDragV=(X=0.000000,Y=0.500000,Z=0.000000)))),Rail_Grinding=(RigidBodyAnimNode=(SimSpaceSettings=(WorldAlpha=1.000000,VelocityScaleZ=0.000000,MaxLinearVelocity=700.000000,MaxAngularVelocity=3.000000,MaxLinearAcceleration=300.000000,MaxAngularAcceleration=1000.000000,ExternalLinearDragV=(X=0.000000,Y=0.500000,Z=0.000000)))),)
 }
 RBAN Settings Out {type=FortRigidBodyAnimNodeParameters}
 Sim Space Settings {
  type=AnimPhysSimSpaceSettings
  default {maxAngularAcceleration=10000, maxAngularVelocity=10000, simSpaceAngularAlpha=0, externalAngularVelocity=<0 0 0>}
 }
 Sim Space Settings_BR {
  type=AnimPhysSimSpaceSettings
  default {maxAngularAcceleration=10000, maxAngularVelocity=10000, simSpaceAngularAlpha=0, externalAngularVelocity=<0 0 0>}
 }
 Sim Space Settings_FE {
  type=AnimPhysSimSpaceSettings
  default {maxAngularAcceleration=10000, maxAngularVelocity=10000, simSpaceAngularAlpha=0.3, externalAngularVelocity=<0.2 0.2 0.2>}
 }
 SimSpaceSettings default {
  type=AnimPhysSimSpaceSettings
  default {maxAngularAcceleration=10000, maxAngularVelocity=10000, simSpaceAngularAlpha=0, externalAngularVelocity=<0 0 0>}
 }
 SimSpaceSettings frontEnd {
  type=AnimPhysSimSpaceSettings
  default {maxAngularAcceleration=10000, maxAngularVelocity=10000, simSpaceAngularAlpha=0, externalAngularVelocity=<0 0 0>}
 }
 SimSpaceSettings output {
  type=AnimPhysSimSpaceSettings
  default {maxAngularAcceleration=10000, maxAngularVelocity=10000, simSpaceAngularAlpha=0, externalAngularVelocity=<0 0 0>}
 }
]
