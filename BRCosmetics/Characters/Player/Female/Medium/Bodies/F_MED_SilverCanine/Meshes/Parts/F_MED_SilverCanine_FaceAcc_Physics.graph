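# Physics asset for the face accessory of the female medium SilverCanine body cosmetic, defining two left-side dangling bone chains, front and side.
# Each chain has seven capsule segments linked from a base bone on the head, with tight swing limits and zero-degree twist limits between segments.
# Larger capsules on the head, upper spine, clavicles and left upper arm form a body proxy, joined by free angular constraints.
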PhysicsAsset /BRCosmetics/Characters/Player/Female/Medium/Bodies/F_MED_SilverCanine/Meshes/Parts/F_MED_SilverCanine_FaceAcc_Physics
bodies {
 base_front_1_l [{shapeType=Capsule, length=0.505, radius=0.505, shapeName=base_front_1_l_capsule, rotation=<rot 0 -90 0>}]
 base_side_1_l [{shapeType=Capsule, length=3, radius=3, shapeName=base_side_1_l_capsule, center=<2 0 0>, rotation=<rot 90 0 0>}]
 clavicle_l [{shapeType=Capsule, length=10, radius=6, shapeName=clavicle_l_capsule, center=<0.2357 -2.7068 0>, rotation=<rot 90 0 0>}]
 clavicle_r [{shapeType=Capsule, length=7, radius=5, shapeName=clavicle_r_capsule, center=<1 0 0>, rotation=<rot 90 0 0>}]
 dyn_front_1_l [{shapeType=Capsule, length=3, radius=3, shapeName=dyn_front_1_l_capsule, center=<1.5 0 0>, rotation=<rot 90 0 0>}]
 dyn_front_2_l [{shapeType=Capsule, length=3, radius=3, shapeName=dyn_front_2_l_capsule, center=<1.5 0 0>, rotation=<rot 90 0 0>}]
 dyn_front_3_l [{shapeType=Capsule, length=3, radius=3, shapeName=dyn_front_3_l_capsule, center=<1.5 0 0>, rotation=<rot 90 0 0>}]
 dyn_front_4_l [{shapeType=Capsule, length=3, radius=3, shapeName=dyn_front_4_l_capsule, center=<1.5 0 0>, rotation=<rot 90 0 0>}]
 dyn_front_5_l [{shapeType=Capsule, length=3, radius=3, shapeName=dyn_front_5_l_capsule, center=<1.5 0 0>, rotation=<rot 90 0 0>}]
 dyn_front_6_l [{shapeType=Capsule, length=3, radius=3, shapeName=dyn_front_6_l_capsule, center=<1.5 0 0>, rotation=<rot 90 0 0>}]
 dyn_front_7_l [{shapeType=Capsule, length=3, radius=3, shapeName=dyn_front_7_l_capsule, center=<1.5 0 0>, rotation=<rot 90 0 0>}]
 dyn_side_1_l [{shapeType=Capsule, length=3, radius=2, shapeName=dyn_side_1_l_capsule, center=<1.5 0 0>, rotation=<rot 90 0 0>}]
 dyn_side_2_l [{shapeType=Capsule, length=3, radius=2, shapeName=dyn_side_2_l_capsule, center=<1.5 0 0>, rotation=<rot 90 0 0>}]
 dyn_side_3_l [{shapeType=Capsule, length=3, radius=2, shapeName=dyn_side_3_l_capsule, center=<1.5 0 0>, rotation=<rot 90 0 0>}]
 dyn_side_4_l [{shapeType=Capsule, length=3, radius=2, shapeName=dyn_side_4_l_capsule, center=<1.5 0 0>, rotation=<rot 90 0 0>}]
 dyn_side_5_l [{shapeType=Capsule, length=3, radius=2, shapeName=dyn_side_5_l_capsule, center=<1.5 0 0>, rotation=<rot 90 0 0>}]
 dyn_side_6_l [{shapeType=Capsule, length=3, radius=2, shapeName=dyn_side_6_l_capsule, center=<1.5 0 0>, rotation=<rot 90 0 0>}]
 dyn_side_7_l [{shapeType=Capsule, length=3, radius=2, shapeName=dyn_side_7_l_capsule, center=<1.5 0 0>, rotation=<rot 90 0 0>}]
 head [{shapeType=Capsule, length=5, radius=10, center=<1 0 0>, rotation=<rot 90 0 0>}]
 spine_04 [{shapeType=Capsule, length=5, radius=10, shapeName=spine_04_capsule, center=<1 0 0>}]
 spine_05 [{shapeType=Capsule, length=6, radius=9, shapeName=spine_05_capsule, center=<1 -0.524 0>}]
 upperarm_l [{shapeType=Capsule, length=10, radius=4.4, shapeName=upperarm_l_capsule, center=<6.6217 0 -0.3525>, rotation=<rot 90 0 0>}]
}
constraints [
 {
  bone1Name=base_front_1_l
  bone2Name=head
  swing1LimitDegrees=45
  swing1Motion=Free
  swing2LimitDegrees=45
  swing2Motion=Free
  twistLimitDegrees=45
  twistMotion=Free
 }
 {bone1Name=dyn_front_1_l, bone2Name=base_front_1_l, swing1LimitDegrees=10, swing2LimitDegrees=10, twistLimitDegrees=0}
 {bone1Name=dyn_front_2_l, bone2Name=dyn_front_1_l, swing1LimitDegrees=10, swing2LimitDegrees=10, twistLimitDegrees=0}
 {bone1Name=dyn_front_3_l, bone2Name=dyn_front_2_l, swing1LimitDegrees=10, swing2LimitDegrees=10, twistLimitDegrees=0}
 {bone1Name=dyn_front_4_l, bone2Name=dyn_front_3_l, swing1LimitDegrees=10, swing2LimitDegrees=10, twistLimitDegrees=0}
 {bone1Name=dyn_front_5_l, bone2Name=dyn_front_4_l, swing1LimitDegrees=10, swing2LimitDegrees=10, twistLimitDegrees=0}
 {bone1Name=dyn_front_6_l, bone2Name=dyn_front_5_l, swing1LimitDegrees=10, swing2LimitDegrees=10, twistLimitDegrees=0}
 {bone1Name=dyn_front_7_l, bone2Name=dyn_front_6_l, swing1LimitDegrees=10, swing2LimitDegrees=10, twistLimitDegrees=0}
 {
  bone1Name=head
  bone2Name=spine_05
  swing1LimitDegrees=45
  swing1Motion=Free
  swing2LimitDegrees=45
  swing2Motion=Free
  twistLimitDegrees=45
  twistMotion=Free
 }
 {
  bone1Name=clavicle_l
  bone2Name=spine_05
  swing1LimitDegrees=45
  swing1Motion=Free
  swing2LimitDegrees=45
  swing2Motion=Free
  twistLimitDegrees=45
  twistMotion=Free
 }
 {
  bone1Name=clavicle_r
  bone2Name=spine_05
  swing1LimitDegrees=45
  swing1Motion=Free
  swing2LimitDegrees=45
  swing2Motion=Free
  twistLimitDegrees=45
  twistMotion=Free
 }
 {
  bone1Name=spine_05
  bone2Name=spine_04
  swing1LimitDegrees=45
  swing1Motion=Free
  swing2LimitDegrees=45
  swing2Motion=Free
  twistLimitDegrees=45
  twistMotion=Free
 }
 {bone1Name=base_side_1_l, bone2Name=head, swing1LimitDegrees=15, swing2LimitDegrees=15, twistLimitDegrees=0}
 {bone1Name=dyn_side_1_l, bone2Name=base_side_1_l, swing1LimitDegrees=10, swing2LimitDegrees=10, twistLimitDegrees=0}
 {bone1Name=dyn_side_2_l, bone2Name=dyn_side_1_l, swing1LimitDegrees=10, swing2LimitDegrees=10, twistLimitDegrees=0}
 {bone1Name=dyn_side_3_l, bone2Name=dyn_side_2_l, swing1LimitDegrees=10, swing2LimitDegrees=10, twistLimitDegrees=0}
 {bone1Name=dyn_side_4_l, bone2Name=dyn_side_3_l, swing1LimitDegrees=10, swing2LimitDegrees=10, twistLimitDegrees=0}
 {bone1Name=dyn_side_5_l, bone2Name=dyn_side_4_l, swing1LimitDegrees=10, swing2LimitDegrees=10, twistLimitDegrees=0}
 {bone1Name=dyn_side_6_l, bone2Name=dyn_side_5_l, swing1LimitDegrees=10, swing2LimitDegrees=10, twistLimitDegrees=0}
 {bone1Name=dyn_side_7_l, bone2Name=dyn_side_6_l, swing1LimitDegrees=10, swing2LimitDegrees=10, twistLimitDegrees=0}
 {
  bone1Name=upperarm_l
  bone2Name=clavicle_l
  swing1LimitDegrees=45
  swing1Motion=Free
  swing2LimitDegrees=45
  swing2Motion=Free
  twistLimitDegrees=45
  twistMotion=Free
 }
]
